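# Physics asset of cape colliders for the medium male Headband_S player body.
# Capsules cover the pelvis, spine, clavicles and neck, a sphere covers the head, and arm, leg and foot bodies have no shapes.
# Five back and side skirt bone chains carry identical-sized capsules, and a long capsule covers the sword bone.
# Constraints follow the bone hierarchy, with skirt chains and sword attached to the pelvis, all swing and twist free.
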PhysicsAsset /BRCosmetics/Characters/Player/Male/Medium/Bodies/M_MED_Headband_S/Meshes/M_MED_Headband_S_CapeColliders
bodies {
 calf_l []
 calf_r []
 clavicle_l [{shapeType=Capsule, length=7.7636, radius=11.1427, center=<9.9642 -0.7741 -1.5923>, rotation=<rot 75.9352 -0.0001 175.5622>}]
 clavicle_r [{shapeType=Capsule, length=7.7636, radius=11.1427, center=<-9.9642 0.7741 1.5923>, rotation=<rot 75.9351 -0.0001 -4.4378>}]
 dyn_skirt_bk_a [{shapeType=Capsule, length=15, radius=10, shapeName=dyn_skirt_bk_a_capsule, center=<5 0 -6>, rotation=<rot 90 0 0>}]
 dyn_skirt_bk_a_01 [{shapeType=Capsule, length=15, radius=10, shapeName=dyn_skirt_bk_a_01_capsule, center=<5 0 -6>, rotation=<rot 90 0 0>}]
 dyn_skirt_bk_a_02 [{shapeType=Capsule, length=15, radius=10, shapeName=dyn_skirt_bk_a_02_capsule, center=<5 0 -6>, rotation=<rot 90 0 0>}]
 dyn_skirt_bk_a_03 [{shapeType=Capsule, length=15, radius=10, shapeName=dyn_skirt_bk_a_03_capsule, center=<5 0 -6>, rotation=<rot 90 0 0>}]
 dyn_skirt_bk_a_04 [{shapeType=Capsule, length=15, radius=10, shapeName=dyn_skirt_bk_a_04_capsule, center=<5 0 -6>, rotation=<rot 90 0 0>}]
 dyn_skirt_bk_a_l [{shapeType=Capsule, length=15, radius=10, shapeName=dyn_skirt_bk_a_l_capsule, center=<5 0 -6>, rotation=<rot 90 0 0>}]
 dyn_skirt_bk_a_r [{shapeType=Capsule, length=15, radius=10, shapeName=dyn_skirt_bk_a_r_capsule, center=<-5 0 6>, rotation=<rot 90 0 0>}]
 dyn_skirt_bk_b_l [{shapeType=Capsule, length=15, radius=10, shapeName=dyn_skirt_bk_b_l_capsule, center=<5 0 -6>, rotation=<rot 90 0 0>}]
 dyn_skirt_bk_b_l_01 [{shapeType=Capsule, length=15, radius=10, shapeName=dyn_skirt_bk_b_l_01_capsule, center=<5 0 -6>, rotation=<rot 90 0 0>}]
 dyn_skirt_bk_b_l_02 [{shapeType=Capsule, length=15, radius=10, shapeName=dyn_skirt_bk_b_l_02_capsule, center=<5 0 -6>, rotation=<rot 90 0 0>}]
 dyn_skirt_bk_b_l_03 [{shapeType=Capsule, length=15, radius=10, shapeName=dyn_skirt_bk_b_l_03_capsule, center=<5 0 -6>, rotation=<rot 90 0 0>}]
 dyn_skirt_bk_b_r [{shapeType=Capsule, length=15, radius=10, shapeName=dyn_skirt_bk_b_r_capsule, center=<5 0 -6>, rotation=<rot 90 0 0>}]
 dyn_skirt_bk_b_r_01 [{shapeType=Capsule, length=15, radius=10, shapeName=dyn_skirt_bk_b_r_01_capsule, center=<5 0 -6>, rotation=<rot 90 0 0>}]
 dyn_skirt_bk_b_r_02 [{shapeType=Capsule, length=15, radius=10, shapeName=dyn_skirt_bk_b_r_02_capsule, center=<5 0 -6>, rotation=<rot 90 0 0>}]
 dyn_skirt_bk_b_r_03 [{shapeType=Capsule, length=15, radius=10, shapeName=dyn_skirt_bk_b_r_03_capsule, center=<5 0 -6>, rotation=<rot 90 0 0>}]
 dyn_skirt_side_a_l [{shapeType=Capsule, length=15, radius=10, shapeName=dyn_skirt_side_a_l_capsule, center=<5 0 -6>, rotation=<rot 90 0 0>}]
 dyn_skirt_side_a_r [{shapeType=Capsule, length=15, radius=10, shapeName=dyn_skirt_side_a_r_capsule, center=<-5 0 6>, rotation=<rot 90 0 0>}]
 dyn_skirt_side_b_l [{shapeType=Capsule, length=15, radius=10, shapeName=dyn_skirt_side_b_l_capsule, center=<5 0 -6>, rotation=<rot 90 0 0>}]
 dyn_skirt_side_b_l_01 [{shapeType=Capsule, length=15, radius=10, shapeName=dyn_skirt_side_b_l_01_capsule, center=<5 0 -6>, rotation=<rot 90 0 0>}]
 dyn_skirt_side_b_l_02 [{shapeType=Capsule, length=15, radius=10, shapeName=dyn_skirt_side_b_l_02_capsule, center=<5 0 -6>, rotation=<rot 90 0 0>}]
 dyn_skirt_side_b_l_03 [{shapeType=Capsule, length=15, radius=10, shapeName=dyn_skirt_side_b_l_03_capsule, center=<5 0 -6>, rotation=<rot 90 0 0>}]
 dyn_skirt_side_b_r [{shapeType=Capsule, length=15, radius=10, shapeName=dyn_skirt_side_b_r_capsule, center=<5 0 -6>, rotation=<rot 90 0 0>}]
 dyn_skirt_side_b_r_01 [{shapeType=Capsule, length=15, radius=10, shapeName=dyn_skirt_side_b_r_01_capsule, center=<5 0 -6>, rotation=<rot 90 0 0>}]
 dyn_skirt_side_b_r_02 [{shapeType=Capsule, length=15, radius=10, shapeName=dyn_skirt_side_b_r_02_capsule, center=<5 0 -6>, rotation=<rot 90 0 0>}]
 dyn_skirt_side_b_r_03 [{shapeType=Capsule, length=15, radius=10, shapeName=dyn_skirt_side_b_r_03_capsule, center=<5 0 -6>, rotation=<rot 90 0 0>}]
 dyn_sword [{
   shapeType=Capsule
   length=108.5775
   radius=3.3609
   shapeName=dyn_sword_capsule
   center=<4.3565 11.2334 1.0004>
   rotation=<rot 0 159.0636 -87.064>
  }]
 foot_l []
 foot_r []
 head [{shapeType=Sphere, radius=17.3125, center=<11.028 -0.174 0>}]
 lowerarm_l []
 lowerarm_r []
 neck_02 [{shapeType=Capsule, length=13.4375, radius=7.3175, center=<-0.2642 -0.3408 0>, rotation=<rot 90 -171.1193 -175.3978>}]
 pelvis [{shapeType=Capsule, length=12.2713, radius=13.5388, center=<-2.6215 -1.3695 0>}]
 spine_01 [{shapeType=Capsule, length=10.2338, radius=13.7194, center=<-2.6215 -0.7156 0>}]
 spine_02 [
  {shapeType=Capsule, length=6.7286, radius=14.8101, center=<-2.6215 0.4993 0.0001>}
  {shapeType=Capsule, length=29.875, radius=11.1013, center=<2.5024 0.4992 3.9825>, rotation=<rot 90 0.0049 0.0049>}
  {shapeType=Capsule, length=29.875, radius=11.1013, center=<2.4704 0.4991 -3.9825>, rotation=<rot 90 -0.0396 -0.0396>}
 ]
 spine_03 [{shapeType=Capsule, length=5.4697, radius=14.444, center=<-1.953 0.1364 0.0001>}]
 spine_04 [{shapeType=Capsule, length=14.6562, radius=13.6325, center=<2.1467 -1.2748 0.0002>}]
 spine_05 [
  {shapeType=Capsule, length=17.8437, radius=17.195, center=<-1.7514 2.4879 0>}
  {shapeType=Capsule, length=12.7812, radius=17.0625, center=<0.7875 3.1143 0>}
 ]
 thigh_l []
 thigh_r []
 upperarm_l []
 upperarm_r []
}
constraints [
 {
  bone1Name=spine_01
  bone2Name=pelvis
  swing1LimitDegrees=45
  swing1Motion=Free
  swing2LimitDegrees=45
  swing2Motion=Free
  twistLimitDegrees=45
  twistMotion=Free
 }
 {
  bone1Name=spine_02
  bone2Name=spine_01
  swing1LimitDegrees=45
  swing1Motion=Free
  swing2LimitDegrees=45
  swing2Motion=Free
  twistLimitDegrees=45
  twistMotion=Free
 }
 {
  bone1Name=spine_03
  bone2Name=spine_02
  swing1LimitDegrees=45
  swing1Motion=Free
  swing2LimitDegrees=45
  swing2Motion=Free
  twistLimitDegrees=45
  twistMotion=Free
 }
 {
  bone1Name=spine_04
  bone2Name=spine_03
  swing1LimitDegrees=45
  swing1Motion=Free
  swing2LimitDegrees=45
  swing2Motion=Free
  twistLimitDegrees=45
  twistMotion=Free
 }
 {
  bone1Name=spine_05
  bone2Name=spine_04
  swing1LimitDegrees=45
  swing1Motion=Free
  swing2LimitDegrees=45
  swing2Motion=Free
  twistLimitDegrees=45
  twistMotion=Free
 }
 {
  bone1Name=clavicle_l
  bone2Name=spine_05
  swing1LimitDegrees=45
  swing1Motion=Free
  swing2LimitDegrees=45
  swing2Motion=Free
  twistLimitDegrees=45
  twistMotion=Free
 }
 {
  bone1Name=upperarm_l
  bone2Name=clavicle_l
  swing1LimitDegrees=45
  swing1Motion=Free
  swing2LimitDegrees=45
  swing2Motion=Free
  twistLimitDegrees=45
  twistMotion=Free
 }
 {
  bone1Name=lowerarm_l
  bone2Name=upperarm_l
  swing1LimitDegrees=45
  swing1Motion=Free
  swing2LimitDegrees=45
  swing2Motion=Free
  twistLimitDegrees=45
  twistMotion=Free
 }
 {
  bone1Name=neck_02
  bone2Name=spine_05
  swing1LimitDegrees=45
  swing1Motion=Free
  swing2LimitDegrees=45
  swing2Motion=Free
  twistLimitDegrees=45
  twistMotion=Free
 }
 {
  bone1Name=head
  bone2Name=neck_02
  swing1LimitDegrees=45
  swing1Motion=Free
  swing2LimitDegrees=45
  swing2Motion=Free
  twistLimitDegrees=45
  twistMotion=Free
 }
 {
  bone1Name=thigh_l
  bone2Name=pelvis
  swing1LimitDegrees=45
  swing1Motion=Free
  swing2LimitDegrees=45
  swing2Motion=Free
  twistLimitDegrees=45
  twistMotion=Free
 }
 {
  bone1Name=calf_l
  bone2Name=thigh_l
  swing1LimitDegrees=45
  swing1Motion=Free
  swing2LimitDegrees=45
  swing2Motion=Free
  twistLimitDegrees=45
  twistMotion=Free
 }
 {
  bone1Name=foot_l
  bone2Name=calf_l
  swing1LimitDegrees=45
  swing1Motion=Free
  swing2LimitDegrees=45
  swing2Motion=Free
  twistLimitDegrees=45
  twistMotion=Free
 }
 {
  bone1Name=clavicle_r
  bone2Name=spine_05
  swing1LimitDegrees=45
  swing1Motion=Free
  swing2LimitDegrees=45
  swing2Motion=Free
  twistLimitDegrees=45
  twistMotion=Free
 }
 {
  bone1Name=upperarm_r
  bone2Name=clavicle_r
  swing1LimitDegrees=45
  swing1Motion=Free
  swing2LimitDegrees=45
  swing2Motion=Free
  twistLimitDegrees=45
  twistMotion=Free
 }
 {
  bone1Name=lowerarm_r
  bone2Name=upperarm_r
  swing1LimitDegrees=45
  swing1Motion=Free
  swing2LimitDegrees=45
  swing2Motion=Free
  twistLimitDegrees=45
  twistMotion=Free
 }
 {
  bone1Name=thigh_r
  bone2Name=pelvis
  swing1LimitDegrees=45
  swing1Motion=Free
  swing2LimitDegrees=45
  swing2Motion=Free
  twistLimitDegrees=45
  twistMotion=Free
 }
 {
  bone1Name=calf_r
  bone2Name=thigh_r
  swing1LimitDegrees=45
  swing1Motion=Free
  swing2LimitDegrees=45
  swing2Motion=Free
  twistLimitDegrees=45
  twistMotion=Free
 }
 {
  bone1Name=foot_r
  bone2Name=calf_r
  swing1LimitDegrees=45
  swing1Motion=Free
  swing2LimitDegrees=45
  swing2Motion=Free
  twistLimitDegrees=45
  twistMotion=Free
 }
 {
  bone1Name=dyn_skirt_side_b_l_03
  bone2Name=dyn_skirt_side_b_l_02
  swing1LimitDegrees=45
  swing1Motion=Free
  swing2LimitDegrees=45
  swing2Motion=Free
  twistLimitDegrees=45
  twistMotion=Free
 }
 {
  bone1Name=dyn_skirt_side_a_l
  bone2Name=pelvis
  swing1LimitDegrees=45
  swing1Motion=Free
  swing2LimitDegrees=45
  swing2Motion=Free
  twistLimitDegrees=45
  twistMotion=Free
 }
 {
  bone1Name=dyn_skirt_side_b_l
  bone2Name=dyn_skirt_side_a_l
  swing1LimitDegrees=45
  swing1Motion=Free
  swing2LimitDegrees=45
  swing2Motion=Free
  twistLimitDegrees=45
  twistMotion=Free
 }
 {
  bone1Name=dyn_skirt_side_b_l_01
  bone2Name=dyn_skirt_side_b_l
  swing1LimitDegrees=45
  swing1Motion=Free
  swing2LimitDegrees=45
  swing2Motion=Free
  twistLimitDegrees=45
  twistMotion=Free
 }
 {
  bone1Name=dyn_skirt_side_b_l_02
  bone2Name=dyn_skirt_side_b_l_01
  swing1LimitDegrees=45
  swing1Motion=Free
  swing2LimitDegrees=45
  swing2Motion=Free
  twistLimitDegrees=45
  twistMotion=Free
 }
 {
  bone1Name=dyn_skirt_bk_b_l_03
  bone2Name=dyn_skirt_bk_b_l_02
  swing1LimitDegrees=45
  swing1Motion=Free
  swing2LimitDegrees=45
  swing2Motion=Free
  twistLimitDegrees=45
  twistMotion=Free
 }
 {
  bone1Name=dyn_skirt_bk_a_l
  bone2Name=pelvis
  swing1LimitDegrees=45
  swing1Motion=Free
  swing2LimitDegrees=45
  swing2Motion=Free
  twistLimitDegrees=45
  twistMotion=Free
 }
 {
  bone1Name=dyn_skirt_bk_b_l
  bone2Name=dyn_skirt_bk_a_l
  swing1LimitDegrees=45
  swing1Motion=Free
  swing2LimitDegrees=45
  swing2Motion=Free
  twistLimitDegrees=45
  twistMotion=Free
 }
 {
  bone1Name=dyn_skirt_bk_b_l_01
  bone2Name=dyn_skirt_bk_b_l
  swing1LimitDegrees=45
  swing1Motion=Free
  swing2LimitDegrees=45
  swing2Motion=Free
  twistLimitDegrees=45
  twistMotion=Free
 }
 {
  bone1Name=dyn_skirt_bk_b_l_02
  bone2Name=dyn_skirt_bk_b_l_01
  swing1LimitDegrees=45
  swing1Motion=Free
  swing2LimitDegrees=45
  swing2Motion=Free
  twistLimitDegrees=45
  twistMotion=Free
 }
 {
  bone1Name=dyn_skirt_side_b_r_03
  bone2Name=dyn_skirt_side_b_r_02
  swing1LimitDegrees=45
  swing1Motion=Free
  swing2LimitDegrees=45
  swing2Motion=Free
  twistLimitDegrees=45
  twistMotion=Free
 }
 {
  bone1Name=dyn_skirt_side_a_r
  bone2Name=pelvis
  swing1LimitDegrees=45
  swing1Motion=Free
  swing2LimitDegrees=45
  swing2Motion=Free
  twistLimitDegrees=45
  twistMotion=Free
 }
 {
  bone1Name=dyn_skirt_side_b_r
  bone2Name=dyn_skirt_side_a_r
  swing1LimitDegrees=45
  swing1Motion=Free
  swing2LimitDegrees=45
  swing2Motion=Free
  twistLimitDegrees=45
  twistMotion=Free
 }
 {
  bone1Name=dyn_skirt_side_b_r_01
  bone2Name=dyn_skirt_side_b_r
  swing1LimitDegrees=45
  swing1Motion=Free
  swing2LimitDegrees=45
  swing2Motion=Free
  twistLimitDegrees=45
  twistMotion=Free
 }
 {
  bone1Name=dyn_skirt_side_b_r_02
  bone2Name=dyn_skirt_side_b_r_01
  swing1LimitDegrees=45
  swing1Motion=Free
  swing2LimitDegrees=45
  swing2Motion=Free
  twistLimitDegrees=45
  twistMotion=Free
 }
 {
  bone1Name=dyn_skirt_bk_b_r_03
  bone2Name=dyn_skirt_bk_b_r_02
  swing1LimitDegrees=45
  swing1Motion=Free
  swing2LimitDegrees=45
  swing2Motion=Free
  twistLimitDegrees=45
  twistMotion=Free
 }
 {
  bone1Name=dyn_skirt_bk_a_r
  bone2Name=pelvis
  swing1LimitDegrees=45
  swing1Motion=Free
  swing2LimitDegrees=45
  swing2Motion=Free
  twistLimitDegrees=45
  twistMotion=Free
 }
 {
  bone1Name=dyn_skirt_bk_b_r
  bone2Name=dyn_skirt_bk_a_r
  swing1LimitDegrees=45
  swing1Motion=Free
  swing2LimitDegrees=45
  swing2Motion=Free
  twistLimitDegrees=45
  twistMotion=Free
 }
 {
  bone1Name=dyn_skirt_bk_b_r_01
  bone2Name=dyn_skirt_bk_b_r
  swing1LimitDegrees=45
  swing1Motion=Free
  swing2LimitDegrees=45
  swing2Motion=Free
  twistLimitDegrees=45
  twistMotion=Free
 }
 {
  bone1Name=dyn_skirt_bk_b_r_02
  bone2Name=dyn_skirt_bk_b_r_01
  swing1LimitDegrees=45
  swing1Motion=Free
  swing2LimitDegrees=45
  swing2Motion=Free
  twistLimitDegrees=45
  twistMotion=Free
 }
 {
  bone1Name=dyn_skirt_bk_a_04
  bone2Name=dyn_skirt_bk_a_03
  swing1LimitDegrees=45
  swing1Motion=Free
  swing2LimitDegrees=45
  swing2Motion=Free
  twistLimitDegrees=45
  twistMotion=Free
 }
 {
  bone1Name=dyn_skirt_bk_a
  bone2Name=pelvis
  swing1LimitDegrees=45
  swing1Motion=Free
  swing2LimitDegrees=45
  swing2Motion=Free
  twistLimitDegrees=45
  twistMotion=Free
 }
 {
  bone1Name=dyn_skirt_bk_a_01
  bone2Name=dyn_skirt_bk_a
  swing1LimitDegrees=45
  swing1Motion=Free
  swing2LimitDegrees=45
  swing2Motion=Free
  twistLimitDegrees=45
  twistMotion=Free
 }
 {
  bone1Name=dyn_skirt_bk_a_02
  bone2Name=dyn_skirt_bk_a_01
  swing1LimitDegrees=45
  swing1Motion=Free
  swing2LimitDegrees=45
  swing2Motion=Free
  twistLimitDegrees=45
  twistMotion=Free
 }
 {
  bone1Name=dyn_skirt_bk_a_03
  bone2Name=dyn_skirt_bk_a_02
  swing1LimitDegrees=45
  swing1Motion=Free
  swing2LimitDegrees=45
  swing2Motion=Free
  twistLimitDegrees=45
  twistMotion=Free
 }
 {
  bone1Name=dyn_sword
  bone2Name=pelvis
  swing1LimitDegrees=45
  swing1Motion=Free
  swing2LimitDegrees=45
  swing2Motion=Free
  twistLimitDegrees=45
  twistMotion=Free
 }
]
