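# Physics asset for the skirt on the medium male Headband_S character, front-end variant, with capsule bodies for the skirt, pelvis and legs.
# Front, side and back bone chains hang from pelvis-linked base bodies with free rotation, mostly using limited swing and tight twist below.
# Extra free-rotation cross-constraints tie the back centre chain to both back side chains, and front-left to side-left only.
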PhysicsAsset /BRCosmetics/Characters/Player/Male/Medium/Bodies/M_MED_Headband_S/Meshes/M_MED_Headband_S_FrontEnd_skirt_Physics
bodies {
 calf_l [{shapeType=Capsule, length=26.0124, radius=13.7789, center=<-25 -2 0>, rotation=<rot 90 0 0>}]
 calf_r [{shapeType=Capsule, length=26.0124, radius=13.7789, center=<25 2 0.1223>, rotation=<rot 90 0 0>}]
 dyn_skirt_bk_a [{shapeType=Capsule, length=7.8409, radius=6.1591, shapeName=dyn_skirt_bk_a_capsule, center=<5 0 5>, rotation=<rot 0 0 90>}]
 dyn_skirt_bk_a_01 [{shapeType=Capsule, length=10, radius=7.5, shapeName=dyn_skirt_bk_a_01_capsule, center=<5 0 5>, rotation=<rot 0 0 90>}]
 dyn_skirt_bk_a_02 [{shapeType=Capsule, length=10, radius=7.5, shapeName=dyn_skirt_bk_a_02_capsule, center=<6 0 5>, rotation=<rot 0 0 90>}]
 dyn_skirt_bk_a_03 [{shapeType=Capsule, length=10, radius=7.5, shapeName=dyn_skirt_bk_a_03_capsule, center=<6 0 5>, rotation=<rot 0 0 90>}]
 dyn_skirt_bk_a_04 [{shapeType=Capsule, length=10, radius=7.5, shapeName=dyn_skirt_bk_a_04_capsule, center=<6 0 5>, rotation=<rot 0 0 90>}]
 dyn_skirt_bk_a_base [{shapeType=Capsule, length=0.505, radius=0.505, shapeName=dyn_skirt_bk_a_base_capsule, rotation=<rot 0 -90 0>}]
 dyn_skirt_bk_a_base_l [{shapeType=Capsule, length=0.505, radius=0.505, shapeName=dyn_skirt_bk_a_base_l_capsule, rotation=<rot 0 -90 0>}]
 dyn_skirt_bk_a_base_r [{shapeType=Capsule, length=0.505, radius=0.505, shapeName=dyn_skirt_bk_a_base_r_capsule, rotation=<rot 0 -90 0>}]
 dyn_skirt_bk_a_l [{shapeType=Capsule, length=8, radius=7, shapeName=dyn_skirt_bk_a_l_capsule, center=<5 0 5>, rotation=<rot 0 0 90>}]
 dyn_skirt_bk_a_r [{shapeType=Capsule, length=8, radius=7, shapeName=dyn_skirt_bk_a_r_capsule, center=<-5 0 -5>, rotation=<rot 0 0 90>}]
 dyn_skirt_bk_b_l [{shapeType=Capsule, length=6.8156, radius=8, shapeName=dyn_skirt_bk_b_l_capsule, center=<5 0 5>, rotation=<rot 0 0 90>}]
 dyn_skirt_bk_b_l_01 [{shapeType=Capsule, length=22.6293, radius=8, shapeName=dyn_skirt_bk_b_l_01_capsule, center=<5 0 5>, rotation=<rot 0 0 86.6655>}]
 dyn_skirt_bk_b_l_02 [{shapeType=Capsule, length=34.9312, radius=9, shapeName=dyn_skirt_bk_b_l_02_capsule, center=<5 0 5>, rotation=<rot 0 0 90>}]
 dyn_skirt_bk_b_l_03 [{shapeType=Capsule, length=20.7513, radius=10.5, shapeName=dyn_skirt_bk_b_l_03_capsule, center=<5 0 5>, rotation=<rot 0 0 90>}]
 dyn_skirt_bk_b_r [{shapeType=Capsule, length=6.8156, radius=8, shapeName=dyn_skirt_bk_b_r_capsule, center=<5 0 5>, rotation=<rot 0 0 90>}]
 dyn_skirt_bk_b_r_01 [{shapeType=Capsule, length=23.7453, radius=8, shapeName=dyn_skirt_bk_b_r_01_capsule, center=<5 0 5>, rotation=<rot 0 0 93.309>}]
 dyn_skirt_bk_b_r_02 [{shapeType=Capsule, length=35.696, radius=9, shapeName=dyn_skirt_bk_b_r_02_capsule, center=<5 0 5>, rotation=<rot 0 0 96.2277>}]
 dyn_skirt_bk_b_r_03 [{shapeType=Capsule, length=21.5161, radius=10.5, shapeName=dyn_skirt_bk_b_r_03_capsule, center=<5 0 5>, rotation=<rot 0 0 96.2262>}]
 dyn_skirt_fr_a [{shapeType=Capsule, length=7.5102, radius=6.4898, shapeName=dyn_skirt_fr_a_capsule, center=<5 0 5>, rotation=<rot 90 0 0>}]
 dyn_skirt_fr_a_base [{shapeType=Capsule, length=0.505, radius=0.505, shapeName=dyn_skirt_fr_a_base_capsule, rotation=<rot 0 -90 0>}]
 dyn_skirt_fr_a_base_l [{shapeType=Capsule, length=0.505, radius=0.505, shapeName=dyn_skirt_fr_a_base_l_capsule, rotation=<rot 0 -90 0>}]
 dyn_skirt_fr_a_base_r [{shapeType=Capsule, length=0.505, radius=0.505, shapeName=dyn_skirt_fr_a_base_r_capsule, rotation=<rot 0 -90 0>}]
 dyn_skirt_fr_a_l [{
   shapeType=Capsule
   length=6.2685
   radius=3.8307
   shapeName=dyn_skirt_fr_a_l_capsule
   center=<10.5724 -2.2325 -0.3063>
   rotation=<rot 0 -103.7997 -76.5821>
  }]
 dyn_skirt_fr_a_r [{
   shapeType=Capsule
   length=3.1774
   radius=3.0406
   shapeName=dyn_skirt_fr_a_r_capsule
   center=<-7.6306 2.1218 0.0127>
   rotation=<rot -31.4512 -90 -22.7621>
  }]
 dyn_skirt_fr_b [{shapeType=Capsule, length=6.5806, radius=7.4194, shapeName=dyn_skirt_fr_b_capsule, center=<5 0 5>, rotation=<rot 90 0 0>}]
 dyn_skirt_fr_b1 [{shapeType=Capsule, length=6.5806, radius=7.4194, shapeName=dyn_skirt_fr_b1_capsule, center=<5 0 5>, rotation=<rot 90 0 0>}]
 dyn_skirt_fr_b_l [{shapeType=Capsule, length=12.4174, radius=4.9399, shapeName=dyn_skirt_fr_b_l_capsule, center=<5 0 5>, rotation=<rot 0 0 90>}]
 dyn_skirt_fr_b_l_01 [{shapeType=Capsule, length=12.9296, radius=5.4277, shapeName=dyn_skirt_fr_b_l_01_capsule, center=<5 0 5>, rotation=<rot 0 0 90>}]
 dyn_skirt_fr_b_l_02 [{shapeType=Capsule, length=9.288, radius=6.712, shapeName=dyn_skirt_fr_b_l_02_capsule, center=<5 0 5>, rotation=<rot -0.1486 44.7574 80.8126>}]
 dyn_skirt_fr_b_r [{shapeType=Capsule, length=8.8153, radius=5.1847, shapeName=dyn_skirt_fr_b_r_capsule, center=<5 0 5>, rotation=<rot 0 0 90>}]
 dyn_skirt_fr_b_r_01 [{shapeType=Capsule, length=9.2928, radius=5.7072, shapeName=dyn_skirt_fr_b_r_01_capsule, center=<5 0 5>, rotation=<rot 0 0 90>}]
 dyn_skirt_fr_b_r_02 [{shapeType=Capsule, length=9.2928, radius=6.7072, shapeName=dyn_skirt_fr_b_r_02_capsule, center=<5 0 5>, rotation=<rot 0 -31.5257 83.6114>}]
 dyn_skirt_side_a_base_l [{shapeType=Capsule, length=0.505, radius=0.505, shapeName=dyn_skirt_side_a_base_l_capsule, rotation=<rot 0 -90 0>}]
 dyn_skirt_side_a_base_r [{shapeType=Capsule, length=0.505, radius=0.505, shapeName=dyn_skirt_side_a_base_r_capsule, rotation=<rot 0 -90 0>}]
 dyn_skirt_side_a_l [{
   shapeType=Capsule
   length=5.5508
   radius=2.648
   shapeName=dyn_skirt_side_a_l_capsule
   center=<14.2675 0.5198 -1.8849>
   rotation=<rot 0 -93.5554 -56.1831>
  }]
 dyn_skirt_side_a_r [{shapeType=Capsule, length=0.505, radius=0.505, shapeName=dyn_skirt_side_a_r_capsule, rotation=<rot 0 -90 0>}]
 dyn_skirt_side_b_l [{shapeType=Capsule, length=8, radius=7, shapeName=dyn_skirt_side_b_l_capsule, center=<5 0 5>, rotation=<rot 0 0 90>}]
 dyn_skirt_side_b_l_01 [{shapeType=Capsule, length=24.0331, radius=7, shapeName=dyn_skirt_side_b_l_01_capsule, center=<5 0 5>, rotation=<rot 0 0 90>}]
 dyn_skirt_side_b_l_02 [{shapeType=Capsule, length=24.0331, radius=8, shapeName=dyn_skirt_side_b_l_02_capsule, center=<5 0 5>, rotation=<rot 0 0 100.2518>}]
 dyn_skirt_side_b_l_03 [{shapeType=Capsule, length=22.4753, radius=9, shapeName=dyn_skirt_side_b_l_03_capsule, center=<5 0 5>, rotation=<rot 0 0 99.2045>}]
 dyn_skirt_side_b_r [{shapeType=Capsule, length=8, radius=7, shapeName=dyn_skirt_side_b_r_capsule, center=<5 0 5>, rotation=<rot 0 0 90>}]
 dyn_skirt_side_b_r_01 [{shapeType=Capsule, length=25.1492, radius=7, shapeName=dyn_skirt_side_b_r_01_capsule, center=<5 0 5>, rotation=<rot 0 0 90>}]
 dyn_skirt_side_b_r_02 [{shapeType=Capsule, length=25.1492, radius=8, shapeName=dyn_skirt_side_b_r_02_capsule, center=<5 0 5>, rotation=<rot 0 0 81.611>}]
 dyn_skirt_side_b_r_03 [{shapeType=Capsule, length=24.2708, radius=9, shapeName=dyn_skirt_side_b_r_03_capsule, center=<5 0 5>, rotation=<rot 0 0 75.2167>}]
 foot_l [{shapeType=Capsule, length=19.5237, radius=13.7507, center=<-3.3488 3.5311 0.4711>, rotation=<rot 0.7349 0.9586 -85.8278>}]
 foot_r [{shapeType=Capsule, length=23.5376, radius=13.7675, center=<3.3649 -3.6031 -0.6189>, rotation=<rot 0.7221 1.1909 -85.9009>}]
 pelvis [{shapeType=Capsule, length=7.9063, radius=9.9315, center=<-3.1833 -1.2489 0>}]
 thigh_l [{shapeType=Capsule, length=30.678, radius=9.4022, shapeName=thigh_l_capsule, center=<-24.2637 -0.0343 0.2386>, rotation=<rot 88.923 4.0361 0>}]
 thigh_r [{
   shapeType=Capsule
   length=27.2565
   radius=7.6128
   shapeName=thigh_r_capsule
   center=<24.2637 -0.0343 -0.2386>
   rotation=<rot 88.923 4.0361 0>
  }]
}
constraints [
 {bone1Name=calf_l, bone2Name=thigh_l, swing1LimitDegrees=45, swing2LimitDegrees=45, twistLimitDegrees=45}
 {bone1Name=foot_l, bone2Name=calf_l, swing1LimitDegrees=45, swing2LimitDegrees=45, twistLimitDegrees=45}
 {bone1Name=calf_r, bone2Name=thigh_r, swing1LimitDegrees=45, swing2LimitDegrees=45, twistLimitDegrees=45}
 {bone1Name=foot_r, bone2Name=calf_r, swing1LimitDegrees=45, swing2LimitDegrees=45, twistLimitDegrees=45}
 {
  bone1Name=thigh_l
  bone2Name=pelvis
  swing1LimitDegrees=45
  swing1Motion=Free
  swing2LimitDegrees=45
  swing2Motion=Free
  twistLimitDegrees=45
  twistMotion=Free
 }
 {
  bone1Name=thigh_r
  bone2Name=pelvis
  swing1LimitDegrees=45
  swing1Motion=Free
  swing2LimitDegrees=45
  swing2Motion=Free
  twistLimitDegrees=45
  twistMotion=Free
 }
 {
  bone1Name=dyn_skirt_bk_a_base
  bone2Name=pelvis
  swing1LimitDegrees=45
  swing1Motion=Free
  swing2LimitDegrees=45
  swing2Motion=Free
  twistLimitDegrees=45
  twistMotion=Free
 }
 {bone1Name=dyn_skirt_bk_a, bone2Name=dyn_skirt_bk_a_base, swing1LimitDegrees=35, swing2LimitDegrees=35, twistLimitDegrees=5}
 {bone1Name=dyn_skirt_bk_a_01, bone2Name=dyn_skirt_bk_a, swing1LimitDegrees=35, swing2LimitDegrees=35, twistLimitDegrees=5}
 {bone1Name=dyn_skirt_bk_a_02, bone2Name=dyn_skirt_bk_a_01, swing1LimitDegrees=35, swing2LimitDegrees=35, twistLimitDegrees=5}
 {bone1Name=dyn_skirt_bk_a_03, bone2Name=dyn_skirt_bk_a_02, swing1LimitDegrees=35, swing2LimitDegrees=35, twistLimitDegrees=5}
 {bone1Name=dyn_skirt_bk_a_04, bone2Name=dyn_skirt_bk_a_03, swing1LimitDegrees=35, swing2LimitDegrees=35, twistLimitDegrees=5}
 {
  bone1Name=dyn_skirt_fr_a_base
  bone2Name=pelvis
  swing1LimitDegrees=45
  swing1Motion=Free
  swing2LimitDegrees=45
  swing2Motion=Free
  twistLimitDegrees=45
  twistMotion=Free
 }
 {bone1Name=dyn_skirt_fr_a, bone2Name=dyn_skirt_fr_a_base, swing1LimitDegrees=35, swing2LimitDegrees=35, twistLimitDegrees=5}
 {bone1Name=dyn_skirt_fr_b, bone2Name=dyn_skirt_fr_a, swing1LimitDegrees=35, swing2LimitDegrees=35, twistLimitDegrees=5}
 {bone1Name=dyn_skirt_fr_b1, bone2Name=dyn_skirt_fr_b, swing1LimitDegrees=35, swing2LimitDegrees=35, twistLimitDegrees=5}
 {
  bone1Name=dyn_skirt_fr_a_base_r
  bone2Name=pelvis
  swing1LimitDegrees=45
  swing1Motion=Free
  swing2LimitDegrees=45
  swing2Motion=Free
  twistLimitDegrees=45
  twistMotion=Free
 }
 {
  bone1Name=dyn_skirt_fr_a_r
  bone2Name=dyn_skirt_fr_a_base_r
  swing1LimitDegrees=45
  swing1Motion=Free
  swing2LimitDegrees=45
  swing2Motion=Free
  twistLimitDegrees=45
  twistMotion=Free
 }
 {bone1Name=dyn_skirt_fr_b_r, bone2Name=dyn_skirt_fr_a_r, swing1LimitDegrees=35, swing2LimitDegrees=35, twistLimitDegrees=5}
 {bone1Name=dyn_skirt_fr_b_r_01, bone2Name=dyn_skirt_fr_b_r, swing1LimitDegrees=5, swing2LimitDegrees=35, twistLimitDegrees=5}
 {bone1Name=dyn_skirt_fr_b_r_02, bone2Name=dyn_skirt_fr_b_r_01, swing1LimitDegrees=35, swing2LimitDegrees=35, twistLimitDegrees=5}
 {
  bone1Name=dyn_skirt_bk_a_base_r
  bone2Name=pelvis
  swing1LimitDegrees=45
  swing1Motion=Free
  swing2LimitDegrees=45
  swing2Motion=Free
  twistLimitDegrees=45
  twistMotion=Free
 }
 {bone1Name=dyn_skirt_bk_a_r, bone2Name=dyn_skirt_bk_a_base_r, swing1LimitDegrees=35, swing2LimitDegrees=35, twistLimitDegrees=5}
 {bone1Name=dyn_skirt_bk_b_r, bone2Name=dyn_skirt_bk_a_r, swing1LimitDegrees=35, swing2LimitDegrees=35, twistLimitDegrees=5}
 {bone1Name=dyn_skirt_bk_b_r_01, bone2Name=dyn_skirt_bk_b_r, swing1LimitDegrees=35, swing2LimitDegrees=35, twistLimitDegrees=5}
 {bone1Name=dyn_skirt_bk_b_r_02, bone2Name=dyn_skirt_bk_b_r_01, swing1LimitDegrees=35, swing2LimitDegrees=35, twistLimitDegrees=5}
 {bone1Name=dyn_skirt_bk_b_r_03, bone2Name=dyn_skirt_bk_b_r_02, swing1LimitDegrees=35, swing2LimitDegrees=35, twistLimitDegrees=5}
 {
  bone1Name=dyn_skirt_side_a_base_r
  bone2Name=pelvis
  swing1LimitDegrees=45
  swing1Motion=Free
  swing2LimitDegrees=45
  swing2Motion=Free
  twistLimitDegrees=45
  twistMotion=Free
 }
 {
  bone1Name=dyn_skirt_side_a_r
  bone2Name=dyn_skirt_side_a_base_r
  swing1LimitDegrees=45
  swing1Motion=Free
  swing2LimitDegrees=45
  swing2Motion=Free
  twistLimitDegrees=45
  twistMotion=Free
 }
 {bone1Name=dyn_skirt_side_b_r, bone2Name=dyn_skirt_side_a_r, swing1LimitDegrees=35, swing2LimitDegrees=35, twistLimitDegrees=5}
 {bone1Name=dyn_skirt_side_b_r_01, bone2Name=dyn_skirt_side_b_r, swing1LimitDegrees=35, swing2LimitDegrees=35, twistLimitDegrees=5}
 {bone1Name=dyn_skirt_side_b_r_02, bone2Name=dyn_skirt_side_b_r_01, swing1LimitDegrees=35, swing2LimitDegrees=35, twistLimitDegrees=5}
 {bone1Name=dyn_skirt_side_b_r_03, bone2Name=dyn_skirt_side_b_r_02, swing1LimitDegrees=35, swing2LimitDegrees=35, twistLimitDegrees=5}
 {
  bone1Name=dyn_skirt_fr_a_base_l
  bone2Name=pelvis
  swing1LimitDegrees=45
  swing1Motion=Free
  swing2LimitDegrees=45
  swing2Motion=Free
  twistLimitDegrees=45
  twistMotion=Free
 }
 {
  bone1Name=dyn_skirt_fr_a_l
  bone2Name=dyn_skirt_fr_a_base_l
  swing1LimitDegrees=45
  swing1Motion=Free
  swing2LimitDegrees=45
  swing2Motion=Free
  twistLimitDegrees=45
  twistMotion=Free
 }
 {bone1Name=dyn_skirt_fr_b_l, bone2Name=dyn_skirt_fr_a_l, swing1LimitDegrees=35, swing2LimitDegrees=35, twistLimitDegrees=5}
 {bone1Name=dyn_skirt_fr_b_l_01, bone2Name=dyn_skirt_fr_b_l, swing1LimitDegrees=5, swing2LimitDegrees=35, twistLimitDegrees=5}
 {bone1Name=dyn_skirt_fr_b_l_02, bone2Name=dyn_skirt_fr_b_l_01, swing1LimitDegrees=35, swing2LimitDegrees=35, twistLimitDegrees=5}
 {
  bone1Name=dyn_skirt_bk_a_base_l
  bone2Name=pelvis
  swing1LimitDegrees=45
  swing1Motion=Free
  swing2LimitDegrees=45
  swing2Motion=Free
  twistLimitDegrees=45
  twistMotion=Free
 }
 {bone1Name=dyn_skirt_bk_a_l, bone2Name=dyn_skirt_bk_a_base_l, swing1LimitDegrees=35, swing2LimitDegrees=35, twistLimitDegrees=5}
 {bone1Name=dyn_skirt_bk_b_l, bone2Name=dyn_skirt_bk_a_l, swing1LimitDegrees=35, swing2LimitDegrees=35, twistLimitDegrees=5}
 {bone1Name=dyn_skirt_bk_b_l_01, bone2Name=dyn_skirt_bk_b_l, swing1LimitDegrees=35, swing2LimitDegrees=35, twistLimitDegrees=5}
 {bone1Name=dyn_skirt_bk_b_l_02, bone2Name=dyn_skirt_bk_b_l_01, swing1LimitDegrees=35, swing2LimitDegrees=35, twistLimitDegrees=5}
 {bone1Name=dyn_skirt_bk_b_l_03, bone2Name=dyn_skirt_bk_b_l_02, swing1LimitDegrees=35, swing2LimitDegrees=35, twistLimitDegrees=5}
 {
  bone1Name=dyn_skirt_side_a_base_l
  bone2Name=pelvis
  swing1LimitDegrees=45
  swing1Motion=Free
  swing2LimitDegrees=45
  swing2Motion=Free
  twistLimitDegrees=45
  twistMotion=Free
 }
 {
  bone1Name=dyn_skirt_side_a_l
  bone2Name=dyn_skirt_side_a_base_l
  swing1LimitDegrees=45
  swing1Motion=Free
  swing2LimitDegrees=45
  swing2Motion=Free
  twistLimitDegrees=45
  twistMotion=Free
 }
 {bone1Name=dyn_skirt_side_b_l, bone2Name=dyn_skirt_side_a_l, swing1LimitDegrees=35, swing2LimitDegrees=35, twistLimitDegrees=5}
 {bone1Name=dyn_skirt_side_b_l_01, bone2Name=dyn_skirt_side_b_l, swing1LimitDegrees=35, swing2LimitDegrees=35, twistLimitDegrees=5}
 {bone1Name=dyn_skirt_side_b_l_02, bone2Name=dyn_skirt_side_b_l_01, swing1LimitDegrees=35, swing2LimitDegrees=35, twistLimitDegrees=5}
 {bone1Name=dyn_skirt_side_b_l_03, bone2Name=dyn_skirt_side_b_l_02, swing1LimitDegrees=35, swing2LimitDegrees=35, twistLimitDegrees=5}
 {
  bone1Name=dyn_skirt_bk_a_03
  bone2Name=dyn_skirt_bk_b_r_02
  swing1LimitDegrees=45
  swing1Motion=Free
  swing2LimitDegrees=45
  swing2Motion=Free
  twistLimitDegrees=45
  twistMotion=Free
 }
 {
  bone1Name=dyn_skirt_bk_a_03
  bone2Name=dyn_skirt_bk_b_l_02
  swing1LimitDegrees=45
  swing1Motion=Free
  swing2LimitDegrees=45
  swing2Motion=Free
  twistLimitDegrees=45
  twistMotion=Free
 }
 {
  bone1Name=dyn_skirt_bk_a_01
  bone2Name=dyn_skirt_bk_b_r
  swing1LimitDegrees=45
  swing1Motion=Free
  swing2LimitDegrees=45
  swing2Motion=Free
  twistLimitDegrees=45
  twistMotion=Free
 }
 {
  bone1Name=dyn_skirt_bk_a_01
  bone2Name=dyn_skirt_bk_b_l
  swing1LimitDegrees=45
  swing1Motion=Free
  swing2LimitDegrees=45
  swing2Motion=Free
  twistLimitDegrees=45
  twistMotion=Free
 }
 {
  bone1Name=dyn_skirt_fr_b_l_01
  bone2Name=dyn_skirt_side_b_l_01
  swing1LimitDegrees=45
  swing1Motion=Free
  swing2LimitDegrees=45
  swing2Motion=Free
  twistLimitDegrees=45
  twistMotion=Free
 }
]
