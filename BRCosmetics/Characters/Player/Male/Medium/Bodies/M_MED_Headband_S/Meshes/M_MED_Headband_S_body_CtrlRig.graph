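# Control Rig for the M_MED_Headband_S male body that procedurally adjusts its collar bones to follow head movement.
# It reads head rotation relative to spine_05 as Euler angles and uses clamped remaps to drive rotation and translation offsets on the dyn_collar master bones.
# Sequence A applies the full-weight correctives, while B layers on extra offsets faded by the smoothed `targeting` variable, with the left-collar offset also scaled by head angle.
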ControlRigBlueprint /BRCosmetics/Characters/Player/Male/Medium/Bodies/M_MED_Headband_S/Meshes/M_MED_Headband_S_body_CtrlRig
graphs [
 RigVMFunctionLibrary {links [], nodes []}
 RigVMModel {nodes [
   RigUnit_BeginExecution = &RigUnit_BeginExecution()
   RigVMFunction_Sequence = &RigVMFunction_Sequence(ExecuteContext=@RigUnit_BeginExecution.ExecutePin)
   n1 = &RigVMFunction_MathQuaternionToEuler(
    RotationOrder=ZYX
    Value=&RigUnit_GetRelativeTransformForItem(Child=(Type=Bone,Name="head"), Parent=(Type=Bone,Name="spine_05")).RelativeTransform.Rotation
   )
   n2 = &RigVMFunction_MathFloatRemap(SourceMaximum=-55, TargetMaximum=2, bClamp=true, Value=@n1.Result.Z)
   n3 = &RigVMFunction_MathQuaternionFromEuler(
    Euler=(X=0,Y=0,Z=0)
    Euler.Y=&RigVMFunction_MathFloatRemap(SourceMaximum=-80, TargetMaximum=-90, bClamp=true, Value=@n1.Result.Z).Result
   )
   &RigUnit_ModifyTransforms(
    ItemToModify="((Item=(Type=Bone,Name=\"dyn_collar_BK_MSTR\"),Transform=(Rotation=(X=0,Y=0,Z=0,W=1),Translation=(X=0,Y=0,Z=0),Scale3D=(X=1,Y=1,Z=1))),(Item=(Type=Bone,Name=\"dyn_collar_BK_r_MSTR\"),Transform=(Rotation=(X=0,Y=0,Z=0,W=1),Translation=(X=0,Y=0,Z=0),Scale3D=(X=1, ...(truncated)"
    Weight=1
    ItemToModify.1.Transform.Translation.Z=@n2.Result
    ItemToModify.2.Transform.Translation.Z=@n2.Result
    ItemToModify.0.Transform.Rotation=@n3.Result
    ItemToModify.1.Transform.Rotation=@n3.Result
    ItemToModify.2.Transform.Rotation=&RigVMFunction_MathQuaternionFromEuler(
     Euler=(X=0,Y=0,Z=0)
     Euler.Z=&RigVMFunction_MathFloatRemap(SourceMaximum=-80, TargetMaximum=-60, bClamp=true, Value=@n1.Result.Z).Result
    ).Result
    ExecutePin=@RigVMFunction_Sequence.A
   )
   n4 = &RigVMFunction_MathQuaternionToEuler(
    RotationOrder=ZYX
    Value=&RigUnit_GetRelativeTransformForItem(Child=(Type=Bone,Name="head"), Parent=(Type=Bone,Name="spine_05")).RelativeTransform.Rotation
   )
   > &RigUnit_ModifyTransforms(
    ItemToModify="((Item=(Type=Bone,Name=\"dyn_collar_l_MSTR\"),Transform=(Rotation=(X=0,Y=0,Z=0,W=1),Translation=(X=0,Y=0,Z=0),Scale3D=(X=1,Y=1,Z=1))),(Item=(Type=Bone,Name=\"None\"),Transform=(Rotation=(X=0,Y=0,Z=0,W=1),Translation=(X=0,Y=0,Z=0),Scale3D=(X=1,Y=1,Z=1))),(Item= ...(truncated)"
    Weight=1
    ItemToModify.0.Transform.Rotation=&RigVMFunction_MathQuaternionFromEuler(
     Euler=(X=0,Y=0,Z=0)
     Euler.Y=&RigVMFunction_MathFloatRemap(SourceMaximum=-45, SourceMinimum=-20, TargetMaximum=-60, bClamp=true, Value=@n4.Result.X).Result
    ).Result
    ItemToModify.2.Transform.Rotation=&RigVMFunction_MathQuaternionFromEuler(
     Euler=(X=0,Y=0,Z=0)
     Euler.Y=&RigVMFunction_MathFloatRemap(SourceMaximum=25, SourceMinimum=0, TargetMaximum=-60, bClamp=true, Value=@n4.Result.X).Result
    ).Result
    ItemToModify.3.Transform.Rotation=&RigVMFunction_MathQuaternionFromEuler(
     Euler=(X=0,Y=0,Z=0)
     Euler.Z=&RigVMFunction_MathFloatRemap(SourceMaximum=25, SourceMinimum=0, TargetMaximum=-15, bClamp=true, Value=@n4.Result.X).Result
    ).Result
   )
   n5 = RigVMVariableNode() {variable=targeting}
   AlphaInterp = &RigVMFunction_AlphaInterp(InterpSpeedDecreasing=10, InterpSpeedIncreasing=10, Scale=1, bClampResult=true, bMapRange=true, Value=@n5.Value)
   n6 = &RigVMFunction_MathQuaternionToEuler(
    RotationOrder=ZYX
    Value=&RigUnit_GetRelativeTransformForItem(Child=(Type=Bone,Name="head"), Parent=(Type=Bone,Name="spine_05")).RelativeTransform.Rotation
   )
   n7 = &RigVMFunction_MathFloatRemap(SourceMaximum=-60, TargetMaximum=7, bClamp=true, Value=@n6.Result.Z)
   &RigUnit_ModifyTransforms(
    ItemToModify="((Item=(Type=Bone,Name=\"dyn_collar_BK_r_MSTR\"),Transform=(Rotation=(X=0,Y=0,Z=0,W=1),Translation=(X=0,Y=9,Z=0),Scale3D=(X=1,Y=1,Z=1))),(Item=(Type=Bone,Name=\"None\"),Transform=(Rotation=(X=0,Y=0,Z=0,W=1),Translation=(X=0,Y=0,Z=0),Scale3D=(X=1,Y=1,Z=1))),(It ...(truncated)"
    Weight=@AlphaInterp.Result
    ItemToModify.0.Transform.Translation.Y=@n7.Result
    ItemToModify.2.Transform.Translation.Y=@n7.Result
    ItemToModify.2.Transform.Translation.Z=&RigVMFunction_MathFloatRemap(SourceMaximum=-55, TargetMaximum=3, bClamp=true, Value=@n6.Result.Z).Result
    ItemToModify.3.Transform.Translation.Z=&RigVMFunction_MathFloatRemap(SourceMaximum=-55, TargetMaximum=7, bClamp=true, Value=@n6.Result.Z).Result
    ItemToModify.3.Transform.Rotation=&RigVMFunction_MathQuaternionFromEuler(
     Euler=(X=0,Y=0,Z=0)
     Euler.Y=&RigVMFunction_MathFloatRemap(SourceMaximum=-55, TargetMaximum=-30, bClamp=true, Value=@n6.Result.Z).Result
    ).Result
    ItemToModify.0.Transform.Rotation=&RigVMFunction_MathQuaternionFromEuler(
     Euler=(X=0,Y=0,Z=0)
     Euler.Y=&RigVMFunction_MathFloatRemap(SourceMaximum=-55, TargetMaximum=-5, bClamp=true, Value=@n6.Result.Z).Result
    ).Result
    ExecutePin=@RigVMFunction_Sequence.B
   )
   n8 = &RigVMFunction_MathQuaternionToEuler(
    RotationOrder=ZYX
    Value=&RigUnit_GetRelativeTransformForItem(Child=(Type=Bone,Name="head"), Parent=(Type=Bone,Name="spine_05")).RelativeTransform.Rotation
   )
   n9 = &RigVMFunction_AlphaInterp(
    InterpSpeedDecreasing=10
    InterpSpeedIncreasing=10
    bClampResult=true
    bMapRange=true
    Scale=&RigVMFunction_MathFloatRemap(SourceMaximum=20, bClamp=true, Value=@n8.Result.Z).Result
    Value=@n5.Value
   )
   > &RigUnit_ModifyTransforms(
    ItemToModify=((Item=(Type=Bone,Name="dyn_collar_l_MSTR"),Transform=(Rotation=(X=0,Y=0,Z=0,W=1),Translation=(X=0,Y=0,Z=5),Scale3D=(X=1,Y=1,Z=1))))
    Weight=@n9.Result
    ItemToModify.0.Transform.Rotation=&RigVMFunction_MathQuaternionFromEuler(Euler=(X=0,Y=-7,Z=-5)).Result
   )
   &RigVMFunction_MathFloatRemap(SourceMaximum=25, TargetMaximum=-15, bClamp=true, Value=@n8.Result.Z)
  ]}
]
hierarchy [
 root {loc=<0 0 0>, rot=<0 0 0 1>}
 pelvis {parent=root, loc=<0 -0.7071 93.1191>, rot=<0.0127 0.707 -0.0127 -0.707>}
 spine_01 {parent=pelvis, loc=<6.7494 -0.6388 0>, rot=<0 0 0.034 -0.9994>}
 spine_02 {parent=spine_01, loc=<7.2518 -0.035 0>, rot=<0 0 0.0238 -0.9997>}
 spine_03 {parent=spine_02, loc=<6.0104 0 0>, rot=<0 0 0.0149 -0.9999>}
 spine_04 {parent=spine_03, loc=<8.1024 0 0>, rot=<0 0 -0.0132 -0.9999>}
 spine_05 {parent=spine_04, loc=<10.0146 0 0>, rot=<0 0 -0.0015 -1>}
 clavicle_l {parent=spine_05, loc=<4.3575 -0.1466 -2.3223>, rot=<0.0465 -0.7289 -0.009 -0.6829>}
 upperarm_l {parent=clavicle_l, loc=<15.5111 0 0>, rot=<0.0226 -0.3234 -0.0088 -0.9459>}
 lowerarm_l {parent=upperarm_l, loc=<26.6514 0 0>, rot=<0.0391 -0.0099 -0.2453 -0.9686>}
 hand_l {parent=lowerarm_l, loc=<22.3491 0 0>, rot=<-0.7081 -0.0507 0.0975 -0.6975>}
 index_metacarpal_l {parent=hand_l, loc=<1.6535 0.0183 -2.96>, rot=<0.0158 -0.0124 -0.0173 -0.9996>}
 index_01_l {parent=index_metacarpal_l, loc=<8.3008 0.0357 -0.0524>, rot=<-0.0044 0.0485 0.0924 -0.9945>}
 index_02_l {parent=index_01_l, loc=<3.3587 -0.3004 0.0452>, rot=<-0.002 0.0027 0.3225 -0.9466>}
 index_03_l {parent=index_02_l, loc=<2.5578 -0.123 -0.0252>, rot=<-0.002 0.0067 0.2558 -0.9667>}
 middle_metacarpal_l {parent=hand_l, loc=<2.1856 -0.1725 -0.5984>, rot=<0.006 0.0168 -0.0329 -0.9993>}
 middle_01_l {parent=middle_metacarpal_l, loc=<7.3275 -0.2447 0.0129>, rot=<-0.0208 0.0042 0.0858 -0.9961>}
 middle_02_l {parent=middle_01_l, loc=<4.0123 0.1401 0.1175>, rot=<-0.0218 0.0258 0.3512 -0.9357>}
 middle_03_l {parent=middle_02_l, loc=<2.6077 0.0534 0.1373>, rot=<0.0082 0.0188 0.1708 -0.9851>}
 pinky_metacarpal_l {parent=hand_l, loc=<1.7616 -0.9753 3.3814>, rot=<-0.0397 0.0814 -0.069 -0.9935>}
 pinky_01_l {parent=pinky_metacarpal_l, loc=<7.1709 -0.4831 -0.1094>, rot=<-0.0654 -0.0035 0.2242 -0.9723>}
 pinky_02_l {parent=pinky_01_l, loc=<2.8274 0.0893 -0.1043>, rot=<0.0037 0.0011 0.2854 -0.9584>}
 pinky_03_l {parent=pinky_02_l, loc=<2.2046 0.0307 -0.0078>, rot=<-0.0002 0.0098 0.1445 -0.9895>}
 ring_metacarpal_l {parent=hand_l, loc=<2.2074 -0.3625 1.3638>, rot=<-0.0284 0.0722 -0.0458 -0.9959>}
 ring_01_l {parent=ring_metacarpal_l, loc=<7.0295 -0.6197 -0.1446>, rot=<-0.0489 -0.0252 0.1892 -0.9804>}
 ring_02_l {parent=ring_01_l, loc=<3.6174 0.4647 -0.1027>, rot=<0 0 0.2806 -0.9598>}
 ring_03_l {parent=ring_02_l, loc=<2.4152 0.114 0.0199>, rot=<0 0 0.1189 -0.9929>}
 thumb_01_l {parent=hand_l, loc=<1.3118 -0.3191 -2.459>, rot=<0.6546 -0.4042 0.0664 -0.6354>}
 thumb_02_l {parent=thumb_01_l, loc=<5.9336 -0.0448 0.0401>, rot=<-0.0037 0.0213 0.298 -0.9543>}
 thumb_03_l {parent=thumb_02_l, loc=<2.2223 0.3682 -0.0907>, rot=<0.0118 -0.013 0.0066 -0.9998>}
 lowerarm_twist_01_l {parent=lowerarm_l, loc=<17.8549 0 0>, rot=<0 0 0 -1>}
 lowerarm_twist_02_l {parent=lowerarm_l, loc=<9.6199 0 0>, rot=<0 0 0 -1>}
 upperarm_twist_01_l {parent=upperarm_l, loc=<0.0041 0 0>, rot=<0 0 0 -1>}
 upperarm_twist_02_l {parent=upperarm_l, loc=<13.7941 0 0>, rot=<0 0 0 -1>}
 clavicle_r {parent=spine_05, loc=<4.3575 -0.1466 2.3223>, rot=<0.7289 0.0465 -0.6829 0.009>}
 upperarm_r {parent=clavicle_r, loc=<-15.5111 0 0>, rot=<0.0226 -0.3234 -0.0088 -0.9459>}
 lowerarm_r {parent=upperarm_r, loc=<-26.6515 0 0>, rot=<0.0391 -0.0099 -0.2453 -0.9686>}
 hand_r {parent=lowerarm_r, loc=<-22.3492 0 0>, rot=<-0.7081 -0.0507 0.0975 -0.6975>}
 index_metacarpal_r {parent=hand_r, loc=<-1.6534 -0.0181 2.9601>, rot=<0.0158 -0.0124 -0.0173 -0.9996>}
 index_01_r {parent=index_metacarpal_r, loc=<-8.3008 -0.0357 0.0524>, rot=<-0.0044 0.0485 0.0924 -0.9945>}
 index_02_r {parent=index_01_r, loc=<-3.3589 0.3003 -0.0453>, rot=<-0.002 0.0027 0.3225 -0.9466>}
 index_03_r {parent=index_02_r, loc=<-2.5578 0.1231 0.0251>, rot=<-0.002 0.0067 0.2558 -0.9667>}
 middle_metacarpal_r {parent=hand_r, loc=<-2.1856 0.1726 0.5984>, rot=<0.006 0.0168 -0.0329 -0.9993>}
 middle_01_r {parent=middle_metacarpal_r, loc=<-7.3273 0.2447 -0.0129>, rot=<-0.0208 0.0042 0.0858 -0.9961>}
 middle_02_r {parent=middle_01_r, loc=<-4.0142 -0.1411 -0.07>, rot=<-0.0217 0.0259 0.3512 -0.9357>}
 middle_03_r {parent=middle_02_r, loc=<-2.6074 -0.0534 -0.1312>, rot=<0.0083 0.0188 0.1708 -0.9851>}
 pinky_metacarpal_r {parent=hand_r, loc=<-1.7617 0.9753 -3.3814>, rot=<-0.0397 0.0814 -0.069 -0.9935>}
 pinky_01_r {parent=pinky_metacarpal_r, loc=<-7.1707 0.4832 0.1094>, rot=<-0.0654 -0.0035 0.2242 -0.9723>}
 pinky_02_r {parent=pinky_01_r, loc=<-2.8274 -0.0892 0.1043>, rot=<0.0037 0.0011 0.2854 -0.9584>}
 pinky_03_r {parent=pinky_02_r, loc=<-2.2045 -0.031 0.0078>, rot=<-0.0002 0.0098 0.1445 -0.9895>}
 ring_metacarpal_r {parent=hand_r, loc=<-2.2073 0.3627 -1.3638>, rot=<-0.0284 0.0722 -0.0458 -0.9959>}
 ring_01_r {parent=ring_metacarpal_r, loc=<-7.0297 0.6197 0.1446>, rot=<-0.0489 -0.0252 0.1892 -0.9804>}
 ring_02_r {parent=ring_01_r, loc=<-3.6171 -0.4648 0.1027>, rot=<0 0 0.2806 -0.9598>}
 ring_03_r {parent=ring_02_r, loc=<-2.4153 -0.1141 -0.0199>, rot=<0 0 0.1189 -0.9929>}
 thumb_01_r {parent=hand_r, loc=<-1.3119 0.3192 2.459>, rot=<0.6546 -0.4042 0.0664 -0.6354>}
 thumb_02_r {parent=thumb_01_r, loc=<-5.9336 0.0447 -0.0401>, rot=<-0.0037 0.0213 0.298 -0.9543>}
 thumb_03_r {parent=thumb_02_r, loc=<-2.2223 -0.3682 0.0907>, rot=<0.0118 -0.013 0.0066 -0.9998>}
 lowerarm_twist_01_r {parent=lowerarm_r, loc=<-17.855 0 0>, rot=<0 0 0 1>}
 lowerarm_twist_02_r {parent=lowerarm_r, loc=<-9.6199 0 0>, rot=<0 0 0 1>}
 upperarm_twist_01_r {parent=upperarm_r, loc=<-0.0041 0 0>, rot=<0 0 0 1>}
 upperarm_twist_02_r {parent=upperarm_r, loc=<-13.7942 0 0>, rot=<0 0 0 1>}
 neck_01 {parent=spine_05, loc=<11.3138 0.1529 0>, rot=<0 0 -0.2003 -0.9797>}
 neck_02 {parent=neck_01, loc=<4.5627 0 0>, rot=<0 0 0.0504 -0.9987>}
 head {parent=neck_02, loc=<5.3914 0 0>, rot=<0 0 0.115 -0.9934>}
 dyn_collar_bk_MSTR {parent=spine_05, loc=<10.3541 -5.2931 0.0001>, rot=<-0.7065 -0.0284 -0.0284 -0.7065>}
 dyn_collar_bk_01 {parent=dyn_collar_bk_MSTR, loc=<0 0 0>, rot=<0 0 0 1>}
 dyn_collar_bk_r_MSTR {parent=spine_05, loc=<11.1295 -3.3384 6.7592>, rot=<-0.3824 -0.0153 -0.037 -0.9231>}
 dyn_collar_bk_r_01 {parent=dyn_collar_bk_r_MSTR, loc=<0 0 0>, rot=<0 0 0 1>}
 dyn_collar_bk_l_MSTR {parent=spine_05, loc=<10.9982 -3.0954 -6.3742>, rot=<-0.3824 -0.0153 -0.037 -0.9231>}
 dyn_collar_bk_l_01 {parent=dyn_collar_bk_l_MSTR, loc=<0 0 0>, rot=<0 0 0 1>}
 dyn_collar_l_MSTR {parent=spine_05, loc=<8.6542 4.0687 -6.3223>, rot=<0.8758 0.1503 -0.2922 -0.3535>}
 dyn_collar_l_01 {parent=dyn_collar_l_MSTR, loc=<0 0 0>, rot=<0 0 0 -1>}
 dyn_collar_r_MSTR {parent=spine_05, loc=<8.5602 4.2272 6.6178>, rot=<0.3535 0.2922 -0.1503 -0.8758>}
 dyn_collar_r_01 {parent=dyn_collar_r_MSTR, loc=<0 0 0>, rot=<0 0 0 -1>}
 thigh_l {parent=pelvis, loc=<-2.5693 -0.0109 -7.7501>, rot=<0.0012 0.0572 0.0155 -0.9982>}
 calf_l {parent=thigh_l, loc=<-39.0484 0 0>, rot=<0.0006 -0.0161 -0.0391 -0.9991>}
 calf_twist_01_l {parent=calf_l, loc=<-35.813 -0.0967 0.0397>, rot=<0.0237 0.0095 -0.0035 -0.9997>}
 calf_twist_02_l {parent=calf_l, loc=<-18.8064 -0.0967 0.0397>, rot=<0.0237 0.0095 -0.0035 -0.9997>}
 foot_l {parent=calf_l, loc=<-42.561 0.1077 0.4292>, rot=<0.0418 -0.0455 0.0343 -0.9975>}
 ball_l {parent=foot_l, loc=<-6.1599 13.8264 0.6806>, rot=<0.0001 -0.0001 -0.7063 -0.7079>}
 thigh_twist_01_l {parent=thigh_l, loc=<-21.6303 0 0>, rot=<0.0013 0 0 -1>}
 thigh_r {parent=pelvis, loc=<-2.5691 -0.0109 7.7501>, rot=<0.0572 -0.0012 0.9982 0.0155>}
 calf_r {parent=thigh_r, loc=<39.0487 0 0>, rot=<0.0006 -0.0161 -0.0391 -0.9991>}
 calf_twist_01_r {parent=calf_r, loc=<35.8132 0.0967 -0.0397>, rot=<0.0237 0.0095 -0.0035 -0.9997>}
 calf_twist_02_r {parent=calf_r, loc=<18.8064 0.0967 -0.0397>, rot=<0.0237 0.0095 -0.0035 -0.9997>}
 foot_r {parent=calf_r, loc=<42.5612 -0.1077 -0.4292>, rot=<0.0418 -0.0455 0.0343 -0.9975>}
 ball_r {parent=foot_r, loc=<6.1598 -13.8264 -0.6806>, rot=<0.0001 -0.0001 -0.7063 -0.7079>}
 thigh_twist_01_r {parent=thigh_r, loc=<21.6303 0 0>, rot=<0.0013 0 0 -1>}
 dyn_skirt_bk_a_base {parent=pelvis, loc=<5.2574 -10.9869 -0.0004>, rot=<0.0926 -0.702 0.7009 -0.0852>}
 dyn_skirt_bk_a {parent=dyn_skirt_bk_a_base, loc=<0 0 0>, rot=<0 0 0 -1>}
 dyn_skirt_bk_a_01 {parent=dyn_skirt_bk_a, loc=<9.9932 0 0>, rot=<-0.0001 -0.1238 0 -0.9923>}
 dyn_skirt_bk_a_02 {parent=dyn_skirt_bk_a_01, loc=<11.4868 0 0>, rot=<0.0005 -0.0432 0 -0.9991>}
 dyn_skirt_bk_a_03 {parent=dyn_skirt_bk_a_02, loc=<13.3097 0 0>, rot=<-0.0002 0.0157 0 -0.9999>}
 dyn_skirt_bk_a_04 {parent=dyn_skirt_bk_a_03, loc=<12.6743 0 0>, rot=<-0.0002 0.0184 0 -0.9998>}
 dyn_skirt_bk_a_05 {parent=dyn_skirt_bk_a_04, loc=<9.4426 0 0>, rot=<0.9911 0 -0.1328 0>}
 dyn_skirt_fr_a_base {parent=pelvis, loc=<5.4731 9.2474 0.001>, rot=<0.0488 0.7046 0.7052 0.0613>}
 dyn_skirt_fr_a {parent=dyn_skirt_fr_a_base, loc=<0 0 0>, rot=<0 0 0 -1>}
 dyn_skirt_fr_b {parent=dyn_skirt_fr_a, loc=<9.9932 0 0>, rot=<0 0 0 -1>}
 dyn_skirt_fr_b1 {parent=dyn_skirt_fr_b, loc=<9.7353 0 0>, rot=<0 0 0 -1>}
 dyn_skirt_fr_a_base_r {parent=pelvis, loc=<5.3826 6.8847 8.69>, rot=<-0.847 0.0267 0.1204 -0.5171>}
 dyn_skirt_fr_a_r {parent=dyn_skirt_fr_a_base_r, loc=<0 0 0>, rot=<0 0 0 -1>}
 dyn_skirt_fr_b_r {parent=dyn_skirt_fr_a_r, loc=<-9.9936 0 0.0001>, rot=<-0.0043 -0.9954 -0.0546 0.0784>}
 dyn_skirt_fr_b_r_01 {parent=dyn_skirt_fr_b_r, loc=<9.2843 0 0.0999>, rot=<-0.0001 -0.0006 -0.0001 -1>}
 dyn_skirt_fr_b_r_02 {parent=dyn_skirt_fr_b_r_01, loc=<9.1415 0 0>, rot=<0.0023 0.0218 0.0024 -0.9998>}
 dyn_skirt_bk_a_base_r {parent=pelvis, loc=<5.2215 -8.0801 9.629>, rot=<0.8878 0.0271 -0.1273 -0.4414>}
 dyn_skirt_bk_a_r {parent=dyn_skirt_bk_a_base_r, loc=<0 0 0>, rot=<0 0 0 1>}
 dyn_skirt_bk_b_r {parent=dyn_skirt_bk_a_r, loc=<-9.9933 0 0>, rot=<-0.0039 -0.995 0.0251 0.0969>}
 dyn_skirt_bk_b_r_01 {parent=dyn_skirt_bk_b_r, loc=<11.6274 0 0>, rot=<0.0019 -0.0335 0.0082 -0.9994>}
 dyn_skirt_bk_b_r_02 {parent=dyn_skirt_bk_b_r_01, loc=<12.3315 0 0>, rot=<-0.001 0.0209 -0.002 -0.9998>}
 dyn_skirt_bk_b_r_03 {parent=dyn_skirt_bk_b_r_02, loc=<11.4639 0 0>, rot=<-0.0013 0.0199 -0.0003 -0.9998>}
 dyn_skirt_side_a_base_r {parent=pelvis, loc=<5.6072 -0.9276 13.3104>, rot=<0.9882 -0.0177 -0.1524 -0.0027>}
 dyn_skirt_side_a_r {parent=dyn_skirt_side_a_base_r, loc=<0 0 0>, rot=<0 0 0 1>}
 dyn_skirt_side_b_r {parent=dyn_skirt_side_a_r, loc=<-9.9934 0 0.0001>, rot=<0 -0.9883 0 0.1525>}
 dyn_skirt_side_b_r_01 {parent=dyn_skirt_side_b_r, loc=<9.2352 0 1.6875>, rot=<0 0 0 1>}
 dyn_skirt_side_b_r_02 {parent=dyn_skirt_side_b_r_01, loc=<10.2699 0 1.2795>, rot=<0 0 0 1>}
 dyn_skirt_side_b_r_03 {parent=dyn_skirt_side_b_r_02, loc=<10.5986 0 0.5213>, rot=<0 0 0 1>}
 dyn_skirt_fr_a_base_l {parent=pelvis, loc=<5.3822 6.8846 -8.6902>, rot=<-0.024 -0.8447 -0.5161 -0.1394>}
 dyn_skirt_fr_a_l {parent=dyn_skirt_fr_a_base_l, loc=<0 0 0>, rot=<0 0 0 1>}
 dyn_skirt_fr_b_l {parent=dyn_skirt_fr_a_l, loc=<9.4327 0 0.074>, rot=<-0.0186 -0.0656 0.0012 -0.9977>}
 dyn_skirt_fr_b_l_01 {parent=dyn_skirt_fr_b_l, loc=<9.542 -0.0014 -0.038>, rot=<0.0048 -0.0401 0.0013 -0.9992>}
 dyn_skirt_fr_b_l_02 {parent=dyn_skirt_fr_b_l_01, loc=<9.5322 0 0>, rot=<0.0138 0.0173 -0.0002 -0.9998>}
 dyn_skirt_bk_a_base_l {parent=pelvis, loc=<5.2214 -8.0797 -9.6287>, rot=<0.0359 -0.8891 0.4402 -0.1199>}
 dyn_skirt_bk_a_l {parent=dyn_skirt_bk_a_base_l, loc=<0 0 0>, rot=<0 0 0 -1>}
 dyn_skirt_bk_b_l {parent=dyn_skirt_bk_a_l, loc=<9.9597 0.0157 0.0318>, rot=<0.0385 -0.0937 -0.0086 -0.9948>}
 dyn_skirt_bk_b_l_01 {parent=dyn_skirt_bk_b_l, loc=<11.6274 0 0>, rot=<-0.0019 -0.0333 -0.0092 -0.9994>}
 dyn_skirt_bk_b_l_02 {parent=dyn_skirt_bk_b_l_01, loc=<12.3315 0 0>, rot=<0.001 0.0209 0.0026 -0.9998>}
 dyn_skirt_bk_b_l_03 {parent=dyn_skirt_bk_b_l_02, loc=<11.464 0 0>, rot=<-0.0023 -0.0555 -0.0156 -0.9983>}
 dyn_skirt_side_a_base_l {parent=pelvis, loc=<5.6072 -0.9276 -13.1>, rot=<-0.0176 -0.9831 0.0033 -0.1822>}
 dyn_skirt_side_a_l {parent=dyn_skirt_side_a_base_l, loc=<0 0 0>, rot=<0 0 0 1>}
 dyn_skirt_side_b_l {parent=dyn_skirt_side_a_l, loc=<9.9932 0 0>, rot=<0 -0.1822 0 -0.9833>}
 dyn_skirt_side_b_l_01 {parent=dyn_skirt_side_b_l, loc=<9.2352 0 0.9822>, rot=<0 0 0 -1>}
 dyn_skirt_side_b_l_02 {parent=dyn_skirt_side_b_l_01, loc=<10.2699 0 1.2802>, rot=<0 0 0 -1>}
 dyn_skirt_side_b_l_03 {parent=dyn_skirt_side_b_l_02, loc=<10.5986 0 0.8448>, rot=<0 0 0 -1>}
 dyn_knot_MSTR {parent=pelvis, loc=<8.1456 10.5026 3.1827>, rot=<0.586 0.3957 0.5715 0.4164>}
 dyn_knot {parent=dyn_knot_MSTR, loc=<0 0 0>, rot=<0 0 0 1>}
 dyn_rope_a_MSTR {parent=dyn_knot, loc=<4.0348 0.0705 3.171>, rot=<0.7347 0.639 -0.199 0.1108>}
 dyn_rope_a_01 {parent=dyn_rope_a_MSTR, loc=<9.9666 0 0>, rot=<-0.063 0.3739 -0.038 -0.9245>}
 dyn_rope_a_02 {parent=dyn_rope_a_01, loc=<10.28 0 0>, rot=<-0.0055 0.0643 -0.0405 -0.9971>}
 dyn_rope_a_03 {parent=dyn_rope_a_02, loc=<12.7998 0.0815 0.221>, rot=<-0.0013 0.0678 -0.0251 -0.9974>}
 dyn_rope_b_MSTR {parent=dyn_knot, loc=<3.8528 0.6301 1.2843>, rot=<0.7652 0.5302 -0.2976 0.2114>}
 dyn_rope_b_01 {parent=dyn_rope_b_MSTR, loc=<7.1622 0 0>, rot=<0.001 0.3184 0.0082 -0.9479>}
 dyn_rope_b_02 {parent=dyn_rope_b_01, loc=<9.3445 0 0>, rot=<0.0024 0.062 -0.0436 -0.9971>}
 dyn_rope_b_03 {parent=dyn_rope_b_02, loc=<10.1036 0 0>, rot=<0.0004 0.0086 0.039 -0.9992>}
 dyn_loop_a_MSTR {parent=dyn_knot, loc=<5.1123 -0.9694 2.1777>, rot=<-0.3308 -0.2033 0.7209 -0.5741>}
 dyn_loop_a_01 {parent=dyn_loop_a_MSTR, loc=<9.5213 0 0>, rot=<-0.035 -0.3104 0.0088 -0.9499>}
 dyn_loop_a_02 {parent=dyn_loop_a_01, loc=<9.6399 0 0>, rot=<-0.0101 -0.0897 -0.0088 -0.9959>}
 dyn_loop_b_MSTR {parent=dyn_knot, loc=<3.2933 -0.9398 2.995>, rot=<-0.3394 -0.241 0.7472 -0.5181>}
 dyn_loop_b_01 {parent=dyn_loop_b_MSTR, loc=<9.0359 0 0>, rot=<0.0023 -0.2352 -0.0013 -0.9719>}
 dyn_loop_b_02 {parent=dyn_loop_b_01, loc=<8.5933 0 0>, rot=<-0.0022 -0.0899 -0.033 -0.9954>}
 dyn_sword_MSTR {parent=pelvis, loc=<8.6358 -14.862 -0.8174>, rot=<-0.0127 -0.707 0.707 0.0127>}
 dyn_sword {parent=dyn_sword_MSTR, loc=<0 0 -1.2811>, rot=<0 0 0 1>}
 dyn_belt_fix {parent=dyn_sword, loc=<2.5612 7.2437 -0.1689>, rot=<0.5 0.5 -0.5 -0.5>}
 dyn_root_simspace {parent=root, loc=<0 0 0>, rot=<0 0 0 1>}
]
variables [targeting {type=double, default=0}]
