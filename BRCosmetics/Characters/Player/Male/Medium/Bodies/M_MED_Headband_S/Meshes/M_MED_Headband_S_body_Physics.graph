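# Physics asset for the medium male Headband_S body mesh, covering its dangling waist knot, loops, ropes and right-side skirt panels.
# Capsules on pelvis, thighs and calves, plus a root sphere, form the body volumes alongside capsules on the dynamic bones.
# Two rope and two loop chains branch from a pelvis-attached knot, with limited swing and tighter twist between segments.
# Front, side and back skirt chains hang from the pelvis with free angular motion, as do body and anchor joints.
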PhysicsAsset /BRCosmetics/Characters/Player/Male/Medium/Bodies/M_MED_Headband_S/Meshes/M_MED_Headband_S_body_Physics
bodies {
 calf_l [{shapeType=Capsule, length=35.1309, radius=14.869, shapeName=calf_l_capsule, center=<-26.247 -3.0905 2.0001>, rotation=<rot 90 0 0>}]
 calf_r [{shapeType=Capsule, length=35.1309, radius=13.869, shapeName=calf_r_capsule, center=<26.247 3.0905 -2.0001>, rotation=<rot 90 180 0>}]
 dyn_knot_MSTR [{shapeType=Capsule, length=0.505, radius=0.505, shapeName=dyn_knot_MSTR_capsule, rotation=<rot 0 -90 0>}]
 dyn_loop_a_01 [{shapeType=Capsule, length=4.5146, radius=9, shapeName=dyn_loop_a_01_capsule, center=<6.0594 3 0>}]
 dyn_loop_a_02 [{shapeType=Capsule, length=1.8173, radius=9, shapeName=dyn_loop_a_02_capsule, center=<6.0132 3 0>}]
 dyn_loop_a_MSTR [{shapeType=Capsule, length=0.505, radius=0.505, shapeName=dyn_loop_a_MSTR_capsule, rotation=<rot 0 -90 0>}]
 dyn_loop_b_01 [{shapeType=Capsule, length=5.5, radius=7, shapeName=dyn_loop_b_01_capsule, center=<5.2687 3 0>, rotation=<rot -8.1843 0.4267 -2.9947>}]
 dyn_loop_b_02 [{shapeType=Capsule, length=5.8828, radius=7, shapeName=dyn_loop_b_02_capsule, center=<6.2555 3 0>, rotation=<rot 0 0 -4.7394>}]
 dyn_loop_b_MSTR [{shapeType=Capsule, length=0.505, radius=0.505, shapeName=dyn_loop_b_MSTR_capsule, rotation=<rot 0 -90 0>}]
 dyn_rope_a_01 [{shapeType=Capsule, length=8, radius=3.0406, shapeName=dyn_rope_a_01_capsule, center=<4.7527 0 0>, rotation=<rot 90 0 0>}]
 dyn_rope_a_02 [{shapeType=Capsule, length=5.4196, radius=3.2985, shapeName=dyn_rope_a_02_capsule, center=<6.9895 0 0>, rotation=<rot 90 0 0>}]
 dyn_rope_a_03 [{shapeType=Capsule, length=7.4414, radius=3.6209, shapeName=dyn_rope_a_03_capsule, center=<8.8756 -0.0395 -0.3605>, rotation=<rot 90 0 0>}]
 dyn_rope_a_MSTR [{shapeType=Capsule, length=0.505, radius=0.505, shapeName=dyn_rope_a_MSTR_capsule, rotation=<rot 0 -90 0>}]
 dyn_rope_b_01 [{shapeType=Capsule, length=5.5, radius=2.2, shapeName=dyn_rope_b_01_capsule, center=<0.6891 2.4691 -0.2398>, rotation=<rot 90 5.0006 0>}]
 dyn_rope_b_02 [{shapeType=Capsule, length=7, radius=3, shapeName=dyn_rope_b_02_capsule, center=<3.2608 0.6469 0>, rotation=<rot 90 -1.9936 0>}]
 dyn_rope_b_03 [{shapeType=Capsule, length=7.4422, radius=3.4833, shapeName=dyn_rope_b_03_capsule, center=<6.6992 0.4334 0>, rotation=<rot 90 -2.044 0>}]
 dyn_rope_b_MSTR [{shapeType=Capsule, length=0.505, radius=0.505, shapeName=dyn_rope_b_MSTR_capsule, center=<0 0 0.2203>, rotation=<rot 0 -90 0>}]
 dyn_skirt_bk_b_r_01 [{shapeType=Capsule, length=12, radius=5, shapeName=dyn_skirt_bk_b_r_capsule, center=<6.2617 -0.3619 -0.0491>, rotation=<rot 0 0 -90>}]
 dyn_skirt_bk_b_r_02 [{shapeType=Capsule, length=12, radius=5, shapeName=dyn_skirt_bk_b_r_capsule, center=<6.2617 -0.3619 -0.0491>, rotation=<rot 0 0 -90>}]
 dyn_skirt_bk_b_r_03 [{shapeType=Capsule, length=12, radius=5, shapeName=dyn_skirt_bk_b_r_capsule, center=<6.2617 -0.3619 -0.0491>, rotation=<rot 0 0 -90>}]
 dyn_skirt_fr_a_r [{
   shapeType=Capsule
   length=4
   radius=5
   shapeName=dyn_skirt_bk_b_r_capsule
   center=<-5.1901 -1.8058 -0.2381>
   rotation=<rot 0.0003 0.0001 -99.1443>
  }]
 dyn_skirt_fr_b_r [{
   shapeType=Capsule
   length=4
   radius=5
   shapeName=dyn_skirt_bk_b_r_capsule
   center=<6.2617 -1.8409 -0.2871>
   rotation=<rot 0.0002 -0.0001 -99.1443>
  }]
 dyn_skirt_fr_b_r_01 [{
   shapeType=Capsule
   length=4
   radius=5
   shapeName=dyn_skirt_bk_b_r_capsule
   center=<6.2617 -1.8409 -0.2871>
   rotation=<rot 0.0002 -0.0001 -99.1443>
  }]
 dyn_skirt_fr_b_r_02 [{
   shapeType=Capsule
   length=4
   radius=5
   shapeName=dyn_skirt_bk_b_r_capsule
   center=<6.2617 -1.8409 -0.2871>
   rotation=<rot 0.0002 -0.0002 -99.1443>
  }]
 dyn_skirt_side_a_r [{shapeType=Capsule, length=6, radius=5, shapeName=dyn_skirt_bk_b_r_capsule, center=<-5.1901 2.5507 0.0622>, rotation=<rot 0 0.0002 -85.0655>}]
 dyn_skirt_side_b_r [{shapeType=Capsule, length=6, radius=5, shapeName=dyn_skirt_bk_b_r_capsule, center=<6.2617 2.5156 -4.3994>, rotation=<rot 0 -0.0002 -85.0654>}]
 dyn_skirt_side_b_r_01 [{
   shapeType=Capsule
   length=10
   radius=5
   shapeName=dyn_skirt_bk_b_r_capsule
   center=<6.2617 4.3085 -4.5542>
   rotation=<rot 0 -0.0002 -85.0654>
  }]
 dyn_skirt_side_b_r_02 [{shapeType=Capsule, length=12, radius=5, shapeName=dyn_skirt_bk_b_r_capsule, center=<6.2617 6.058 -4.2326>, rotation=<rot 0 -0.0004 -76.5889>}]
 dyn_skirt_side_b_r_03 [{shapeType=Capsule, length=6, radius=5, shapeName=dyn_skirt_bk_b_r_capsule, center=<6.2617 7.9962 -4.8726>, rotation=<rot 0 -0.0002 -77.9275>}]
 pelvis [{shapeType=Capsule, length=12.5725, radius=9.3807, center=<3.5211 -0.0001 0>}]
 root [{shapeType=Sphere, radius=30.4872, shapeName=root_sphere, center=<0 2.5 8.5>}]
 thigh_l [{shapeType=Capsule, length=25, radius=14, shapeName=thigh_l_capsule, center=<-15 0 0>, rotation=<rot 90 -90 -89.1297>}]
 thigh_r [{shapeType=Capsule, length=24.5, radius=14.5, shapeName=thigh_r_capsule, center=<19.8412 5.9078 -3.127>, rotation=<rot 84.9623 6.0147 4.963>}]
}
constraints [
 {bone1Name=dyn_rope_b_03, bone2Name=dyn_rope_b_02, swing1LimitDegrees=35, swing2LimitDegrees=35, twistLimitDegrees=3}
 {
  bone1Name=dyn_rope_a_MSTR
  bone2Name=dyn_knot_MSTR
  swing1LimitDegrees=45
  swing1Motion=Free
  swing2LimitDegrees=45
  swing2Motion=Free
  twistLimitDegrees=45
  twistMotion=Free
 }
 {bone1Name=dyn_rope_a_01, bone2Name=dyn_rope_a_MSTR, swing1LimitDegrees=35, swing2LimitDegrees=35, twistLimitDegrees=3}
 {bone1Name=dyn_rope_a_02, bone2Name=dyn_rope_a_01, swing1LimitDegrees=35, swing2LimitDegrees=35, twistLimitDegrees=3}
 {bone1Name=dyn_rope_a_03, bone2Name=dyn_rope_a_02, swing1LimitDegrees=35, swing2LimitDegrees=35, twistLimitDegrees=3}
 {
  bone1Name=dyn_rope_b_MSTR
  bone2Name=dyn_knot_MSTR
  swing1LimitDegrees=45
  swing1Motion=Free
  swing2LimitDegrees=45
  swing2Motion=Free
  twistLimitDegrees=45
  twistMotion=Free
 }
 {bone1Name=dyn_rope_b_01, bone2Name=dyn_rope_b_MSTR, swing1LimitDegrees=30, swing2LimitDegrees=30, twistLimitDegrees=3}
 {bone1Name=dyn_rope_b_02, bone2Name=dyn_rope_b_01, swing1LimitDegrees=35, swing2LimitDegrees=35, twistLimitDegrees=3}
 {
  bone1Name=dyn_knot_MSTR
  bone2Name=pelvis
  swing1LimitDegrees=45
  swing1Motion=Free
  swing2LimitDegrees=45
  swing2Motion=Free
  twistLimitDegrees=45
  twistMotion=Free
 }
 {
  bone1Name=thigh_r
  bone2Name=pelvis
  swing1LimitDegrees=45
  swing1Motion=Free
  swing2LimitDegrees=45
  swing2Motion=Free
  twistLimitDegrees=45
  twistMotion=Free
 }
 {
  bone1Name=thigh_l
  bone2Name=pelvis
  swing1LimitDegrees=45
  swing1Motion=Free
  swing2LimitDegrees=45
  swing2Motion=Free
  twistLimitDegrees=45
  twistMotion=Free
 }
 {
  bone1Name=dyn_loop_a_MSTR
  bone2Name=dyn_knot_MSTR
  swing1LimitDegrees=45
  swing1Motion=Free
  swing2LimitDegrees=45
  swing2Motion=Free
  twistLimitDegrees=45
  twistMotion=Free
 }
 {bone1Name=dyn_loop_a_01, bone2Name=dyn_loop_a_MSTR, swing1LimitDegrees=35, swing2LimitDegrees=35, twistLimitDegrees=10}
 {bone1Name=dyn_loop_a_02, bone2Name=dyn_loop_a_01, swing1LimitDegrees=35, swing2LimitDegrees=35, twistLimitDegrees=10}
 {
  bone1Name=dyn_loop_b_MSTR
  bone2Name=dyn_knot_MSTR
  swing1LimitDegrees=45
  swing1Motion=Free
  swing2LimitDegrees=45
  swing2Motion=Free
  twistLimitDegrees=45
  twistMotion=Free
 }
 {bone1Name=dyn_loop_b_01, bone2Name=dyn_loop_b_MSTR, swing1LimitDegrees=35, swing2LimitDegrees=35, twistLimitDegrees=10}
 {bone1Name=dyn_loop_b_02, bone2Name=dyn_loop_b_01, swing1LimitDegrees=35, swing2LimitDegrees=35, twistLimitDegrees=10}
 {
  bone1Name=pelvis
  bone2Name=root
  swing1LimitDegrees=45
  swing1Motion=Free
  swing2LimitDegrees=45
  swing2Motion=Free
  twistLimitDegrees=45
  twistMotion=Free
 }
 {
  bone1Name=calf_l
  bone2Name=thigh_l
  swing1LimitDegrees=45
  swing1Motion=Free
  swing2LimitDegrees=45
  swing2Motion=Free
  twistLimitDegrees=45
  twistMotion=Free
 }
 {
  bone1Name=calf_r
  bone2Name=thigh_r
  swing1LimitDegrees=45
  swing1Motion=Free
  swing2LimitDegrees=45
  swing2Motion=Free
  twistLimitDegrees=45
  twistMotion=Free
 }
 {
  bone1Name=dyn_skirt_side_a_r
  bone2Name=pelvis
  swing1LimitDegrees=45
  swing1Motion=Free
  swing2LimitDegrees=45
  swing2Motion=Free
  twistLimitDegrees=45
  twistMotion=Free
 }
 {
  bone1Name=dyn_skirt_side_b_r
  bone2Name=dyn_skirt_side_a_r
  swing1LimitDegrees=45
  swing1Motion=Free
  swing2LimitDegrees=45
  swing2Motion=Free
  twistLimitDegrees=45
  twistMotion=Free
 }
 {
  bone1Name=dyn_skirt_side_b_r_01
  bone2Name=dyn_skirt_side_b_r
  swing1LimitDegrees=45
  swing1Motion=Free
  swing2LimitDegrees=45
  swing2Motion=Free
  twistLimitDegrees=45
  twistMotion=Free
 }
 {
  bone1Name=dyn_skirt_side_b_r_02
  bone2Name=dyn_skirt_side_b_r_01
  swing1LimitDegrees=45
  swing1Motion=Free
  swing2LimitDegrees=45
  swing2Motion=Free
  twistLimitDegrees=45
  twistMotion=Free
 }
 {
  bone1Name=dyn_skirt_side_b_r_03
  bone2Name=dyn_skirt_side_b_r_02
  swing1LimitDegrees=45
  swing1Motion=Free
  swing2LimitDegrees=45
  swing2Motion=Free
  twistLimitDegrees=45
  twistMotion=Free
 }
 {
  bone1Name=dyn_skirt_bk_b_r_01
  bone2Name=pelvis
  swing1LimitDegrees=45
  swing1Motion=Free
  swing2LimitDegrees=45
  swing2Motion=Free
  twistLimitDegrees=45
  twistMotion=Free
 }
 {
  bone1Name=dyn_skirt_bk_b_r_02
  bone2Name=dyn_skirt_bk_b_r_01
  swing1LimitDegrees=45
  swing1Motion=Free
  swing2LimitDegrees=45
  swing2Motion=Free
  twistLimitDegrees=45
  twistMotion=Free
 }
 {
  bone1Name=dyn_skirt_bk_b_r_03
  bone2Name=dyn_skirt_bk_b_r_02
  swing1LimitDegrees=45
  swing1Motion=Free
  swing2LimitDegrees=45
  swing2Motion=Free
  twistLimitDegrees=45
  twistMotion=Free
 }
 {
  bone1Name=dyn_skirt_fr_a_r
  bone2Name=pelvis
  swing1LimitDegrees=45
  swing1Motion=Free
  swing2LimitDegrees=45
  swing2Motion=Free
  twistLimitDegrees=45
  twistMotion=Free
 }
 {
  bone1Name=dyn_skirt_fr_b_r
  bone2Name=dyn_skirt_fr_a_r
  swing1LimitDegrees=45
  swing1Motion=Free
  swing2LimitDegrees=45
  swing2Motion=Free
  twistLimitDegrees=45
  twistMotion=Free
 }
 {
  bone1Name=dyn_skirt_fr_b_r_01
  bone2Name=dyn_skirt_fr_b_r
  swing1LimitDegrees=45
  swing1Motion=Free
  swing2LimitDegrees=45
  swing2Motion=Free
  twistLimitDegrees=45
  twistMotion=Free
 }
 {
  bone1Name=dyn_skirt_fr_b_r_02
  bone2Name=dyn_skirt_fr_b_r_01
  swing1LimitDegrees=45
  swing1Motion=Free
  swing2LimitDegrees=45
  swing2Motion=Free
  twistLimitDegrees=45
  twistMotion=Free
 }
]
