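# Physics asset for the collar of the M_MED_Headband_S male player cosmetic, setting up secondary motion for five collar flaps: back, back-left, back-right, left and right.
# Each flap is a box body on its 01 bone, joined by a constraint with identical swing and twist limits to a tiny capsule on its MSTR bone.
# Two overlapping capsules on the head bone provide an unconstrained collision volume alongside the flaps.
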PhysicsAsset /BRCosmetics/Characters/Player/Male/Medium/Bodies/M_MED_Headband_S/Meshes/M_MED_Headband_S_collar_Physics
bodies {
 dyn_collar_bk_01 [{shapeType=Box, extentX=10, extentY=15, extentZ=3.6987, shapeName=dyn_collar_bk_01_box, center=<7 0 3>, rotation=<rot 15 0 0>}]
 dyn_collar_bk_MSTR [{shapeType=Capsule, length=0.505, radius=0.505, shapeName=dyn_collar_bk_MSTR_capsule, rotation=<rot 0 -90 0>}]
 dyn_collar_bk_l_01 [{shapeType=Box, extentX=10, extentY=2.9645, extentZ=15.8453, shapeName=dyn_collar_bk_l_01_box, center=<6 -2 0>, rotation=<rot 0 -15 -6.9582>}]
 dyn_collar_bk_l_MSTR [{shapeType=Capsule, length=0.505, radius=0.505, shapeName=dyn_collar_bk_l_MSTR_capsule, rotation=<rot 0 -90 0>}]
 dyn_collar_bk_r_01 [{
   shapeType=Box
   extentX=10
   extentY=15
   extentZ=4.0305
   shapeName=dyn_collar_bk_r_01_box
   center=<6 0 2>
   rotation=<rot 14.4996 4.8657 14.6152>
  }]
 dyn_collar_bk_r_MSTR [{shapeType=Capsule, length=0.505, radius=0.505, shapeName=dyn_collar_bk_r_MSTR_capsule, rotation=<rot 0 -90 0>}]
 dyn_collar_l_01 [{shapeType=Box, extentX=10, extentY=15, extentZ=3.4854, shapeName=dyn_collar_l_01_box, center=<6 0 0>, rotation=<rot -6 0 25>}]
 dyn_collar_l_MSTR [{shapeType=Capsule, length=0.505, radius=0.505, shapeName=dyn_collar_l_MSTR_capsule, rotation=<rot 0 -90 0>}]
 dyn_collar_r_01 [{shapeType=Box, extentX=10, extentY=15, extentZ=2, shapeName=dyn_collar_r_01_box, center=<6 0 0>, rotation=<rot -6 2 -25>}]
 dyn_collar_r_MSTR [{shapeType=Capsule, length=0.505, radius=0.505, shapeName=dyn_collar_r_MSTR_capsule, rotation=<rot 0 -90 0>}]
 head [
  {shapeType=Capsule, length=5.5113, radius=8.8509, shapeName=head_capsule, center=<5.6039 -4.8181 0.0001>, rotation=<rot 90 0 0>}
  {shapeType=Capsule, length=8, radius=9.5, shapeName=head_capsule, center=<1.8464 2.2423 0>, rotation=<rot 90 0 0>}
 ]
}
constraints [
 {bone1Name=dyn_collar_bk_01, bone2Name=dyn_collar_bk_MSTR, swing1LimitDegrees=45, swing2LimitDegrees=45, twistLimitDegrees=15}
 {bone1Name=dyn_collar_bk_r_01, bone2Name=dyn_collar_bk_r_MSTR, swing1LimitDegrees=45, swing2LimitDegrees=45, twistLimitDegrees=15}
 {bone1Name=dyn_collar_bk_l_01, bone2Name=dyn_collar_bk_l_MSTR, swing1LimitDegrees=45, swing2LimitDegrees=45, twistLimitDegrees=15}
 {bone1Name=dyn_collar_l_01, bone2Name=dyn_collar_l_MSTR, swing1LimitDegrees=45, swing2LimitDegrees=45, twistLimitDegrees=15}
 {bone1Name=dyn_collar_r_01, bone2Name=dyn_collar_r_MSTR, swing1LimitDegrees=45, swing2LimitDegrees=45, twistLimitDegrees=15}
]
