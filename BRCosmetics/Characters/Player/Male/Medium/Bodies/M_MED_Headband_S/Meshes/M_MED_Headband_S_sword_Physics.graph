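# Physics asset for the sword on the M_MED_Headband_S medium male player body, apparently for secondary motion on the dyn_sword bone.
# A long sword capsule hangs from a tiny dyn_sword_MSTR anchor with a narrow, nearly single-axis swing limit and free twist.
# Capsules on the pelvis, both thighs and right upper arm look like body colliders for the sword.
# The anchor, thighs and arm connect to the pelvis with angularly free constraints.
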PhysicsAsset /BRCosmetics/Characters/Player/Male/Medium/Bodies/M_MED_Headband_S/Meshes/M_MED_Headband_S_sword_Physics
bodies {
 dyn_sword [{shapeType=Capsule, length=97.9393, radius=5.0062, shapeName=dyn_sword_capsule, center=<-5.5 -15 2.5>, rotation=<rot 0 159.0759 -87.0884>}]
 dyn_sword_MSTR [{shapeType=Capsule, length=0.505, radius=0.505, shapeName=dyn_sword_MSTR_capsule, rotation=<rot 0 -90 0>}]
 pelvis [{shapeType=Capsule, length=12.5725, radius=9.3807, center=<3.5211 -0.0001 0>}]
 thigh_l [{
   shapeType=Capsule
   length=14.4207
   radius=18.7136
   shapeName=thigh_l_capsule
   center=<-21.1364 -6.1838 4.1498>
   rotation=<rot 90 -90 -86.0901>
  }]
 thigh_r [{
   shapeType=Capsule
   length=14.1251
   radius=19.0092
   shapeName=thigh_r_capsule
   center=<21.1413 6.2553 -4.7054>
   rotation=<rot 90 14.0363 17.9462>
  }]
 upperarm_twist_01_r [{
   shapeType=Capsule
   length=35.8974
   radius=10.9207
   shapeName=upperarm_twist_01_r_capsule
   center=<-16.8772 -3.2436 3.0246>
   rotation=<rot 90 0 0>
  }]
}
constraints [
 {
  bone1Name=thigh_r
  bone2Name=pelvis
  swing1LimitDegrees=45
  swing1Motion=Free
  swing2LimitDegrees=45
  swing2Motion=Free
  twistLimitDegrees=45
  twistMotion=Free
 }
 {
  bone1Name=thigh_l
  bone2Name=pelvis
  swing1LimitDegrees=45
  swing1Motion=Free
  swing2LimitDegrees=45
  swing2Motion=Free
  twistLimitDegrees=45
  twistMotion=Free
 }
 {
  bone1Name=dyn_sword_MSTR
  bone2Name=pelvis
  swing1LimitDegrees=45
  swing1Motion=Free
  swing2LimitDegrees=45
  swing2Motion=Free
  twistLimitDegrees=45
  twistMotion=Free
 }
 {bone1Name=dyn_sword, bone2Name=dyn_sword_MSTR, swing1LimitDegrees=25, swing2LimitDegrees=1, twistLimitDegrees=75, twistMotion=Free}
 {
  bone1Name=upperarm_twist_01_r
  bone2Name=pelvis
  swing1LimitDegrees=45
  swing1Motion=Free
  swing2LimitDegrees=45
  swing2Motion=Free
  twistLimitDegrees=45
  twistMotion=Free
 }
]
