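# Animation Blueprint for the M_MED_Headband_S male player body cosmetic, driving secondary motion of its skirt, ropes, loops, knot and sword.
# Constraints, Look Ats and thigh-driven additive bone offsets keep cloth bones following the legs, alongside RigidBody simulation.
# Sprint, targeting, emote, zipline and movement variables toggle corrections, including a targeting Control Rig.
# Cached poses link stages; the output applies sword Anim Dynamics, then skirt and rope offsets.
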AnimBlueprint /BRCosmetics/Characters/Player/Male/Medium/Bodies/M_MED_Headband_S/Meshes/M_MED_Headband_S_body_AnimBP
target_skeleton=&/M_MED_Headband_S_body_Skeleton
graphs [
 AnimGraph {nodes [
   Apply Additive!(
    Base
    Additive="Sequence Player (additive) 'M_MED_Headband_S_body_referencepose_Frontend_Sequence'"() {sequence=&/M_MED_Headband_S_body_referencepose_Frontend_Sequence}
    bAlphaBoolEnabled:bool=Get bRigidBodyEnabled()
   ) -> (Pose) {alphaInputType=Bool, alphaBoolBlend {blendInTime=0.1, blendOption=Linear, blendOutTime=0.1}}
   n1 = Get thigh_rot_up_l() -> (thigh_rot_up_l:double)
   n2 = "Transform (Modify) Bone - Bone: dyn_skirt_fr_a_l"(
    ComponentPose=Local To Component(LocalPose=Use cached pose 'prerigid'())
    Alpha=@n1
    Translation="-5,5,5"
    Rotation="15,0.000,0"
    Scale="1,1,1"
   ) {
    rotationMode=BMM_Additive
    rotationSpace=BCS_BoneSpace
    translationMode=BMM_Additive
    translationSpace=BCS_BoneSpace
    alphaBoolBlend {blendOption=Linear}
    boneToModify {boneName=dyn_skirt_fr_a_l}
    rotation {pitch=15}
    translation=<-5 5 5>
   }
   n3 = "Transform (Modify) Bone - Bone: dyn_skirt_side_a_l"(ComponentPose=@n2, Alpha=@n1, Translation="-7,2,3", Rotation="10,0.000,0", Scale="1,1,1") {
    rotationMode=BMM_Additive
    rotationSpace=BCS_BoneSpace
    translationMode=BMM_Additive
    translationSpace=BCS_BoneSpace
    alphaBoolBlend {blendOption=Linear}
    boneToModify {boneName=dyn_skirt_side_a_l}
    rotation {pitch=10}
    translation=<-7 2 3>
   }
   n4 = Get thigh_rot_up_r() -> (thigh_rot_up_r:double)
   n5 = "Transform (Modify) Bone - Bone: dyn_skirt_fr_a_r"(ComponentPose=@n3, Alpha=@n4, Translation="-5,-5,-7", Rotation="15,0,0", Scale="1,1,1") {
    rotationMode=BMM_Additive
    rotationSpace=BCS_BoneSpace
    translationMode=BMM_Additive
    translationSpace=BCS_BoneSpace
    alphaBoolBlend {blendOption=Linear}
    boneToModify {boneName=dyn_skirt_fr_a_r}
    rotation {pitch=15}
    translation=<-5 -5 -7>
   }
   n6 = "Transform (Modify) Bone - Bone: dyn_skirt_side_a_r"(ComponentPose=@n5, Alpha=@n4, Translation="7,-2,-3", Rotation="10,0,0.000", Scale="1,1,1") {
    rotationMode=BMM_Additive
    rotationSpace=BCS_BoneSpace
    translationMode=BMM_Additive
    translationSpace=BCS_BoneSpace
    alphaBoolBlend {blendOption=Linear}
    boneToModify {boneName=dyn_skirt_side_a_r}
    rotation {pitch=10}
    translation=<7 -2 -3>
   }
   n7 = RigidBody!(ComponentPose=@n6, Alpha, OverrideWorldGravity:Vector=Get GravityOverride(), ExternalForce) -> (Pose) {
    bClampLinearTranslationLimitToRefPose=true
    bOverrideWorldGravity=true
    overridePhysicsAsset=&/M_MED_Headband_S_Physics
    simulationSpace=BaseBoneSpace
    alphaBoolBlend {blendOption=Linear}
    baseBoneRef {boneName=dyn_root_simspace}
    componentAppliedLinearAccClamp=<3000 3000 3000>
    componentLinearAccScale=<0.25 0.25 0.4>
    componentLinearVelScale=<0.2 0.2 0.2>
   }
   Component To Local(ComponentPose=@n7)
   n8 = "Constraint - dyn_skirt_fr_a_base_l (RR)"!(ComponentPose, Alpha=1, ConstraintWeights_0=0.1, ConstraintWeights_1:float=0.9) -> (Pose) {
    alphaBoolBlend {blendOption=Linear}
    boneToModify {boneName=dyn_skirt_fr_a_base_l}
    constraintSetup [{transformType=Rotation, targetBone {boneName=pelvis}}, {transformType=Rotation, targetBone {boneName=thigh_l}}]
    constraintWeights [1,1]
   }
   n9 = "Constraint - dyn_skirt_fr_a_base_r (RR)"!(ComponentPose=@n8, Alpha=1, ConstraintWeights_0=0.1, ConstraintWeights_1:float=0.9) -> (Pose) {
    alphaBoolBlend {blendOption=Linear}
    boneToModify {boneName=dyn_skirt_fr_a_base_r}
    constraintSetup [{transformType=Rotation, targetBone {boneName=pelvis}}, {transformType=Rotation, targetBone {boneName=thigh_r}}]
    constraintWeights [1,1]
   }
   n10 = "Constraint - dyn_skirt_side_a_base_l (RR)"!(ComponentPose=@n9, Alpha=1, ConstraintWeights_0=0.4, ConstraintWeights_1:float=0.6) -> (Pose) {
    alphaBoolBlend {blendOption=Linear}
    boneToModify {boneName=dyn_skirt_side_a_base_l}
    constraintSetup [{transformType=Rotation, targetBone {boneName=pelvis}}, {transformType=Rotation, targetBone {boneName=thigh_l}}]
    constraintWeights [1,1]
   }
   n11 = "Constraint - dyn_skirt_side_a_base_r (RR)"!(ComponentPose=@n10, Alpha=1, ConstraintWeights_0=0.4, ConstraintWeights_1:float=0.6) -> (Pose) {
    alphaBoolBlend {blendOption=Linear}
    boneToModify {boneName=dyn_skirt_side_a_base_r}
    constraintSetup [{transformType=Rotation, targetBone {boneName=pelvis}}, {transformType=Rotation, targetBone {boneName=thigh_r}}]
    constraintWeights [1,1]
   }
   n12 = "Constraint - dyn_skirt_bk_a_base_l (RR)"!(ComponentPose=@n11, Alpha=1, ConstraintWeights_0=0.9, ConstraintWeights_1:float=0.1) -> (Pose) {
    alphaBoolBlend {blendOption=Linear}
    boneToModify {boneName=dyn_skirt_bk_a_base_l}
    constraintSetup [{transformType=Rotation, targetBone {boneName=pelvis}}, {transformType=Rotation, targetBone {boneName=thigh_l}}]
    constraintWeights [1,1]
   }
   n13 = "Constraint - dyn_skirt_bk_a_base_r (RR)"!(ComponentPose=@n12, Alpha=1, ConstraintWeights_0=0.9, ConstraintWeights_1:float=0.1) -> (Pose) {
    alphaBoolBlend {blendOption=Linear}
    boneToModify {boneName=dyn_skirt_bk_a_base_r}
    constraintSetup [{transformType=Rotation, targetBone {boneName=pelvis}}, {transformType=Rotation, targetBone {boneName=thigh_r}}]
    constraintWeights [1,1]
   }
   n14 = "Constraint - dyn_skirt_fr_a_base_l (P)"!(ComponentPose=@n13, bAlphaBoolEnabled:bool=Get bIsSprinting(), ConstraintWeights_0=1) -> (Pose) {
    alphaInputType=Bool
    alphaBoolBlend {blendInTime=0.1, blendOption=Linear, blendOutTime=0.1}
    boneToModify {boneName=dyn_skirt_fr_a_base_l}
    constraintSetup [{targetBone {boneName=thigh_l}}]
    constraintWeights [1]
   }
   n15 = Look At - Bone: dyn_skirt_side_a_base_r(ComponentPose=@n14, Alpha=1) {
    alphaBoolBlend {blendOption=Linear}
    alphaScaleBiasClamp {bInterpResult=true}
    boneToModify {boneName=dyn_skirt_side_a_base_r}
    lookAtLocation=<0 -5 25>
    lookAtTarget {boneReference {boneName=calf_r}}
    lookAt_Axis {axis=<-1 0>}
    lookUp_Axis {axis {x=0}}
   }
   n16 = Look At - Bone: dyn_skirt_side_a_base_l(ComponentPose=@n15, Alpha=1) {
    alphaBoolBlend {blendOption=Linear}
    alphaScaleBiasClamp {bInterpResult=true}
    boneToModify {boneName=dyn_skirt_side_a_base_l}
    lookAtLocation=<0 5 -25>
    lookAtTarget {boneReference {boneName=calf_l}}
    lookAt_Axis {axis=<1 0>}
    lookUp_Axis {axis {x=0}}
   }
   n17 = Get thigh_rot_up_r()
   n18 = "Transform (Modify) Bone - Bone: dyn_skirt_fr_a_r"(ComponentPose=@n16, Alpha=@n17, Translation="2,-5,-7", Rotation="25,5,0", Scale="1,1,1") {
    rotationMode=BMM_Additive
    rotationSpace=BCS_BoneSpace
    translationMode=BMM_Additive
    translationSpace=BCS_BoneSpace
    alphaBoolBlend {blendOption=Linear}
    boneToModify {boneName=dyn_skirt_fr_a_r}
    rotation {pitch=25, yaw=5}
    translation=<2 -5 -7>
   }
   n19 = "Transform (Modify) Bone - Bone: dyn_skirt_side_a_r"(ComponentPose=@n18, Alpha=@n17, Translation="0,-2,-3", Rotation="5.000,20,0.000", Scale="1,1,1") {
    rotationMode=BMM_Additive
    rotationSpace=BCS_BoneSpace
    translationMode=BMM_Additive
    translationSpace=BCS_BoneSpace
    alphaBoolBlend {blendOption=Linear}
    boneToModify {boneName=dyn_skirt_side_a_r}
    rotation {pitch=5, yaw=20}
    translation {y=-2, z=-3}
   }
   n20 = Get thigh_rot_up_l()
   n21 = "Transform (Modify) Bone - Bone: dyn_skirt_fr_a_l"(ComponentPose=@n19, Alpha=@n20, Translation="-2,5,7", Rotation="25,-5,0", Scale="1,1,1") {
    rotationMode=BMM_Additive
    rotationSpace=BCS_BoneSpace
    translationMode=BMM_Additive
    translationSpace=BCS_BoneSpace
    alphaBoolBlend {blendOption=Linear}
    boneToModify {boneName=dyn_skirt_fr_a_l}
    rotation {pitch=25, yaw=-5}
    translation=<-2 5 7>
   }
   n22 = "Transform (Modify) Bone - Bone: dyn_skirt_side_a_l"(ComponentPose=@n21, Alpha=@n20, Translation="0,2,3", Rotation="5,20,0", Scale="1,1,1") {
    rotationMode=BMM_Additive
    rotationSpace=BCS_BoneSpace
    translationMode=BMM_Additive
    translationSpace=BCS_BoneSpace
    alphaBoolBlend {blendOption=Linear}
    boneToModify {boneName=dyn_skirt_side_a_l}
    rotation {pitch=5, yaw=20}
    translation {y=2, z=3}
   }
   Component To Local(ComponentPose=@n22)
   n23 = "Constraint - dyn_loop_a_01 (P)"(Alpha=1, ConstraintWeights_0=1) {
    alphaBoolBlend {blendInTime=0.1, blendOption=Linear, blendOutTime=0.1}
    boneToModify {boneName=dyn_loop_a_01}
    constraintSetup [{targetBone {boneName=dyn_skirt_fr_a}}]
    constraintWeights [1]
   }
   n24 = "Constraint - dyn_loop_b_01 (P)"(ComponentPose=@n23, Alpha=1, ConstraintWeights_0=1) {
    alphaBoolBlend {blendInTime=0.1, blendOption=Linear, blendOutTime=0.1}
    boneToModify {boneName=dyn_loop_b_01}
    constraintSetup [{targetBone {boneName=dyn_skirt_fr_a}}]
    constraintWeights [1]
   }
   n25 = "Constraint - dyn_loop_a_02 (P)"(ComponentPose=@n24, Alpha=1, ConstraintWeights_0=1) {
    alphaBoolBlend {blendInTime=0.1, blendOption=Linear, blendOutTime=0.1}
    boneToModify {boneName=dyn_loop_a_02}
    constraintSetup [{targetBone {boneName=dyn_skirt_fr_a}}]
    constraintWeights [1]
   }
   "Constraint - dyn_loop_b_02 (P)"(ComponentPose=@n25, Alpha=1, ConstraintWeights_0=1) {
    alphaBoolBlend {blendInTime=0.1, blendOption=Linear, blendOutTime=0.1}
    boneToModify {boneName=dyn_loop_b_02}
    constraintSetup [{targetBone {boneName=dyn_skirt_fr_a}}]
    constraintWeights [1]
   }
   n26 = "Constraint - dyn_loop_a_MSTR (P)"!(ComponentPose, bAlphaBoolEnabled:bool=True, ConstraintWeights_0=1) -> (Pose) {
    alphaInputType=Bool
    alphaBoolBlend {blendInTime=0.1, blendOption=Linear, blendOutTime=0.1}
    boneToModify {boneName=dyn_loop_a_MSTR}
    constraintSetup [{targetBone {boneName=thigh_l}}]
    constraintWeights [1]
   }
   "Constraint - dyn_loop_b_MSTR (P)"!(ComponentPose=@n26, bAlphaBoolEnabled:bool=True, ConstraintWeights_0=1) -> (Pose) {
    alphaInputType=Bool
    alphaBoolBlend {blendInTime=0.1, blendOption=Linear, blendOutTime=0.1}
    boneToModify {boneName=dyn_loop_b_MSTR}
    constraintSetup [{targetBone {boneName=thigh_l}}]
    constraintWeights [1]
   }
   n27 = "Constraint - dyn_loop_a_MSTR (P)"!(ComponentPose, Alpha:float=1, ConstraintWeights_0=1) -> (Pose) {
    alphaBoolBlend {blendInTime=0.1, blendOption=Linear, blendOutTime=0.1}
    boneToModify {boneName=dyn_loop_a_MSTR}
    constraintSetup [{targetBone {boneName=dyn_skirt_fr_a}}]
    constraintWeights [1]
   }
   "Constraint - dyn_loop_b_MSTR (P)"!(ComponentPose=@n27, Alpha:float=1, ConstraintWeights_0=1) -> (Pose) {
    alphaBoolBlend {blendInTime=0.1, blendOption=Linear, blendOutTime=0.1}
    boneToModify {boneName=dyn_loop_b_MSTR}
    constraintSetup [{targetBone {boneName=dyn_skirt_fr_a}}]
    constraintWeights [1]
   }
   n28 = "Constraint - dyn_skirt_fr_a_base_l (RR)"(ComponentPose=Local To Component(), Alpha=1, ConstraintWeights_0=0.1, ConstraintWeights_1=0.9) {
    alphaBoolBlend {blendOption=Linear}
    boneToModify {boneName=dyn_skirt_fr_a_base_l}
    constraintSetup [{transformType=Rotation, targetBone {boneName=pelvis}}, {transformType=Rotation, targetBone {boneName=thigh_l}}]
    constraintWeights [1,1]
   }
   n29 = "Constraint - dyn_skirt_fr_a_base_r (RR)"(ComponentPose=@n28, Alpha=1, ConstraintWeights_0=0.1, ConstraintWeights_1=0.9) {
    alphaBoolBlend {blendOption=Linear}
    boneToModify {boneName=dyn_skirt_fr_a_base_r}
    constraintSetup [{transformType=Rotation, targetBone {boneName=pelvis}}, {transformType=Rotation, targetBone {boneName=thigh_r}}]
    constraintWeights [1,1]
   }
   n30 = "Constraint - dyn_skirt_side_a_base_l (RR)"(ComponentPose=@n29, Alpha=1, ConstraintWeights_0=0.4, ConstraintWeights_1=0.6) {
    alphaBoolBlend {blendOption=Linear}
    boneToModify {boneName=dyn_skirt_side_a_base_l}
    constraintSetup [{transformType=Rotation, targetBone {boneName=pelvis}}, {transformType=Rotation, targetBone {boneName=thigh_l}}]
    constraintWeights [1,1]
   }
   n31 = "Constraint - dyn_skirt_side_a_base_r (RR)"(ComponentPose=@n30, Alpha=1, ConstraintWeights_0=0.4, ConstraintWeights_1=0.6) {
    alphaBoolBlend {blendOption=Linear}
    boneToModify {boneName=dyn_skirt_side_a_base_r}
    constraintSetup [{transformType=Rotation, targetBone {boneName=pelvis}}, {transformType=Rotation, targetBone {boneName=thigh_r}}]
    constraintWeights [1,1]
   }
   n32 = "Constraint - dyn_skirt_bk_a_base_l (RR)"(ComponentPose=@n31, Alpha=1, ConstraintWeights_0=0.9, ConstraintWeights_1=0.1) {
    alphaBoolBlend {blendOption=Linear}
    boneToModify {boneName=dyn_skirt_bk_a_base_l}
    constraintSetup [{transformType=Rotation, targetBone {boneName=pelvis}}, {transformType=Rotation, targetBone {boneName=thigh_l}}]
    constraintWeights [1,1]
   }
   n33 = "Constraint - dyn_skirt_bk_a_base_r (RR)"(ComponentPose=@n32, Alpha=1, ConstraintWeights_0=0.9, ConstraintWeights_1=0.1) {
    alphaBoolBlend {blendOption=Linear}
    boneToModify {boneName=dyn_skirt_bk_a_base_r}
    constraintSetup [{transformType=Rotation, targetBone {boneName=pelvis}}, {transformType=Rotation, targetBone {boneName=thigh_r}}]
    constraintWeights [1,1]
   }
   n34 = Get thigh_rot_inr_r() -> (thigh_rot_inr_r:double)
   n35 = Look At - Bone: dyn_skirt_side_a_base_r(ComponentPose=@n33, Alpha=@n34) {
    alphaBoolBlend {blendOption=Linear}
    alphaScaleBiasClamp {bInterpResult=true}
    boneToModify {boneName=dyn_skirt_side_a_base_r}
    lookAtLocation=<0 -5 25>
    lookAtTarget {boneReference {boneName=calf_r}}
    lookAt_Axis {axis=<-1 0>}
    lookUp_Axis {axis {x=0}}
   }
   n36 = Get thigh_rot_inr_l() -> (thigh_rot_inr_l:double)
   n37 = Look At - Bone: dyn_skirt_side_a_base_l(ComponentPose=@n35, Alpha=@n36) {
    alphaBoolBlend {blendOption=Linear}
    alphaScaleBiasClamp {bInterpResult=true}
    boneToModify {boneName=dyn_skirt_side_a_base_l}
    lookAtLocation=<0 5 -25>
    lookAtTarget {boneReference {boneName=calf_l}}
    lookAt_Axis {axis=<1 0>}
    lookUp_Axis {axis {x=0}}
   }
   n38 = Get thigh_rot_up_r()
   n39 = "Transform (Modify) Bone - Bone: dyn_skirt_fr_a_r"(ComponentPose=@n37, Alpha=@n38, Translation="2,-5,-7", Rotation="25,5,0", Scale="1,1,1") {
    rotationMode=BMM_Additive
    rotationSpace=BCS_BoneSpace
    translationMode=BMM_Additive
    translationSpace=BCS_BoneSpace
    alphaBoolBlend {blendOption=Linear}
    boneToModify {boneName=dyn_skirt_fr_a_r}
    rotation {pitch=25, yaw=5}
    translation=<2 -5 -7>
   }
   n40 = "Transform (Modify) Bone - Bone: dyn_skirt_side_a_r"(ComponentPose=@n39, Alpha=@n38, Translation="0,-2,-3", Rotation="5.000,20,0.000", Scale="1,1,1") {
    rotationMode=BMM_Additive
    rotationSpace=BCS_BoneSpace
    translationMode=BMM_Additive
    translationSpace=BCS_BoneSpace
    alphaBoolBlend {blendOption=Linear}
    boneToModify {boneName=dyn_skirt_side_a_r}
    rotation {pitch=5, yaw=20}
    translation {y=-2, z=-3}
   }
   n41 = Get thigh_rot_up_l()
   n42 = "Transform (Modify) Bone - Bone: dyn_skirt_fr_a_l"(ComponentPose=@n40, Alpha=@n41, Translation="-2,5,7", Rotation="25,-5,0", Scale="1,1,1") {
    rotationMode=BMM_Additive
    rotationSpace=BCS_BoneSpace
    translationMode=BMM_Additive
    translationSpace=BCS_BoneSpace
    alphaBoolBlend {blendOption=Linear}
    boneToModify {boneName=dyn_skirt_fr_a_l}
    rotation {pitch=25, yaw=-5}
    translation=<-2 5 7>
   }
   n43 = "Transform (Modify) Bone - Bone: dyn_skirt_side_a_l"(ComponentPose=@n42, Alpha=@n41, Translation="0,2,3", Rotation="5,20,0", Scale="1,1,1") {
    rotationMode=BMM_Additive
    rotationSpace=BCS_BoneSpace
    translationMode=BMM_Additive
    translationSpace=BCS_BoneSpace
    alphaBoolBlend {blendOption=Linear}
    boneToModify {boneName=dyn_skirt_side_a_l}
    rotation {pitch=5, yaw=20}
    translation {y=2, z=3}
   }
   n44 = "Constraint - dyn_skirt_fr_a (PP)"!(ComponentPose=@n43, Alpha=1, ConstraintWeights_0=0.5, ConstraintWeights_1:float=0.5) -> (Pose) {
    alphaBoolBlend {blendOption=Linear}
    boneToModify {boneName=dyn_skirt_fr_a}
    constraintSetup [{targetBone {boneName=dyn_skirt_fr_b_l}}, {targetBone {boneName=dyn_skirt_fr_b_r}}]
    constraintWeights [1,1]
   }
   n45 = "Constraint - dyn_skirt_fr_b (PP)"!(ComponentPose=@n44, Alpha=0.75, ConstraintWeights_0=0.5, ConstraintWeights_1:float=0.5) -> (Pose) {
    alphaBoolBlend {blendOption=Linear}
    boneToModify {boneName=dyn_skirt_fr_b}
    constraintSetup [{targetBone {boneName=dyn_skirt_fr_b_l_01}}, {targetBone {boneName=dyn_skirt_fr_b_r_01}}]
    constraintWeights [1,1]
   }
   "Constraint - dyn_skirt_fr_b1 (PP)"!(ComponentPose=@n45, Alpha=1, ConstraintWeights_0=0.5, ConstraintWeights_1:float=0.5) -> (Pose) {
    alphaBoolBlend {blendOption=Linear}
    boneToModify {boneName=dyn_skirt_fr_b1}
    constraintSetup [{targetBone {boneName=dyn_skirt_fr_b_l_02}}, {targetBone {boneName=dyn_skirt_fr_b_r_02}}]
    constraintWeights [1,1]
   }
   Constraint!(ComponentPose, Alpha=1) -> (Pose) {alphaBoolBlend {blendOption=Linear}, boneToModify=~}
   "Constraint - dyn_rope_a_01 (RT)"!(ComponentPose, Alpha=1, ConstraintWeights_0=1, ConstraintWeights_1:float=0) -> (Pose) {
    alphaBoolBlend {blendOption=Linear}
    boneToModify {boneName=dyn_rope_a_01}
    constraintSetup [{transformType=Rotation, targetBone {boneName=dyn_skirt_side_a_base_r}}, {transformType=Translation, targetBone {boneName=None}}]
    constraintWeights [1,1]
   }
   "Constraint - dyn_skirt_bk_a_base (PP)"!(ComponentPose, Alpha=1, ConstraintWeights_0=0.5, ConstraintWeights_1:float=0.5) -> (Pose) {
    alphaBoolBlend {blendOption=Linear}
    boneToModify {boneName=dyn_skirt_bk_a_base}
    constraintSetup [{targetBone {boneName=dyn_skirt_bk_a_l}}, {targetBone {boneName=dyn_skirt_bk_a_r}}]
    constraintWeights [1,1]
   }
   "Constraint - dyn_skirt_fr_b (P)"(Alpha=1, ConstraintWeights_0=1) {
    alphaBoolBlend {blendOption=Linear}
    boneToModify {boneName=dyn_skirt_fr_b}
    constraintSetup [{targetBone {boneName=thigh_l}}]
    constraintWeights [1]
   }
   n46 = "Constraint - dyn_skirt_fr_a (PP)"(Alpha=1, ConstraintWeights_0=0.5, ConstraintWeights_1=0.5) {
    alphaBoolBlend {blendOption=Linear}
    boneToModify {boneName=dyn_skirt_fr_a}
    constraintSetup [{targetBone {boneName=dyn_skirt_fr_b_l}}, {targetBone {boneName=dyn_skirt_fr_b_r}}]
    constraintWeights [1,1]
   }
   n47 = "Constraint - dyn_skirt_fr_b (PP)"(ComponentPose=@n46, Alpha=0.75, ConstraintWeights_0=0.5, ConstraintWeights_1=0.5) {
    alphaBoolBlend {blendOption=Linear}
    boneToModify {boneName=dyn_skirt_fr_b}
    constraintSetup [{targetBone {boneName=dyn_skirt_fr_b_l_01}}, {targetBone {boneName=dyn_skirt_fr_b_r_01}}]
    constraintWeights [1,1]
   }
   "Constraint - dyn_skirt_fr_b1 (PP)"(ComponentPose=@n47, Alpha=1, ConstraintWeights_0=0.5, ConstraintWeights_1=0.5) {
    alphaBoolBlend {blendOption=Linear}
    boneToModify {boneName=dyn_skirt_fr_b1}
    constraintSetup [{targetBone {boneName=dyn_skirt_fr_b_l_02}}, {targetBone {boneName=dyn_skirt_fr_b_r_02}}]
    constraintWeights [1,1]
   }
   Control Rig!(Source, bAlphaBoolEnabled:bool=Get bIsTargeting()) -> (Pose) {
    alphaInputType=Bool
    alphaBoolBlend {blendInTime=0.1, blendOption=Linear, blendOutTime=0.1}
    controlRigAssetReference {blueprintRigClass=&/M_MED_Headband_S_target_CtrlRig.M_MED_Headband_S_target_CtrlRig_C}
   }
   Local To Component()
   n48 = "Constraint - dyn_skirt_fr_a_base_l (RP)"!(ComponentPose=Local To Component(LocalPose=Use cached pose 'constraints'()), Alpha=1, ConstraintWeights_0=0, ConstraintWeights_1:float=1) -> (Pose) {
    alphaBoolBlend {blendOption=Linear}
    boneToModify {boneName=dyn_skirt_fr_a_base_l}
    constraintSetup [{transformType=Rotation, targetBone {boneName=pelvis}}, {targetBone {boneName=thigh_l}}]
    constraintWeights [1,1]
   }
   n49 = "Constraint - dyn_skirt_side_a_base_l (RR)"(ComponentPose=@n48, Alpha=1, ConstraintWeights_0=0.2, ConstraintWeights_1=0.8) {
    alphaBoolBlend {blendOption=Linear}
    boneToModify {boneName=dyn_skirt_side_a_base_l}
    constraintSetup [{transformType=Rotation, targetBone {boneName=pelvis}}, {transformType=Rotation, targetBone {boneName=thigh_l}}]
    constraintWeights [1,1]
   }
   n50 = "Constraint - dyn_skirt_bk_a_base_l (RR)"(ComponentPose=@n49, Alpha=1, ConstraintWeights_0=0.9, ConstraintWeights_1=0.1) {
    alphaBoolBlend {blendOption=Linear}
    boneToModify {boneName=dyn_skirt_bk_a_base_l}
    constraintSetup [{transformType=Rotation, targetBone {boneName=pelvis}}, {transformType=Rotation, targetBone {boneName=thigh_l}}]
    constraintWeights [1,1]
   }
   n51 = "Constraint - dyn_skirt_fr_a_base_r (RP)"!(ComponentPose=@n50, Alpha=1, ConstraintWeights_0=0, ConstraintWeights_1:float=1) -> (Pose) {
    alphaBoolBlend {blendOption=Linear}
    boneToModify {boneName=dyn_skirt_fr_a_base_r}
    constraintSetup [{transformType=Rotation, targetBone {boneName=pelvis}}, {targetBone {boneName=thigh_r}}]
    constraintWeights [1,1]
   }
   n52 = "Constraint - dyn_skirt_side_a_base_r (RR)"(ComponentPose=@n51, Alpha=1, ConstraintWeights_0=0.2, ConstraintWeights_1=0.8) {
    alphaBoolBlend {blendOption=Linear}
    boneToModify {boneName=dyn_skirt_side_a_base_r}
    constraintSetup [{transformType=Rotation, targetBone {boneName=pelvis}}, {transformType=Rotation, targetBone {boneName=thigh_r}}]
    constraintWeights [1,1]
   }
   n53 = "Constraint - dyn_skirt_bk_a_base_r (RR)"(ComponentPose=@n52, Alpha=1, ConstraintWeights_0=0.9, ConstraintWeights_1=0.1) {
    alphaBoolBlend {blendOption=Linear}
    boneToModify {boneName=dyn_skirt_bk_a_base_r}
    constraintSetup [{transformType=Rotation, targetBone {boneName=pelvis}}, {transformType=Rotation, targetBone {boneName=thigh_r}}]
    constraintWeights [1,1]
   }
   n54 = Look At - Bone: dyn_skirt_side_a_base_r(ComponentPose=@n53, Alpha=1) {
    alphaBoolBlend {blendOption=Linear}
    alphaScaleBiasClamp {bInterpResult=true}
    boneToModify {boneName=dyn_skirt_side_a_base_r}
    lookAtLocation=<0 -5 25>
    lookAtTarget {boneReference {boneName=calf_r}}
    lookAt_Axis {axis=<-1 0>}
    lookUp_Axis {axis {x=0}}
   }
   n55 = Look At - Bone: dyn_skirt_side_a_base_l(ComponentPose=@n54, Alpha=1) {
    alphaBoolBlend {blendOption=Linear}
    alphaScaleBiasClamp {bInterpResult=true}
    boneToModify {boneName=dyn_skirt_side_a_base_l}
    lookAtLocation=<0 5 -25>
    lookAtTarget {boneReference {boneName=calf_l}}
    lookAt_Axis {axis=<1 0>}
    lookUp_Axis {axis {x=0}}
   }
   n56 = "Transform (Modify) Bone - Bone: dyn_skirt_fr_a_r"(ComponentPose=@n55, Alpha=1, Translation="0,0,0", Rotation="5,0,0", Scale="1,1,1") {
    rotationMode=BMM_Additive
    rotationSpace=BCS_BoneSpace
    translationMode=BMM_Additive
    translationSpace=BCS_BoneSpace
    alphaBoolBlend {blendOption=Linear}
    boneToModify {boneName=dyn_skirt_fr_a_r}
    rotation {pitch=5}
   }
   n57 = "Transform (Modify) Bone - Bone: dyn_skirt_fr_a_l"(ComponentPose=@n56, Alpha=1, Translation="0,0,0", Rotation="5,0,0", Scale="1,1,1") {
    rotationMode=BMM_Additive
    rotationSpace=BCS_BoneSpace
    translationMode=BMM_Additive
    translationSpace=BCS_BoneSpace
    alphaBoolBlend {blendOption=Linear}
    boneToModify {boneName=dyn_skirt_fr_a_l}
    rotation {pitch=5}
   }
   n58 = Get thigh_rot_r()
   n59 = "Transform (Modify) Bone - Bone: dyn_skirt_fr_a_r"(ComponentPose=@n57, Alpha=@n58, Translation="2,-5,-7", Rotation="20,10,0", Scale="1,1,1") {
    rotationMode=BMM_Additive
    rotationSpace=BCS_BoneSpace
    translationMode=BMM_Additive
    translationSpace=BCS_BoneSpace
    alphaBoolBlend {blendOption=Linear}
    boneToModify {boneName=dyn_skirt_fr_a_r}
    rotation {pitch=20, yaw=10}
    translation=<2 -5 -7>
   }
   n60 = "Transform (Modify) Bone - Bone: dyn_skirt_side_a_r"(ComponentPose=@n59, Alpha=@n58, Translation="0,-2,-3", Rotation="5.000,20,0.000", Scale="1,1,1") {
    rotationMode=BMM_Additive
    rotationSpace=BCS_BoneSpace
    translationMode=BMM_Additive
    translationSpace=BCS_BoneSpace
    alphaBoolBlend {blendOption=Linear}
    boneToModify {boneName=dyn_skirt_side_a_r}
    rotation {pitch=5, yaw=20}
    translation {y=-2, z=-3}
   }
   n61 = Get thigh_rot_l()
   n62 = "Transform (Modify) Bone - Bone: dyn_skirt_fr_a_l"(ComponentPose=@n60, Alpha=@n61, Translation="-2,5,7", Rotation="25,-10,0", Scale="1,1,1") {
    rotationMode=BMM_Additive
    rotationSpace=BCS_BoneSpace
    translationMode=BMM_Additive
    translationSpace=BCS_BoneSpace
    alphaBoolBlend {blendOption=Linear}
    boneToModify {boneName=dyn_skirt_fr_a_l}
    rotation {pitch=25, yaw=-10}
    translation=<-2 5 7>
   }
   n63 = "Transform (Modify) Bone - Bone: dyn_skirt_side_a_l"(ComponentPose=@n62, Alpha=@n61, Translation="0,2,3", Rotation="5,20,0", Scale="1,1,1") {
    rotationMode=BMM_Additive
    rotationSpace=BCS_BoneSpace
    translationMode=BMM_Additive
    translationSpace=BCS_BoneSpace
    alphaBoolBlend {blendOption=Linear}
    boneToModify {boneName=dyn_skirt_side_a_l}
    rotation {pitch=5, yaw=20}
    translation {y=2, z=3}
   }
   Local To Component(LocalPose=Blend Poses by bool(BlendPose_0=Component To Local(ComponentPose=@n63), BlendPose_1=Component To Local(), bActiveValue=Get bIsPlayingEmote()) {blendType=Linear})
   Look At(Alpha=1) {alphaBoolBlend {blendOption=Linear}, boneToModify=~, lookAtTarget=~}
   Look At - Bone: dyn_skirt_fr_a_base_l(Alpha=@n36) {
    alphaBoolBlend {blendOption=Linear}
    alphaScaleBiasClamp {bInterpResult=true}
    boneToModify {boneName=dyn_skirt_fr_a_base_l}
    lookAtLocation=<0 15 -5>
    lookAtTarget {boneReference {boneName=calf_l}}
    lookAt_Axis {axis=<1 0>}
    lookUp_Axis {axis {x=0}}
   }
   Look At - Bone: dyn_skirt_fr_a_base_r(Alpha=@n34) {
    alphaBoolBlend {blendOption=Linear}
    alphaScaleBiasClamp {bInterpResult=true}
    boneToModify {boneName=dyn_skirt_fr_a_base_r}
    lookAtLocation=<0 -15 5>
    lookAtTarget {boneReference {boneName=calf_r}}
    lookAt_Axis {axis=<-1 0>}
    lookUp_Axis {axis {x=0}}
   }
   "Transform (Modify) Bone - Bone: dyn_knot_MSTR"!(ComponentPose, bAlphaBoolEnabled:bool=True, Translation="1.151,6.086,1.106", Rotation="3.938,23.870,14.518", Scale="1,1,1") -> (Pose) {
    alphaInputType=Bool
    rotationMode=BMM_Additive
    rotationSpace=BCS_BoneSpace
    translationMode=BMM_Additive
    translationSpace=BCS_BoneSpace
    alphaBoolBlend {blendInTime=0.1, blendOption=Linear, blendOutTime=0.1}
    boneToModify {boneName=dyn_knot_MSTR}
    rotation=<rot 3.938 23.87 14.518>
    translation=<1.151 6.086 1.106>
   }
   "Transform (Modify) Bone - Bone: dyn_knot_MSTR"(bAlphaBoolEnabled=True, Translation="-2,10,0.776", Rotation="2.676,55,31.882", Scale="1,1,1") {
    alphaInputType=Bool
    rotationMode=BMM_Additive
    rotationSpace=BCS_BoneSpace
    translationMode=BMM_Additive
    translationSpace=BCS_BoneSpace
    alphaBoolBlend {blendInTime=0.1, blendOption=Linear, blendOutTime=0.1}
    boneToModify {boneName=dyn_knot_MSTR}
    rotation=<rot 2.676 55 31.882>
    translation=<-2 10 0.776>
   }
   "Transform (Modify) Bone - Bone: dyn_knot_MSTR"(bAlphaBoolEnabled=True, Translation="-2,7,0.776", Rotation="2.676,35,31.882", Scale="1,1,1") {
    alphaInputType=Bool
    rotationMode=BMM_Additive
    rotationSpace=BCS_BoneSpace
    translationMode=BMM_Additive
    translationSpace=BCS_BoneSpace
    alphaBoolBlend {blendInTime=0.1, blendOption=Linear, blendOutTime=0.1}
    boneToModify {boneName=dyn_knot_MSTR}
    rotation=<rot 2.676 35 31.882>
    translation=<-2 7 0.776>
   }
   "Transform (Modify) Bone - Bone: dyn_loop_b_MSTR"(Alpha=1, Translation="-3,-2.5,4", Rotation="45,90,-20", Scale="1,1,1") {
    rotationMode=BMM_Additive
    rotationSpace=BCS_BoneSpace
    translationMode=BMM_Additive
    translationSpace=BCS_BoneSpace
    alphaBoolBlend {blendOption=Linear}
    boneToModify {boneName=dyn_loop_b_MSTR}
    rotation=<rot 45 90 -20>
    translation=<-3 -2.5 4>
   }
   "Transform (Modify) Bone - Bone: dyn_knot_MSTR"(bAlphaBoolEnabled=True, Translation="-4.443,14.539,0.776", Rotation="2.676,74.879,31.882", Scale="1,1,1") {
    alphaInputType=Bool
    rotationMode=BMM_Additive
    rotationSpace=BCS_BoneSpace
    translationMode=BMM_Additive
    translationSpace=BCS_BoneSpace
    alphaBoolBlend {blendInTime=0.1, blendOption=Linear, blendOutTime=0.1}
    boneToModify {boneName=dyn_knot_MSTR}
    rotation=<rot 2.676 74.8788 31.8821>
    translation=<-4.443 14.539 0.776>
   }
   "Transform (Modify) Bone - Bone: dyn_knot_MSTR"(bAlphaBoolEnabled=True, Translation="1,6,0.5", Rotation="5,35,31.882", Scale="1,1,1") {
    alphaInputType=Bool
    rotationMode=BMM_Additive
    rotationSpace=BCS_BoneSpace
    translationMode=BMM_Additive
    translationSpace=BCS_BoneSpace
    alphaBoolBlend {blendInTime=0.1, blendOption=Linear, blendOutTime=0.1}
    boneToModify {boneName=dyn_knot_MSTR}
    rotation=<rot 5 35 31.882>
    translation=<1 6 0.5>
   }
   "Transform (Modify) Bone - Bone: dyn_loop_a_MSTR"(Alpha=1, Translation="0,0,0", Rotation="-3.312,27.130,24.794", Scale="1,1,1") {
    rotationMode=BMM_Additive
    rotationSpace=BCS_BoneSpace
    translationMode=BMM_Additive
    translationSpace=BCS_BoneSpace
    alphaBoolBlend {blendOption=Linear}
    boneToModify {boneName=dyn_loop_a_MSTR}
    rotation=<rot -3.312 27.13 24.794>
   }
   "Transform (Modify) Bone - Bone: dyn_sword_MSTR"!(
    ComponentPose
    bAlphaBoolEnabled:bool=Get BwdMovement_sword_bool() -> (BwdMovement_sword_bool:bool)
    Translation="0,0,0"
    Rotation="-20,-5,0"
    Scale="1,1,1"
   ) -> (Pose) {
    alphaInputType=Bool
    rotationMode=BMM_Additive
    rotationSpace=BCS_BoneSpace
    translationMode=BMM_Additive
    translationSpace=BCS_BoneSpace
    alphaBoolBlend {blendInTime=0.1, blendOption=Linear, blendOutTime=0.5}
    boneToModify {boneName=dyn_sword_MSTR}
    rotation {pitch=-20, yaw=-5}
   }
   n64 = Get thigh_rot_up_r()
   n65 = "Transform (Modify) Bone - Bone: dyn_loop_a_MSTR"(Alpha=@n64, Translation="0,0,0", Rotation="35.000,35.000,-35", Scale="1,1,1") {
    rotationMode=BMM_Additive
    rotationSpace=BCS_BoneSpace
    translationMode=BMM_Additive
    translationSpace=BCS_BoneSpace
    alphaBoolBlend {blendOption=Linear}
    alphaScaleBiasClamp {bInterpResult=true, interpSpeedDecreasing=40, interpSpeedIncreasing=40}
    boneToModify {boneName=dyn_loop_a_MSTR}
    rotation=<rot 35 35 -35>
   }
   "Transform (Modify) Bone - Bone: dyn_loop_b_MSTR"(ComponentPose=@n65, Alpha=@n64, Translation="0,0,0", Rotation="35,35,-35", Scale="1,1,1") {
    rotationMode=BMM_Additive
    rotationSpace=BCS_BoneSpace
    translationMode=BMM_Additive
    translationSpace=BCS_BoneSpace
    alphaBoolBlend {blendOption=Linear}
    alphaScaleBiasClamp {bInterpResult=true, interpSpeedDecreasing=40, interpSpeedIncreasing=40}
    boneToModify {boneName=dyn_loop_b_MSTR}
    rotation=<rot 35 35 -35>
   }
   "Transform (Modify) Bone - Bone: dyn_rope_b_MSTR"(Alpha=1, Translation="0,-5,-5", Rotation="-35.000,0,0", Scale="1,1,1") {
    rotationMode=BMM_Additive
    rotationSpace=BCS_BoneSpace
    translationMode=BMM_Additive
    translationSpace=BCS_BoneSpace
    alphaBoolBlend {blendOption=Linear}
    alphaScaleBiasClamp {bInterpResult=true, interpSpeedDecreasing=40, interpSpeedIncreasing=40}
    boneToModify {boneName=dyn_rope_b_MSTR}
    rotation {pitch=-35}
    translation {y=-5, z=-5}
   }
   "Transform (Modify) Bone - Bone: dyn_knot"(Alpha=1, Translation="0,0,0", Rotation="5.659,2.536,24.195", Scale="1,1,1") {
    rotationMode=BMM_Additive
    rotationSpace=BCS_BoneSpace
    translationMode=BMM_Additive
    translationSpace=BCS_BoneSpace
    alphaBoolBlend {blendInTime=0.1, blendOption=Linear, blendOutTime=0.1}
    alphaScaleBiasClamp {bInterpResult=true, interpSpeedDecreasing=40, interpSpeedIncreasing=40}
    boneToModify {boneName=dyn_knot}
    rotation=<rot 5.659 2.536 24.1947>
   }
   RigidBody!(ComponentPose, Alpha, ExternalForce) -> (Pose) {
    bClampLinearTranslationLimitToRefPose=true
    bOverrideWorldGravity=true
    overridePhysicsAsset=&/M_MED_Headband_S_sword_Physics
    simulationSpace=BaseBoneSpace
    alphaBoolBlend {blendOption=Linear}
    baseBoneRef {boneName=spine_05}
    componentAppliedLinearAccClamp=<3000 3000 3000>
    componentLinearAccScale=<0.3 0.3 0.3>
    componentLinearVelScale=<0.3 0.3 0.3>
    overrideWorldGravity {z=-980}
   }
   RigidBody!(ComponentPose, bAlphaBoolEnabled:bool=NOT Boolean(A=Get BwdMovement_sword_bool()), OverrideWorldGravity:Vector="0,0,-980", ExternalForce) -> (Pose) {
    alphaInputType=Bool
    bClampLinearTranslationLimitToRefPose=true
    bOverrideWorldGravity=true
    overridePhysicsAsset=&M_MED_Headband_S_sword_Physics
    simulationSpace=BaseBoneSpace
    alphaBoolBlend {blendInTime=0.1, blendOption=Linear, blendOutTime=0.5}
    baseBoneRef {boneName=pelvis}
    componentAppliedLinearAccClamp=<3000 3000 3000>
    componentLinearAccScale=<0.25 0.25 0.4>
    componentLinearVelScale=<0.3 0.3 0.3>
   }
   n66 = "Transform (Modify) Bone - Bone: dyn_sword_MSTR"(
    ComponentPose=Local To Component(LocalPose=Use cached pose 'weapon_pose'())
    bAlphaBoolEnabled=True
    Translation="0,0,2"
    Rotation="-10,-20,0"
    Scale="1,1,1"
   ) {
    alphaInputType=Bool
    rotationMode=BMM_Additive
    rotationSpace=BCS_BoneSpace
    scaleMode=BMM_Additive
    scaleSpace=BCS_BoneSpace
    translationMode=BMM_Additive
    translationSpace=BCS_BoneSpace
    alphaBoolBlend {blendOption=Linear}
    boneToModify {boneName=dyn_sword_MSTR}
    rotation {pitch=-10, yaw=-20}
    translation {z=2}
   }
   n67 = Component To Local(ComponentPose=@n66)
   n68 = Use cached pose 'weapon_pose'()
   n69 = Get bIsMoving2D()
   n70 = Anim Dynamics - Bone: dyn_sword_MSTR(
    ComponentPose=Local To Component(LocalPose=Blend Poses by bool(BlendPose_0=@n67, BlendPose_1=@n68, bActiveValue=@n69) {blendType=Linear})
    Alpha=0.5
    ExternalForce="0,0,0"
   ) {
    angularDampingOverride=0.999
    angularSpringConstant=1000
    bAngularSpring=true
    bGravityOverrideInSimSpace=true
    bOverrideAngularDamping=true
    bUseGravityOverride=true
    gravityScale=0.5
    simulationSpace=BoneRelative
    alphaBoolBlend {blendOption=Linear}
    boundBone {boneName=dyn_sword_MSTR}
    componentAppliedLinearAccClamp=<1000 1000 1000>
    componentLinearAccScale=<1 1 1>
    componentLinearVelScale=<1 1 1>
    gravityOverride {z=-980}
    physicsBodyDefinitions [{
      sphereCollisionRadius=0
      boundBone {boneName=dyn_sword_MSTR}
      boxExtents=<10 10 15>
      constraintSetup {angularTargetAxis=AxisY, twistAxis=AxisY, angularLimitsMax=<1 0 10>, angularLimitsMin=<-1 0 -10>, angularTarget=<0 0 0>}
      localJointOffset=<1 -0.05 0>
     }]
    relativeSpaceBone {boneName=root}
   }
   n71 = Get zipline() -> (zipline:bool)
   n72 = "Transform (Modify) Bone - Bone: dyn_skirt_bk_a_base_l"!(ComponentPose=@n70, bAlphaBoolEnabled:bool=@n71, Translation="0,0,0", Rotation="5,0,0", Scale="1,1,1") -> (Pose) {
    alphaInputType=Bool
    rotationMode=BMM_Additive
    rotationSpace=BCS_BoneSpace
    alphaBoolBlend {blendInTime=0.1, blendOption=Linear, blendOutTime=0.1}
    alphaScaleBiasClamp {bInterpResult=true, interpSpeedDecreasing=2, interpSpeedIncreasing=5}
    boneToModify {boneName=dyn_skirt_bk_a_base_l}
    rotation {pitch=5}
   }
   n73 = "Transform (Modify) Bone - Bone: dyn_skirt_bk_a_base"!(ComponentPose=@n72, bAlphaBoolEnabled:bool=@n71, Translation="0,0,0", Rotation="15,0,0", Scale="1,1,1") -> (Pose) {
    alphaInputType=Bool
    rotationMode=BMM_Additive
    rotationSpace=BCS_BoneSpace
    alphaBoolBlend {blendInTime=0.1, blendOption=Linear, blendOutTime=0.1}
    alphaScaleBiasClamp {bInterpResult=true, interpSpeedDecreasing=2, interpSpeedIncreasing=5}
    boneToModify {boneName=dyn_skirt_bk_a_base}
    rotation {pitch=15}
   }
   n74 = "Transform (Modify) Bone - Bone: dyn_skirt_bk_a_base_r"!(ComponentPose=@n73, bAlphaBoolEnabled:bool=@n71, Translation="0,0,0", Rotation="5,0,0", Scale="1,1,1") -> (Pose) {
    alphaInputType=Bool
    rotationMode=BMM_Additive
    rotationSpace=BCS_BoneSpace
    alphaBoolBlend {blendInTime=0.1, blendOption=Linear, blendOutTime=0.1}
    alphaScaleBiasClamp {bInterpResult=true, interpSpeedDecreasing=2, interpSpeedIncreasing=5}
    boneToModify {boneName=dyn_skirt_bk_a_base_r}
    rotation {pitch=5}
   }
   n75 = "Transform (Modify) Bone - Bone: dyn_rope_a_MSTR"!(ComponentPose=@n74, bAlphaBoolEnabled:bool=True, Translation="0,5,0", Rotation="0,0,25", Scale="1,1,1") -> (Pose) {
    alphaInputType=Bool
    rotationMode=BMM_Additive
    rotationSpace=BCS_BoneSpace
    translationMode=BMM_Additive
    translationSpace=BCS_BoneSpace
    alphaBoolBlend {blendInTime=0.2, blendOption=Linear, blendOutTime=0.2}
    boneToModify {boneName=dyn_rope_a_MSTR}
    rotation {roll=25}
    translation {y=5}
   }
   n76 = "Transform (Modify) Bone - Bone: dyn_rope_b_MSTR"!(ComponentPose=@n75, bAlphaBoolEnabled:bool=True, Translation="0,0,0", Rotation="0,0,25", Scale="1,1,1") -> (Pose) {
    alphaInputType=Bool
    rotationMode=BMM_Additive
    rotationSpace=BCS_BoneSpace
    translationMode=BMM_Additive
    translationSpace=BCS_BoneSpace
    alphaBoolBlend {blendInTime=0.2, blendOption=Linear, blendOutTime=0.2}
    boneToModify {boneName=dyn_rope_b_MSTR}
    rotation {roll=25}
   }
   Output Pose(Result=Component To Local(ComponentPose=@n76)) {layerGroup=None}
   n77 = Get thigh_rot_up_r()
   n78 = "Transform (Modify) Bone - Bone: dyn_rope_a_MSTR"!(
    ComponentPose=Local To Component(LocalPose=Use cached pose 'prerigid'())
    Alpha:float=@n77
    Translation="0,-5,0"
    Rotation="-35.000,-5,-5"
    Scale="1,1,1"
   ) -> (Pose) {
    rotationMode=BMM_Additive
    rotationSpace=BCS_BoneSpace
    translationMode=BMM_Additive
    translationSpace=BCS_BoneSpace
    alphaBoolBlend {blendOption=Linear}
    boneToModify {boneName=dyn_rope_a_MSTR}
    rotation=<rot -35 -5 -5>
    translation {y=-5}
   }
   n79 = "Transform (Modify) Bone - Bone: dyn_rope_b_MSTR"!(ComponentPose=@n78, Alpha:float=@n77, Translation="0,-5,-5", Rotation="-35.000,0,0", Scale="1,1,1") -> (Pose) {
    rotationMode=BMM_Additive
    rotationSpace=BCS_BoneSpace
    translationMode=BMM_Additive
    translationSpace=BCS_BoneSpace
    alphaBoolBlend {blendOption=Linear}
    boneToModify {boneName=dyn_rope_b_MSTR}
    rotation {pitch=-35}
    translation {y=-5, z=-5}
   }
   n80 = RigidBody!(ComponentPose=@n79, Alpha:float, OverrideWorldGravity=Get GravityOverride(), ExternalForce) -> (Pose) {
    bClampLinearTranslationLimitToRefPose=true
    bOverrideWorldGravity=true
    overridePhysicsAsset=&M_MED_Headband_S_Physics
    simulationSpace=BaseBoneSpace
    alphaBoolBlend {blendOption=Linear}
    baseBoneRef {boneName=pelvis}
    componentAppliedLinearAccClamp=<3000 3000 3000>
    componentLinearAccScale=<0.25 0.25 0.4>
    componentLinearVelScale=<0.2 0.2 0.2>
   }
   n81 = RigidBody(ComponentPose=@n80, OverrideWorldGravity=Get GravityOverride()) {
    bClampLinearTranslationLimitToRefPose=true
    bOverrideWorldGravity=true
    overridePhysicsAsset=&/M_MED_Headband_S_body_Physics
    simulationSpace=BaseBoneSpace
    alphaBoolBlend {blendOption=Linear}
    baseBoneRef {boneName=pelvis}
    componentAppliedLinearAccClamp=<3000 3000 3000>
    componentLinearAccScale=<0.25 0.25 0.4>
    componentLinearVelScale=<0.2 0.2 0.2>
   }
   n82 = Get bIsSprinting()
   n83 = "Transform (Modify) Bone - Bone: dyn_skirt_fr_a_base_l"!(ComponentPose=@n81, bAlphaBoolEnabled:bool=@n82, Translation="0,0,0", Rotation="5,0,0", Scale="1,1,1") -> (Pose) {
    alphaInputType=Bool
    rotationMode=BMM_Additive
    rotationSpace=BCS_BoneSpace
    translationMode=BMM_Additive
    translationSpace=BCS_BoneSpace
    alphaBoolBlend {blendInTime=0.1, blendOption=Linear, blendOutTime=0.1}
    boneToModify {boneName=dyn_skirt_fr_a_base_l}
    rotation {pitch=5}
   }
   n84 = "Transform (Modify) Bone - Bone: dyn_skirt_fr_a_base"!(ComponentPose=@n83, bAlphaBoolEnabled:bool=@n82, Translation="0,0,0", Rotation="5,0,0", Scale="1,1,1") -> (Pose) {
    alphaInputType=Bool
    rotationMode=BMM_Additive
    rotationSpace=BCS_BoneSpace
    translationMode=BMM_Additive
    translationSpace=BCS_BoneSpace
    alphaBoolBlend {blendInTime=0.1, blendOption=Linear, blendOutTime=0.1}
    boneToModify {boneName=dyn_skirt_fr_a_base}
    rotation {pitch=5}
   }
   n85 = Component To Local(ComponentPose=@n84)
   n86 = "Transform (Modify) Bone - Bone: dyn_knot_MSTR"(
    ComponentPose=Local To Component(LocalPose=Use cached pose 'prerigid'())
    bAlphaBoolEnabled=True
    Translation="-2,10,0.776"
    Rotation="2.676,55,31.882"
    Scale="1,1,1"
   ) {
    alphaInputType=Bool
    rotationMode=BMM_Additive
    rotationSpace=BCS_BoneSpace
    translationMode=BMM_Additive
    translationSpace=BCS_BoneSpace
    alphaBoolBlend {blendInTime=0.1, blendOption=Linear, blendOutTime=0.1}
    boneToModify {boneName=dyn_knot_MSTR}
    rotation=<rot 2.676 55 31.882>
    translation=<-2 10 0.776>
   }
   n87 = Get thigh_rot_up_r()
   n88 = "Transform (Modify) Bone - Bone: dyn_rope_a_MSTR"(ComponentPose=@n86, Alpha=@n87, Translation="0,0,0", Rotation="-35,0,15", Scale="1,1,1") {
    rotationMode=BMM_Additive
    rotationSpace=BCS_BoneSpace
    translationMode=BMM_Additive
    translationSpace=BCS_BoneSpace
    alphaBoolBlend {blendOption=Linear}
    alphaScaleBiasClamp {bInterpResult=true, interpSpeedDecreasing=40, interpSpeedIncreasing=40}
    boneToModify {boneName=dyn_rope_a_MSTR}
    rotation {pitch=-35, roll=15}
   }
   n89 = "Transform (Modify) Bone - Bone: dyn_loop_a_MSTR"(ComponentPose=@n88, Alpha=@n87, Translation="0,0,0", Rotation="39.071,44.560,-20.113", Scale="1,1,1") {
    rotationMode=BMM_Additive
    rotationSpace=BCS_BoneSpace
    translationMode=BMM_Additive
    translationSpace=BCS_BoneSpace
    alphaBoolBlend {blendOption=Linear}
    alphaScaleBiasClamp {bInterpResult=true, interpSpeedDecreasing=40, interpSpeedIncreasing=40}
    boneToModify {boneName=dyn_loop_a_MSTR}
    rotation=<rot 39.0709 44.5596 -20.113>
   }
   n90 = "Transform (Modify) Bone - Bone: dyn_loop_b_MSTR"(ComponentPose=@n89, Alpha=@n87, Translation="0,0,0", Rotation="35,35,-35", Scale="1,1,1") {
    rotationMode=BMM_Additive
    rotationSpace=BCS_BoneSpace
    translationMode=BMM_Additive
    translationSpace=BCS_BoneSpace
    alphaBoolBlend {blendOption=Linear}
    alphaScaleBiasClamp {bInterpResult=true, interpSpeedDecreasing=40, interpSpeedIncreasing=40}
    boneToModify {boneName=dyn_loop_b_MSTR}
    rotation=<rot 35 35 -35>
   }
   n91 = RigidBody(ComponentPose=@n90, OverrideWorldGravity=Get GravityOverride()) {
    bClampLinearTranslationLimitToRefPose=true
    bOverrideWorldGravity=true
    overridePhysicsAsset=&/M_MED_Headband_S_crouch_Physics
    simulationSpace=BaseBoneSpace
    alphaBoolBlend {blendOption=Linear}
    baseBoneRef {boneName=pelvis}
    componentAppliedLinearAccClamp=<3000 3000 3000>
    componentLinearAccScale=<0.25 0.25 0.4>
    componentLinearVelScale=<0.2 0.2 0.2>
   }
   n92 = RigidBody(ComponentPose=@n91, OverrideWorldGravity=Get GravityOverride()) {
    bClampLinearTranslationLimitToRefPose=true
    bOverrideWorldGravity=true
    overridePhysicsAsset=&M_MED_Headband_S_body_Physics
    simulationSpace=BaseBoneSpace
    alphaBoolBlend {blendOption=Linear}
    baseBoneRef {boneName=pelvis}
    componentAppliedLinearAccClamp=<3000 3000 3000>
    componentLinearAccScale=<0.25 0.25 0.4>
    componentLinearVelScale=<0.2 0.2 0.2>
   }
   n93 = Get bIsMoving2D()
   n94 = "Transform (Modify) Bone - Bone: dyn_skirt_fr_a_base"(ComponentPose=@n92, bAlphaBoolEnabled=@n93, Translation="-2,-2,5", Rotation="0,0,0", Scale="1,1,1") {
    alphaInputType=Bool
    rotationMode=BMM_Additive
    rotationSpace=BCS_BoneSpace
    translationMode=BMM_Additive
    translationSpace=BCS_BoneSpace
    alphaBoolBlend {blendInTime=0.1, blendOption=Linear, blendOutTime=0.1}
    boneToModify {boneName=dyn_skirt_fr_a_base}
    rotation=<rot 0 0 0>
    translation=<-2 -2 5>
   }
   n95 = "Transform (Modify) Bone - Bone: dyn_skirt_fr_a_base_r"!(ComponentPose=@n94, bAlphaBoolEnabled:bool=@n93, Translation="2,-5,-2", Rotation="0,0,0", Scale="1,1,1") -> (Pose) {
    alphaInputType=Bool
    rotationMode=BMM_Additive
    rotationSpace=BCS_BoneSpace
    translationMode=BMM_Additive
    translationSpace=BCS_BoneSpace
    alphaBoolBlend {blendInTime=0.1, blendOption=Linear, blendOutTime=0.1}
    boneToModify {boneName=dyn_skirt_fr_a_base_r}
    translation=<2 -5 -2>
   }
   n96 = Component To Local(ComponentPose=@n95)
   n97 = Get bIsCrouching()
   n98 = Get thigh_rot_up_r()
   n99 = "Transform (Modify) Bone - Bone: dyn_rope_a_MSTR"(
    ComponentPose=Local To Component(LocalPose=Use cached pose 'prerigid'())
    Alpha=@n98
    Translation="0,-5,0"
    Rotation="-35.000,-5,-5"
    Scale="1,1,1"
   ) {
    rotationMode=BMM_Additive
    rotationSpace=BCS_BoneSpace
    translationMode=BMM_Additive
    translationSpace=BCS_BoneSpace
    alphaBoolBlend {blendOption=Linear}
    boneToModify {boneName=dyn_rope_a_MSTR}
    rotation=<rot -35 -5 -5>
    translation {y=-5}
   }
   n100 = "Transform (Modify) Bone - Bone: dyn_rope_b_MSTR"(ComponentPose=@n99, Alpha=@n98, Translation="0,-5,-5", Rotation="-35.000,0,0", Scale="1,1,1") {
    rotationMode=BMM_Additive
    rotationSpace=BCS_BoneSpace
    translationMode=BMM_Additive
    translationSpace=BCS_BoneSpace
    alphaBoolBlend {blendOption=Linear}
    boneToModify {boneName=dyn_rope_b_MSTR}
    rotation {pitch=-35}
    translation {y=-5, z=-5}
   }
   n101 = RigidBody(ComponentPose=@n100, OverrideWorldGravity=Get GravityOverride()) {
    bClampLinearTranslationLimitToRefPose=true
    bOverrideWorldGravity=true
    overridePhysicsAsset=&/M_MED_Headband_S_FrontEnd_skirt_Physics
    simulationSpace=BaseBoneSpace
    alphaBoolBlend {blendOption=Linear}
    baseBoneRef {boneName=dyn_root_simspace}
    componentAppliedLinearAccClamp=<3000 3000 3000>
    componentLinearAccScale=<0.25 0.25 0.4>
    componentLinearVelScale=<0.2 0.2 0.2>
   }
   n102 = RigidBody(ComponentPose=@n101, OverrideWorldGravity=Get GravityOverride()) {
    bClampLinearTranslationLimitToRefPose=true
    bOverrideWorldGravity=true
    overridePhysicsAsset=&/M_MED_Headband_S_frontend_Physics
    simulationSpace=BaseBoneSpace
    alphaBoolBlend {blendOption=Linear}
    baseBoneRef {boneName=dyn_root_simspace}
    componentAppliedLinearAccClamp=<3000 3000 3000>
    componentLinearAccScale=<0.25 0.25 0.4>
    componentLinearVelScale=<0.2 0.2 0.2>
   }
   n103 = Get thigh_rot_up_r()
   n104 = "Transform (Modify) Bone - Bone: dyn_rope_a_MSTR"(
    ComponentPose=Local To Component(LocalPose=Use cached pose 'prerigid'())
    Alpha=@n103
    Translation="0,-5,0"
    Rotation="-35.000,-5,-5"
    Scale="1,1,1"
   ) {
    rotationMode=BMM_Additive
    rotationSpace=BCS_BoneSpace
    translationMode=BMM_Additive
    translationSpace=BCS_BoneSpace
    alphaBoolBlend {blendOption=Linear}
    boneToModify {boneName=dyn_rope_a_MSTR}
    rotation=<rot -35 -5 -5>
    translation {y=-5}
   }
   n105 = "Transform (Modify) Bone - Bone: dyn_rope_b_MSTR"(ComponentPose=@n104, Alpha=@n103, Translation="0,-5,-5", Rotation="-35.000,0,0", Scale="1,1,1") {
    rotationMode=BMM_Additive
    rotationSpace=BCS_BoneSpace
    translationMode=BMM_Additive
    translationSpace=BCS_BoneSpace
    alphaBoolBlend {blendOption=Linear}
    boneToModify {boneName=dyn_rope_b_MSTR}
    rotation {pitch=-35}
    translation {y=-5, z=-5}
   }
   n106 = Get thigh_rot_extra_r() -> (thigh_rot_extra_r:double)
   n107 = "Transform (Modify) Bone - Bone: dyn_rope_a_MSTR"(ComponentPose=@n105, Alpha=@n106, Translation="-0.773,-10.000,-7.406", Rotation="0,0,0", Scale="1,1,1") {
    rotationMode=BMM_Additive
    rotationSpace=BCS_BoneSpace
    translationMode=BMM_Additive
    translationSpace=BCS_BoneSpace
    alphaBoolBlend {blendOption=Linear}
    boneToModify {boneName=dyn_rope_a_MSTR}
    translation=<-0.773 -10 -7.406>
   }
   n108 = "Transform (Modify) Bone - Bone: dyn_rope_b_MSTR"(ComponentPose=@n107, Alpha=@n106, Translation="-6.057,-10.000,-12.376", Rotation="0,0,0", Scale="1,1,1") {
    rotationMode=BMM_Additive
    rotationSpace=BCS_BoneSpace
    translationMode=BMM_Additive
    translationSpace=BCS_BoneSpace
    alphaBoolBlend {blendOption=Linear}
    boneToModify {boneName=dyn_rope_b_MSTR}
    translation=<-6.0566 -10 -12.3761>
   }
   n109 = RigidBody(ComponentPose=@n108, OverrideWorldGravity=Get GravityOverride()) {
    bClampLinearTranslationLimitToRefPose=true
    bOverrideWorldGravity=true
    overridePhysicsAsset=&M_MED_Headband_S_crouch_Physics
    simulationSpace=BaseBoneSpace
    alphaBoolBlend {blendOption=Linear}
    baseBoneRef {boneName=pelvis}
    componentAppliedLinearAccClamp=<3000 3000 3000>
    componentLinearAccScale=<0.25 0.25 0.4>
    componentLinearVelScale=<0.2 0.2 0.2>
   }
   n110 = RigidBody(ComponentPose=@n109, OverrideWorldGravity=Get GravityOverride()) {
    bClampLinearTranslationLimitToRefPose=true
    bOverrideWorldGravity=true
    overridePhysicsAsset=&M_MED_Headband_S_body_Physics
    simulationSpace=BaseBoneSpace
    alphaBoolBlend {blendOption=Linear}
    baseBoneRef {boneName=pelvis}
    componentAppliedLinearAccClamp=<3000 3000 3000>
    componentLinearAccScale=<0.25 0.25 0.4>
    componentLinearVelScale=<0.2 0.2 0.2>
   }
   n111 = Blend Poses by bool(
    BlendPose_0=Component To Local(ComponentPose=@n110)
    BlendPose_1=Blend Poses by bool(
     BlendPose_0=Component To Local(ComponentPose=@n102)
     BlendPose_1=Blend Poses by bool(BlendPose_0=@n96, BlendPose_1=@n85, bActiveValue=@n97) {blendType=Linear}
     bActiveValue=Get bIsInFrontEnd()
    ) {blendType=Linear}
    BlendTime_1=0.5
    bActiveValue=Get bIsPlayingEmote()
   ) {blendType=Linear}
   n112 = Use cached pose 'no rigids'()
   n113 = Get bRigidBodyEnabled()
   n114 = "Constraint - dyn_skirt_bk_a_01 (PP)"!(
    ComponentPose=Local To Component(LocalPose=Blend Poses by bool(BlendPose_0=@n111, BlendPose_1=@n112, bActiveValue=@n113) {blendType=Linear})
    Alpha=1
    ConstraintWeights_0=0.5
    ConstraintWeights_1:float=0.5
   ) -> (Pose) {
    alphaBoolBlend {blendOption=Linear}
    boneToModify {boneName=dyn_skirt_bk_a_01}
    constraintSetup [{targetBone {boneName=dyn_skirt_bk_a_l}}, {targetBone {boneName=dyn_skirt_bk_a_r}}]
    constraintWeights [1,1]
   }
   n115 = "Constraint - dyn_skirt_bk_a_02 (PP)"!(ComponentPose=@n114, Alpha=1, ConstraintWeights_0=0.5, ConstraintWeights_1:float=0.5) -> (Pose) {
    alphaBoolBlend {blendOption=Linear}
    boneToModify {boneName=dyn_skirt_bk_a_02}
    constraintSetup [{targetBone {boneName=dyn_skirt_bk_b_l}}, {targetBone {boneName=dyn_skirt_bk_b_r}}]
    constraintWeights [1,1]
   }
   n116 = "Constraint - dyn_skirt_bk_a_03 (PP)"!(ComponentPose=@n115, Alpha=1, ConstraintWeights_0=0.5, ConstraintWeights_1:float=0.5) -> (Pose) {
    alphaBoolBlend {blendOption=Linear}
    boneToModify {boneName=dyn_skirt_bk_a_03}
    constraintSetup [{targetBone {boneName=dyn_skirt_bk_b_l_01}}, {targetBone {boneName=dyn_skirt_bk_b_r_01}}]
    constraintWeights [1,1]
   }
   n117 = "Constraint - dyn_skirt_bk_a_04 (PP)"!(ComponentPose=@n116, Alpha=1, ConstraintWeights_0=0.5, ConstraintWeights_1:float=0.5) -> (Pose) {
    alphaBoolBlend {blendOption=Linear}
    boneToModify {boneName=dyn_skirt_bk_a_04}
    constraintSetup [{targetBone {boneName=dyn_skirt_bk_b_l_02}}, {targetBone {boneName=dyn_skirt_bk_b_r_02}}]
    constraintWeights [1,1]
   }
   n118 = "Constraint - dyn_skirt_bk_a_05 (PP)"!(ComponentPose=@n117, Alpha=1, ConstraintWeights_0=0.5, ConstraintWeights_1:float=0.5) -> (Pose) {
    alphaBoolBlend {blendOption=Linear}
    boneToModify {boneName=dyn_skirt_bk_a_05}
    constraintSetup [{targetBone {boneName=dyn_skirt_bk_b_l_03}}, {targetBone {boneName=dyn_skirt_bk_b_r_03}}]
    constraintWeights [1,1]
   }
   n119 = "Transform (Modify) Bone - Bone: dyn_skirt_fr_a_r"!(ComponentPose=@n118, bAlphaBoolEnabled:bool=Get bIsInFrontEnd(), Translation="2,-5,-7", Rotation="-10,0,0", Scale="1,1,1") -> (Pose) {
    alphaInputType=Bool
    rotationMode=BMM_Additive
    rotationSpace=BCS_BoneSpace
    translationMode=BMM_Additive
    translationSpace=BCS_BoneSpace
    alphaBoolBlend {blendOption=Linear}
    boneToModify {boneName=dyn_skirt_fr_a_r}
    rotation {pitch=-10}
    translation=<2 -5 -7>
   }
   Save cached pose 'skirt_bk_a_pose'(Pose=Component To Local(ComponentPose=@n119)) {cacheName=skirt_bk_a_pose}
   n120 = "Constraint - dyn_knot_MSTR (PPPPPP)"!(
    ComponentPose=Local To Component(LocalPose=Copy Pose From Mesh() {bCopyCurves=true, bUseAttachedParent=true})
    Alpha=1
    ConstraintWeights_0=0.0068
    ConstraintWeights_1:float=0.3072
    ConstraintWeights_2:float=0.1486
    ConstraintWeights_3:float=0.0914
    ConstraintWeights_4:float=0.0161
    ConstraintWeights_5:float=0.4299
   ) -> (Pose) {
    alphaBoolBlend {blendOption=Linear}
    boneToModify {boneName=dyn_knot_MSTR}
    constraintSetup [
     {targetBone {boneName=thigh_twist_01_r}}
     {targetBone {boneName=spine_01}}
     {targetBone {boneName=spine_02}}
     {targetBone {boneName=spine_03}}
     {targetBone {boneName=spine_04}}
     {targetBone {boneName=pelvis}}
    ]
    constraintWeights [1,1,1,1,1,1]
   }
   n121 = "Constraint - dyn_sword_MSTR (PPPPPP)"!(
    ComponentPose=@n120
    Alpha=1
    ConstraintWeights_0=0.0002
    ConstraintWeights_1:float=0.2246
    ConstraintWeights_2:float=0.014
    ConstraintWeights_3:float=0.018
    ConstraintWeights_4:float=0.7394
    ConstraintWeights_5:float=0.0037
   ) -> (Pose) {
    alphaBoolBlend {blendOption=Linear}
    boneToModify {boneName=dyn_sword_MSTR}
    constraintSetup [
     {targetBone {boneName=thigh_twist_01_r}}
     {targetBone {boneName=spine_01}}
     {targetBone {boneName=spine_02}}
     {targetBone {boneName=spine_03}}
     {targetBone {boneName=pelvis}}
     {targetBone {boneName=thigh_twist_01_l}}
    ]
    constraintWeights [1,1,1,1,1,1]
   }
   n122 = "Constraint - dyn_collar_bk_MSTR (PPPPPPPP)"!(
    ComponentPose=@n121
    Alpha=1
    ConstraintWeights_0=0.0361
    ConstraintWeights_1:float=0.0225
    ConstraintWeights_2:float=0.61
    ConstraintWeights_3:float=0.0681
    ConstraintWeights_4:float=0.1407
    ConstraintWeights_5:float=0.0584
    ConstraintWeights_6:float=0.0339
    ConstraintWeights_7:float=0.0303
   ) -> (Pose) {
    alphaBoolBlend {blendOption=Linear}
    boneToModify {boneName=dyn_collar_bk_MSTR}
    constraintSetup [
     {targetBone {boneName=spine_03}}
     {targetBone {boneName=spine_04}}
     {targetBone {boneName=spine_05}}
     {targetBone {boneName=neck_01}}
     {targetBone {boneName=neck_02}}
     {targetBone {boneName=head}}
     {targetBone {boneName=clavicle_l}}
     {targetBone {boneName=clavicle_r}}
    ]
    constraintWeights [1,1,1,1,1,1,1,1]
   }
   n123 = "Constraint - dyn_collar_bk_r_MSTR (PPPPPPPP)"!(
    ComponentPose=@n122
    Alpha=1
    ConstraintWeights_0=0.0311
    ConstraintWeights_1:float=0.0263
    ConstraintWeights_2:float=0.5589
    ConstraintWeights_3:float=0.0582
    ConstraintWeights_4:float=0.1239
    ConstraintWeights_5:float=0.1107
    ConstraintWeights_6:float=0.0224
    ConstraintWeights_7:float=0.0686
   ) -> (Pose) {
    alphaBoolBlend {blendOption=Linear}
    boneToModify {boneName=dyn_collar_bk_r_MSTR}
    constraintSetup [
     {targetBone {boneName=spine_03}}
     {targetBone {boneName=spine_04}}
     {targetBone {boneName=spine_05}}
     {targetBone {boneName=neck_01}}
     {targetBone {boneName=neck_02}}
     {targetBone {boneName=head}}
     {targetBone {boneName=clavicle_l}}
     {targetBone {boneName=clavicle_r}}
    ]
    constraintWeights [1,1,1,1,1,1,1,1]
   }
   n124 = "Constraint - dyn_collar_l_MSTR (PPPPPPPP)"!(
    ComponentPose=@n123
    Alpha=1
    ConstraintWeights_0=0.0271
    ConstraintWeights_1:float=0.0553
    ConstraintWeights_2:float=0.6096
    ConstraintWeights_3:float=0.0553
    ConstraintWeights_4:float=0.1162
    ConstraintWeights_5:float=0.0546
    ConstraintWeights_6:float=0.065
    ConstraintWeights_7:float=0.0169
   ) -> (Pose) {
    alphaBoolBlend {blendOption=Linear}
    boneToModify {boneName=dyn_collar_l_MSTR}
    constraintSetup [
     {targetBone {boneName=spine_03}}
     {targetBone {boneName=spine_04}}
     {targetBone {boneName=spine_05}}
     {targetBone {boneName=neck_01}}
     {targetBone {boneName=neck_02}}
     {targetBone {boneName=head}}
     {targetBone {boneName=clavicle_l}}
     {targetBone {boneName=clavicle_r}}
    ]
    constraintWeights [1,1,1,1,1,1,1,1]
   }
   n125 = "Constraint - dyn_collar_r_MSTR (PPPPPPPP)"!(
    ComponentPose=@n124
    Alpha=1
    ConstraintWeights_0=0.0246
    ConstraintWeights_1:float=0.0456
    ConstraintWeights_2:float=0.5193
    ConstraintWeights_3:float=0.0468
    ConstraintWeights_4:float=0.0999
    ConstraintWeights_5:float=0.1972
    ConstraintWeights_6:float=0.0213
    ConstraintWeights_7:float=0.0452
   ) -> (Pose) {
    alphaBoolBlend {blendOption=Linear}
    boneToModify {boneName=dyn_collar_r_MSTR}
    constraintSetup [
     {targetBone {boneName=spine_03}}
     {targetBone {boneName=spine_04}}
     {targetBone {boneName=spine_05}}
     {targetBone {boneName=neck_01}}
     {targetBone {boneName=neck_02}}
     {targetBone {boneName=head}}
     {targetBone {boneName=clavicle_l}}
     {targetBone {boneName=clavicle_r}}
    ]
    constraintWeights [1,1,1,1,1,1,1,1]
   }
   n126 = "Constraint - dyn_collar_bk_l_MSTR (PPPPPPPP)"!(
    ComponentPose=@n125
    Alpha=1
    ConstraintWeights_0=0.03
    ConstraintWeights_1:float=0.0301
    ConstraintWeights_2:float=0.5899
    ConstraintWeights_3:float=0.0635
    ConstraintWeights_4:float=0.1365
    ConstraintWeights_5:float=0.053
    ConstraintWeights_6:float=0.0785
    ConstraintWeights_7:float=0.0184
   ) -> (Pose) {
    alphaBoolBlend {blendOption=Linear}
    boneToModify {boneName=dyn_collar_bk_l_MSTR}
    constraintSetup [
     {targetBone {boneName=spine_03}}
     {targetBone {boneName=spine_04}}
     {targetBone {boneName=spine_05}}
     {targetBone {boneName=neck_01}}
     {targetBone {boneName=neck_02}}
     {targetBone {boneName=head}}
     {targetBone {boneName=clavicle_l}}
     {targetBone {boneName=clavicle_r}}
    ]
    constraintWeights [1,1,1,1,1,1,1,1]
   }
   n127 = "Constraint - dyn_zipper_mstr (PPPPPP)"!(
    ComponentPose=@n126
    Alpha=1
    ConstraintWeights_0=0.0243
    ConstraintWeights_1:float=0.0873
    ConstraintWeights_2:float=0.6847
    ConstraintWeights_3:float=0.0528
    ConstraintWeights_4:float=0.1059
    ConstraintWeights_5:float=0.0449
   ) -> (Pose) {
    alphaBoolBlend {blendOption=Linear}
    boneToModify {boneName=dyn_zipper_mstr}
    constraintSetup [
     {targetBone {boneName=spine_03}}
     {targetBone {boneName=spine_04}}
     {targetBone {boneName=spine_05}}
     {targetBone {boneName=neck_01}}
     {targetBone {boneName=neck_02}}
     {targetBone {boneName=head}}
    ]
    constraintWeights [1,1,1,1,1,1]
   }
   n128 = Control Rig!(Source=Component To Local(ComponentPose=@n127), Alpha:float, targeting:double=Get targeting() -> (targeting:double)) -> (Pose) {
    alphaBoolBlend {blendOption=Linear}
    controlRigAssetReference {blueprintRigClass=&/M_MED_Headband_S_body_CtrlRig.M_MED_Headband_S_body_CtrlRig_C}
   }
   n129 = Control Rig!(
    Source=@n128
    bAlphaBoolEnabled:bool=True
    emote:double=Get emote() -> (emote:double)
    crouching:double=Get crouchingAlpha() -> (crouchingAlpha:double)
   ) -> (Pose) {
    alphaInputType=Bool
    alphaBoolBlend {blendInTime=0.1, blendOption=Linear, blendOutTime=0.1}
    controlRigAssetReference {blueprintRigClass=&M_MED_Headband_S_target_CtrlRig_C}
   }
   Save cached pose 'constraints'(Pose=@n129) {cacheName=constraints}
   n130 = Look At - Bone: dyn_skirt_bk_b_r_01(ComponentPose=Local To Component(LocalPose=Use cached pose 'prerigid'()), Alpha=1) {
    alphaBoolBlend {blendOption=Linear}
    boneToModify {boneName=dyn_skirt_bk_b_r_01}
    lookAtLocation=<0 15>
    lookAtTarget {boneReference {boneName=foot_r}}
    lookAt_Axis {axis=<1 0>}
   }
   n131 = Look At - Bone: dyn_skirt_bk_b_l_01(ComponentPose=@n130, Alpha=1) {
    alphaBoolBlend {blendOption=Linear}
    boneToModify {boneName=dyn_skirt_bk_b_l_01}
    lookAtLocation=<0 -15>
    lookAtTarget {boneReference {boneName=foot_l}}
    lookAt_Axis {axis=<1 0>}
   }
   n132 = Get thigh_rot_r()
   n133 = "Transform (Modify) Bone - Bone: dyn_skirt_fr_a_r"!(ComponentPose=@n131, Alpha:float=@n132, Translation="15,0,-5", Rotation="-15,0,0", Scale="1,1,1") -> (Pose) {
    rotationMode=BMM_Additive
    rotationSpace=BCS_BoneSpace
    translationMode=BMM_Additive
    translationSpace=BCS_BoneSpace
    alphaBoolBlend {blendOption=Linear}
    boneToModify {boneName=dyn_skirt_fr_a_r}
    rotation {pitch=-15}
    translation {x=15, z=-5}
   }
   n134 = "Transform (Modify) Bone - Bone: dyn_skirt_side_a_r"(ComponentPose=@n133, Alpha=@n132, Translation="-10,0,0", Rotation="-5,0,0", Scale="1,1,1") {
    rotationMode=BMM_Additive
    rotationSpace=BCS_BoneSpace
    translationMode=BMM_Additive
    translationSpace=BCS_BoneSpace
    alphaBoolBlend {blendOption=Linear}
    boneToModify {boneName=dyn_skirt_side_a_r}
    rotation {pitch=-5}
    translation {x=-10}
   }
   n135 = Get thigh_rot_l()
   n136 = "Transform (Modify) Bone - Bone: dyn_skirt_fr_a_l"(ComponentPose=@n134, Alpha=@n135, Translation="-15,0,5", Rotation="-15,0,0", Scale="1,1,1") {
    rotationMode=BMM_Additive
    rotationSpace=BCS_BoneSpace
    translationMode=BMM_Additive
    translationSpace=BCS_BoneSpace
    alphaBoolBlend {blendOption=Linear}
    boneToModify {boneName=dyn_skirt_fr_a_l}
    rotation {pitch=-15}
    translation {x=-15, z=5}
   }
   n137 = "Transform (Modify) Bone - Bone: dyn_skirt_side_a_l"(ComponentPose=@n136, Alpha=@n135, Translation="-10,0,0", Rotation="-5,0,0", Scale="1,1,1") {
    rotationMode=BMM_Additive
    rotationSpace=BCS_BoneSpace
    translationMode=BMM_Additive
    translationSpace=BCS_BoneSpace
    alphaBoolBlend {blendOption=Linear}
    boneToModify {boneName=dyn_skirt_side_a_l}
    rotation {pitch=-5}
    translation {x=-10}
   }
   n138 = Get isNotCrouchRunning() -> (isNotCrouchRunning:double)
   n139 = "Transform (Modify) Bone - Bone: dyn_skirt_bk_a_l"(ComponentPose=@n137, Alpha=@n138, Translation="0,0,0", Rotation="10,0,0.000", Scale="1,1,1") {
    rotationMode=BMM_Additive
    rotationSpace=BCS_BoneSpace
    alphaBoolBlend {blendOption=Linear}
    alphaScaleBiasClamp {bInterpResult=true, interpSpeedDecreasing=5}
    boneToModify {boneName=dyn_skirt_bk_a_l}
    rotation {pitch=10}
   }
   n140 = "Transform (Modify) Bone - Bone: dyn_skirt_bk_a_r"(ComponentPose=@n139, Alpha=@n138, Translation="0,0,0", Rotation="10,0,0.000", Scale="1,1,1") {
    rotationMode=BMM_Additive
    rotationSpace=BCS_BoneSpace
    alphaBoolBlend {blendOption=Linear}
    alphaScaleBiasClamp {bInterpResult=true, interpSpeedDecreasing=5}
    boneToModify {boneName=dyn_skirt_bk_a_r}
    rotation {pitch=10}
   }
   n141 = "Transform (Modify) Bone - Bone: dyn_skirt_bk_a_01"(ComponentPose=@n140, Alpha=@n138, Translation="0,0,1", Rotation="25,0,0.000", Scale="1,1,1") {
    rotationMode=BMM_Additive
    rotationSpace=BCS_BoneSpace
    alphaBoolBlend {blendOption=Linear}
    alphaScaleBiasClamp {bInterpResult=true, interpSpeedDecreasing=5}
    boneToModify {boneName=dyn_skirt_bk_a_01}
    rotation {pitch=25}
    translation {z=1}
   }
   n142 = "Transform (Modify) Bone - Bone: dyn_skirt_bk_a_02"(ComponentPose=@n141, Alpha=@n138, Translation="0,0,0", Rotation="15,0,0.000", Scale="1,1,1") {
    rotationMode=BMM_Additive
    rotationSpace=BCS_BoneSpace
    alphaBoolBlend {blendOption=Linear}
    alphaScaleBiasClamp {bInterpResult=true, interpSpeedDecreasing=5}
    boneToModify {boneName=dyn_skirt_bk_a_02}
    rotation {pitch=15}
   }
   n143 = "Transform (Modify) Bone - Bone: dyn_skirt_bk_a_03"(ComponentPose=@n142, Alpha=@n138, Translation="0,0,0", Rotation="25,0,0.000", Scale="1,1,1") {
    rotationMode=BMM_Additive
    rotationSpace=BCS_BoneSpace
    alphaBoolBlend {blendOption=Linear}
    alphaScaleBiasClamp {bInterpResult=true, interpSpeedDecreasing=5}
    boneToModify {boneName=dyn_skirt_bk_a_03}
    rotation {pitch=25}
   }
   n144 = "Transform (Modify) Bone - Bone: dyn_skirt_bk_b_r"(ComponentPose=@n143, Alpha=@n138, Translation="0,0,1", Rotation="15,10,0.000", Scale="1,1,1") {
    rotationMode=BMM_Additive
    rotationSpace=BCS_BoneSpace
    alphaBoolBlend {blendOption=Linear}
    alphaScaleBiasClamp {bInterpResult=true, interpSpeedDecreasing=5}
    boneToModify {boneName=dyn_skirt_bk_b_r}
    rotation {pitch=15, yaw=10}
    translation {z=1}
   }
   n145 = "Transform (Modify) Bone - Bone: dyn_skirt_bk_b_r_01"(ComponentPose=@n144, Alpha=@n138, Translation="0,0,0", Rotation="-15,0,0.000", Scale="1,1,1") {
    rotationMode=BMM_Additive
    rotationSpace=BCS_BoneSpace
    alphaBoolBlend {blendOption=Linear}
    alphaScaleBiasClamp {bInterpResult=true, interpSpeedDecreasing=5}
    boneToModify {boneName=dyn_skirt_bk_b_r_01}
    rotation {pitch=-15}
   }
   n146 = "Transform (Modify) Bone - Bone: dyn_skirt_bk_b_r_03"(ComponentPose=@n145, Alpha=@n138, Translation="0,0,0", Rotation="-25,0,0.000", Scale="1,1,1") {
    rotationMode=BMM_Additive
    rotationSpace=BCS_BoneSpace
    alphaBoolBlend {blendOption=Linear}
    alphaScaleBiasClamp {bInterpResult=true, interpSpeedDecreasing=5}
    boneToModify {boneName=dyn_skirt_bk_b_r_03}
    rotation {pitch=-25}
   }
   n147 = Component To Local(ComponentPose=@n146)
   n148 = Look At - Bone: dyn_skirt_bk_b_r_01(ComponentPose=Local To Component(LocalPose=Use cached pose 'prerigid'()), Alpha=1) {
    alphaBoolBlend {blendOption=Linear}
    boneToModify {boneName=dyn_skirt_bk_b_r_01}
    lookAtLocation=<0 15>
    lookAtTarget {boneReference {boneName=foot_r}}
    lookAt_Axis {axis=<1 0>}
   }
   n149 = Look At - Bone: dyn_skirt_bk_b_l_01(ComponentPose=@n148, Alpha=1) {
    alphaBoolBlend {blendOption=Linear}
    boneToModify {boneName=dyn_skirt_bk_b_l_01}
    lookAtLocation=<0 -15>
    lookAtTarget {boneReference {boneName=foot_l}}
    lookAt_Axis {axis=<1 0>}
   }
   n150 = Get bIsMoving2D()
   n151 = "Transform (Modify) Bone - Bone: dyn_skirt_bk_a_l"!(ComponentPose=@n149, bAlphaBoolEnabled:bool=@n150, Translation="0,0,0", Rotation="20,5,0.000", Scale="1,1,1") -> (Pose) {
    alphaInputType=Bool
    rotationMode=BMM_Additive
    rotationSpace=BCS_BoneSpace
    alphaBoolBlend {blendInTime=0.2, blendOption=Linear, blendOutTime=0.2}
    alphaScaleBiasClamp {bInterpResult=true, interpSpeedDecreasing=2, interpSpeedIncreasing=5}
    boneToModify {boneName=dyn_skirt_bk_a_l}
    rotation {pitch=20, yaw=5}
   }
   n152 = "Transform (Modify) Bone - Bone: dyn_skirt_bk_a_r"!(ComponentPose=@n151, bAlphaBoolEnabled:bool=@n150, Translation="0,0,0", Rotation="20,-5,0.000", Scale="1,1,1") -> (Pose) {
    alphaInputType=Bool
    rotationMode=BMM_Additive
    rotationSpace=BCS_BoneSpace
    alphaBoolBlend {blendInTime=0.2, blendOption=Linear, blendOutTime=0.2}
    alphaScaleBiasClamp {bInterpResult=true, interpSpeedDecreasing=2, interpSpeedIncreasing=5}
    boneToModify {boneName=dyn_skirt_bk_a_r}
    rotation {pitch=20, yaw=-5}
   }
   n153 = "Transform (Modify) Bone - Bone: dyn_skirt_bk_a"!(ComponentPose=@n152, bAlphaBoolEnabled:bool=@n150, Translation="0,0,0", Rotation="20,0,0.000", Scale="1,1,1") -> (Pose) {
    alphaInputType=Bool
    rotationMode=BMM_Additive
    rotationSpace=BCS_BoneSpace
    alphaBoolBlend {blendInTime=0.2, blendOption=Linear, blendOutTime=0.2}
    alphaScaleBiasClamp {bInterpResult=true, interpSpeedDecreasing=2, interpSpeedIncreasing=5}
    boneToModify {boneName=dyn_skirt_bk_a}
    rotation {pitch=20}
   }
   n154 = Get jumpingAlpha() -> (jumpingAlpha:double)
   n155 = "Transform (Modify) Bone - Bone: dyn_skirt_bk_a_l"!(ComponentPose=@n153, Alpha:float=@n154, Translation="0,0,0", Rotation="35,10,0.000", Scale="1,1,1") -> (Pose) {
    rotationMode=BMM_Additive
    rotationSpace=BCS_BoneSpace
    alphaBoolBlend {blendOption=Linear}
    alphaScaleBiasClamp {bInterpResult=true, interpSpeedDecreasing=2, interpSpeedIncreasing=5}
    boneToModify {boneName=dyn_skirt_bk_a_l}
    rotation {pitch=35, yaw=10}
   }
   n156 = "Transform (Modify) Bone - Bone: dyn_skirt_bk_a_r"!(ComponentPose=@n155, Alpha:float=@n154, Translation="0,0,0", Rotation="35,-10,0.000", Scale="1,1,1") -> (Pose) {
    rotationMode=BMM_Additive
    rotationSpace=BCS_BoneSpace
    alphaBoolBlend {blendOption=Linear}
    alphaScaleBiasClamp {bInterpResult=true, interpSpeedDecreasing=2, interpSpeedIncreasing=5}
    boneToModify {boneName=dyn_skirt_bk_a_r}
    rotation {pitch=35, yaw=-10}
   }
   n157 = "Transform (Modify) Bone - Bone: dyn_skirt_bk_a"!(ComponentPose=@n156, Alpha:float=@n154, Translation="0,0,0", Rotation="35,0,0.000", Scale="1,1,1") -> (Pose) {
    rotationMode=BMM_Additive
    rotationSpace=BCS_BoneSpace
    alphaBoolBlend {blendOption=Linear}
    alphaScaleBiasClamp {bInterpResult=true, interpSpeedDecreasing=2, interpSpeedIncreasing=5}
    boneToModify {boneName=dyn_skirt_bk_a}
    rotation {pitch=35}
   }
   n158 = Get skydiveAlpha() -> (skydiveAlpha:double)
   n159 = "Transform (Modify) Bone - Bone: dyn_skirt_bk_a_base_l"!(ComponentPose=@n157, Alpha:float=@n158, Translation="0,0,0", Rotation="15,0,0.000", Scale="1,1,1") -> (Pose) {
    rotationMode=BMM_Additive
    rotationSpace=BCS_BoneSpace
    alphaBoolBlend {blendOption=Linear}
    alphaScaleBiasClamp {bInterpResult=true, interpSpeedDecreasing=2, interpSpeedIncreasing=5}
    boneToModify {boneName=dyn_skirt_bk_a_base_l}
    rotation {pitch=15}
   }
   n160 = "Transform (Modify) Bone - Bone: dyn_skirt_bk_a_base"!(ComponentPose=@n159, Alpha:float=@n158, Translation="0,0,0", Rotation="15,0,0.000", Scale="1,1,1") -> (Pose) {
    rotationMode=BMM_Additive
    rotationSpace=BCS_BoneSpace
    alphaBoolBlend {blendOption=Linear}
    alphaScaleBiasClamp {bInterpResult=true, interpSpeedDecreasing=2, interpSpeedIncreasing=5}
    boneToModify {boneName=dyn_skirt_bk_a_base}
    rotation {pitch=15}
   }
   n161 = "Transform (Modify) Bone - Bone: dyn_skirt_bk_a_base_r"!(ComponentPose=@n160, Alpha:float=@n158, Translation="0,0,0", Rotation="15,0,0.000", Scale="1,1,1") -> (Pose) {
    rotationMode=BMM_Additive
    rotationSpace=BCS_BoneSpace
    alphaBoolBlend {blendOption=Linear}
    alphaScaleBiasClamp {bInterpResult=true, interpSpeedDecreasing=2, interpSpeedIncreasing=5}
    boneToModify {boneName=dyn_skirt_bk_a_base_r}
    rotation {pitch=15}
   }
   n162 = Get ParachuteAlpha() -> (ParachuteAlpha:double)
   n163 = "Transform (Modify) Bone - Bone: dyn_skirt_bk_a_l"(ComponentPose=@n161, Alpha=@n162, Translation="0,0,0", Rotation="15,10,0.000", Scale="1,1,1") {
    rotationMode=BMM_Additive
    rotationSpace=BCS_BoneSpace
    alphaBoolBlend {blendOption=Linear}
    alphaScaleBiasClamp {bInterpResult=true, interpSpeedDecreasing=2, interpSpeedIncreasing=5}
    boneToModify {boneName=dyn_skirt_bk_a_l}
    rotation {pitch=15, yaw=10}
   }
   n164 = "Transform (Modify) Bone - Bone: dyn_skirt_bk_a_r"(ComponentPose=@n163, Alpha=@n162, Translation="0,0,0", Rotation="15,-10,0.000", Scale="1,1,1") {
    rotationMode=BMM_Additive
    rotationSpace=BCS_BoneSpace
    alphaBoolBlend {blendOption=Linear}
    alphaScaleBiasClamp {bInterpResult=true, interpSpeedDecreasing=2, interpSpeedIncreasing=5}
    boneToModify {boneName=dyn_skirt_bk_a_r}
    rotation {pitch=15, yaw=-10}
   }
   n165 = "Transform (Modify) Bone - Bone: dyn_skirt_bk_a"(ComponentPose=@n164, Alpha=@n162, Translation="0,0,0", Rotation="15,0,0.000", Scale="1,1,1") {
    rotationMode=BMM_Additive
    rotationSpace=BCS_BoneSpace
    alphaBoolBlend {blendOption=Linear}
    alphaScaleBiasClamp {bInterpResult=true, interpSpeedDecreasing=2, interpSpeedIncreasing=5}
    boneToModify {boneName=dyn_skirt_bk_a}
    rotation {pitch=15}
   }
   n166 = "Transform (Modify) Bone - Bone: dyn_skirt_side_a_l"(ComponentPose=@n165, Alpha=@n162, Translation="0,0,0", Rotation="15,-15,0.000", Scale="1,1,1") {
    rotationMode=BMM_Additive
    rotationSpace=BCS_BoneSpace
    alphaBoolBlend {blendOption=Linear}
    alphaScaleBiasClamp {bInterpResult=true, interpSpeedDecreasing=2, interpSpeedIncreasing=5}
    boneToModify {boneName=dyn_skirt_side_a_l}
    rotation {pitch=15, yaw=-15}
   }
   n167 = "Transform (Modify) Bone - Bone: dyn_skirt_side_a_r"(ComponentPose=@n166, Alpha=@n162, Translation="0,0,0", Rotation="15,15,0.000", Scale="1,1,1") {
    rotationMode=BMM_Additive
    rotationSpace=BCS_BoneSpace
    alphaBoolBlend {blendOption=Linear}
    alphaScaleBiasClamp {bInterpResult=true, interpSpeedDecreasing=2, interpSpeedIncreasing=5}
    boneToModify {boneName=dyn_skirt_side_a_r}
    rotation {pitch=15, yaw=15}
   }
   n168 = Component To Local(ComponentPose=@n167)
   n169 = Get bIsCrouching()
   n170 = "Constraint - dyn_loop_a_MSTR (R)"(
    ComponentPose=Local To Component(LocalPose=Blend Poses by bool(BlendPose_0=@n147, BlendPose_1=@n168, bActiveValue=@n169) {blendType=Linear})
    Alpha=1
    ConstraintWeights_0=1
   ) {
    alphaBoolBlend {blendOption=Linear}
    boneToModify {boneName=dyn_loop_a_MSTR}
    constraintSetup [{transformType=Rotation, targetBone {boneName=dyn_skirt_fr_b}}]
    constraintWeights [1]
   }
   n171 = "Constraint - dyn_loop_b_MSTR (R)"(ComponentPose=@n170, Alpha=1, ConstraintWeights_0=1) {
    alphaBoolBlend {blendOption=Linear}
    boneToModify {boneName=dyn_loop_b_MSTR}
    constraintSetup [{transformType=Rotation, targetBone {boneName=dyn_skirt_fr_b}}]
    constraintWeights [1]
   }
   n172 = "Constraint - dyn_rope_a_MSTR (R)"(ComponentPose=@n171, Alpha=1, ConstraintWeights_0=1) {
    alphaBoolBlend {blendOption=Linear}
    boneToModify {boneName=dyn_rope_a_MSTR}
    constraintSetup [{transformType=Rotation, targetBone {boneName=dyn_skirt_fr_b_r}}]
    constraintWeights [1]
   }
   n173 = "Constraint - dyn_rope_b_MSTR (R)"(ComponentPose=@n172, Alpha=1, ConstraintWeights_0=1) {
    alphaBoolBlend {blendOption=Linear}
    boneToModify {boneName=dyn_rope_b_MSTR}
    constraintSetup [{transformType=Rotation, targetBone {boneName=dyn_skirt_fr_b_r}}]
    constraintWeights [1]
   }
   n174 = Get thigh_rot_up_r()
   n175 = "Transform (Modify) Bone - Bone: dyn_rope_a_MSTR"(ComponentPose=@n173, Alpha=@n174, Translation="0,-5,0", Rotation="-59.141,-0.993,-8.002", Scale="1,1,1") {
    rotationMode=BMM_Additive
    rotationSpace=BCS_BoneSpace
    translationMode=BMM_Additive
    translationSpace=BCS_BoneSpace
    alphaBoolBlend {blendOption=Linear}
    boneToModify {boneName=dyn_rope_a_MSTR}
    rotation=<rot -59.141 -0.993 -8.0021>
    translation {y=-5}
   }
   n176 = "Transform (Modify) Bone - Bone: dyn_rope_b_MSTR"(ComponentPose=@n175, Alpha=@n174, Translation="0,-5,-5", Rotation="-57.107,0.000,0.000", Scale="1,1,1") {
    rotationMode=BMM_Additive
    rotationSpace=BCS_BoneSpace
    translationMode=BMM_Additive
    translationSpace=BCS_BoneSpace
    alphaBoolBlend {blendOption=Linear}
    boneToModify {boneName=dyn_rope_b_MSTR}
    rotation {pitch=-57.107}
    translation {y=-5, z=-5}
   }
   Save cached pose 'no rigids'(Pose=Component To Local(ComponentPose=@n176)) {cacheName=no rigids}
   n177 = "Transform (Modify) Bone - Bone: dyn_skirt_bk_a_base_l"!(
    ComponentPose=Local To Component(LocalPose=Use cached pose 'skirt_bk_a_pose'())
    bAlphaBoolEnabled:bool=True
    Translation="0,0,0"
    Rotation="5,0,0"
    Scale="1,1,1"
   ) -> (Pose) {
    alphaInputType=Bool
    rotationMode=BMM_Additive
    rotationSpace=BCS_BoneSpace
    alphaBoolBlend {blendInTime=0.1, blendOption=Linear, blendOutTime=0.1}
    alphaScaleBiasClamp {bInterpResult=true, interpSpeedDecreasing=2, interpSpeedIncreasing=5}
    boneToModify {boneName=dyn_skirt_bk_a_base_l}
    rotation {pitch=5}
   }
   n178 = "Transform (Modify) Bone - Bone: dyn_skirt_bk_a_base_r"!(ComponentPose=@n177, bAlphaBoolEnabled:bool=True, Translation="0,0,0", Rotation="5,0,0", Scale="1,1,1") -> (Pose) {
    alphaInputType=Bool
    rotationMode=BMM_Additive
    rotationSpace=BCS_BoneSpace
    alphaBoolBlend {blendInTime=0.1, blendOption=Linear, blendOutTime=0.1}
    alphaScaleBiasClamp {bInterpResult=true, interpSpeedDecreasing=2, interpSpeedIncreasing=5}
    boneToModify {boneName=dyn_skirt_bk_a_base_r}
    rotation {pitch=5}
   }
   n179 = "Transform (Modify) Bone - Bone: dyn_skirt_bk_a"(ComponentPose=@n178, Alpha=1, Translation="0,0,0", Rotation="25,0,0", Scale="1,1,1") {
    rotationMode=BMM_Additive
    rotationSpace=BCS_BoneSpace
    scaleMode=BMM_Additive
    scaleSpace=BCS_BoneSpace
    translationMode=BMM_Additive
    translationSpace=BCS_BoneSpace
    alphaBoolBlend {blendOption=Linear}
    boneToModify {boneName=dyn_skirt_bk_a}
    rotation {pitch=25}
   }
   n180 = Component To Local(ComponentPose=@n179)
   n181 = Use cached pose 'skirt_bk_a_pose'()
   n182 = Get bIsSprinting()
   n183 = "Transform (Modify) Bone - Bone: dyn_sword_MSTR"(
    ComponentPose=Local To Component(LocalPose=Blend Poses by bool(BlendPose_0=@n180, BlendPose_1=@n181, BlendTime_0=0.3, BlendTime_1=0.3, bActiveValue=@n182) {blendType=Linear})
    bAlphaBoolEnabled=Get sword skydive() -> (sword skydive:bool)
    Translation="0,0.5,1"
    Rotation="-10,5,22"
    Scale="1,1,1"
   ) {
    alphaInputType=Bool
    rotationMode=BMM_Additive
    rotationSpace=BCS_BoneSpace
    translationMode=BMM_Additive
    translationSpace=BCS_BoneSpace
    alphaBoolBlend {blendInTime=0.1, blendOption=Linear, blendOutTime=0.5}
    boneToModify {boneName=dyn_sword_MSTR}
    rotation=<rot -10 5 22>
    translation {y=0.5, z=1}
   }
   n184 = RigidBody!(ComponentPose=@n183, Alpha, ExternalForce) -> (Pose) {
    bClampLinearTranslationLimitToRefPose=true
    bOverrideWorldGravity=true
    overridePhysicsAsset=&/M_MED_Headband_S_collar_Physics
    simulationSpace=BaseBoneSpace
    alphaBoolBlend {blendOption=Linear}
    baseBoneRef {boneName=spine_05}
    componentAppliedLinearAccClamp=<3000 3000 3000>
    componentLinearAccScale=<0.3 0.3 0.3>
    componentLinearVelScale=<0.3 0.3 0.3>
    overrideWorldGravity {z=-980}
   }
   n185 = Anim Dynamics - Bone: dyn_zipper!(ComponentPose=@n184, Alpha=0.25, GravityOverride:Vector="0,0,0", ExternalForce) -> (Pose) {
    angularDampingOverride=1
    angularSpringConstant=1500
    bAngularSpring=true
    bGravityOverrideInSimSpace=true
    bOverrideAngularDamping=true
    gravityScale=0.45
    simulationSpace=RootRelative
    alphaBoolBlend {blendOption=Linear}
    alphaScaleBias {scale=2}
    boundBone {boneName=dyn_zipper}
    componentAppliedLinearAccClamp=<2500 2500 2500>
    componentLinearAccScale=<1 1 1>
    componentLinearVelScale=<1 1 1>
    physicsBodyDefinitions [{
      sphereCollisionRadius=0
      boundBone {boneName=dyn_zipper}
      boxExtents=<10 5 5>
      constraintSetup {angularLimitsMax=<1 35 5>, angularLimitsMin=<-1 -35 -5>, angularTarget=<0 0 0>}
      localJointOffset=<2 0 0>
     }]
    relativeSpaceBone {boneName=root}
   }
   n186 = "Transform (Modify) Bone - Bone: dyn_sword_MSTR"!(
    ComponentPose=@n185
    Alpha:float=Get Shoulder_rot_back_r() -> (Shoulder_rot_back_r:double)
    Translation="0,0,0"
    Rotation="-40,0,0"
    Scale="1,1,1"
   ) -> (Pose) {
    rotationMode=BMM_Additive
    rotationSpace=BCS_BoneSpace
    scaleMode=BMM_Additive
    scaleSpace=BCS_BoneSpace
    translationMode=BMM_Additive
    translationSpace=BCS_BoneSpace
    alphaBoolBlend {blendOption=Linear}
    boneToModify {boneName=dyn_sword_MSTR}
    rotation {pitch=-40}
   }
   Save cached pose 'weapon_pose'(Pose=Component To Local(ComponentPose=@n186)) {cacheName=weapon_pose}
   n187 = Use cached pose 'constraints'()
   Save cached pose 'prerigid'(Pose=Component To Local(ComponentPose=Local To Component(LocalPose=@n187))) {cacheName=prerigid}
   Reroute Node(InputPin:Wildcard) -> (OutputPin:Wildcard)
   Reroute Node()
   Get thigh_rot_inr_l()
   Get thigh_rot_inr_r()
   Get thigh_rot_up_l()
  ]}
 EventGraph {nodes [
   Event BlueprintUpdateAnimation()
   > Get!(execute:Exec) -> (then:Exec, else:Exec, Skel_Mesh:Skeletal Mesh Component Object Reference)
   > n1 = Sequence!(execute) -> (then_0, then_1, then_2:Exec, then_3:Exec, then_4:Exec, then_5:Exec, then_6:Exec, then_7:Exec)
   n2 = Get Skel_Mesh()
   > Set thigh_rot_l(thigh_rot_l=GetAlphaFromReferenceTransformDelta(SkeletalMeshComponent=@n2, BoneName=thigh_l, BaseName=pelvis, InputRange=(X=-20,Y=-115), TwistAxis="1,0,0"))
   > Set thigh_rot_r(thigh_rot_r=GetAlphaFromReferenceTransformDelta(SkeletalMeshComponent=@n2, BoneName=thigh_r, BaseName=pelvis, InputRange=(X=-20,Y=-100), TwistAxis="1,0,0"))
   > Set thigh_rot_extra_r(thigh_rot_extra_r:double=GetAlphaFromReferenceTransformDelta(SkeletalMeshComponent=Get Skel_Mesh(), BoneName=thigh_r, BaseName=pelvis, InputRange=(X=-100,Y=-140), TwistAxis="1,0,0")) -> (Output_Get:double)
   > Set RBAN disabled(RBAN disabled:bool=NOT Boolean(A=Get bRigidBodyEnabled())) -> (Output_Get:bool)
   n3 = GetOwnerSkeletalMeshComponent()
   n4 = BreakRotator(InRot=BreakTransform(InTransform=GetDeltaTransformFromRefPose(self=@n3, BoneName=thigh_r, BaseName=pelvis)).Rotation)
   > Set thigh_rot_up_r(thigh_rot_up_r:double=MapRangeClamped(Value=@n4.Yaw, InRangeA=-15, InRangeB=-100)) -> (Output_Get:double)
   > Set thigh_rot_inr_r(thigh_rot_inr_r:double=MapRangeClamped(Value=@n4.Pitch, InRangeA=0, InRangeB=-10, OutRangeA=1, OutRangeB=0)) -> (Output_Get:double)
   > Set thigh_rot_otr_r(thigh_rot_otr_r:double=MapRangeClamped(Value=@n4.Pitch, InRangeA=20, InRangeB=50)) -> (Output_Get:double)
   n5 = GetOwnerSkeletalMeshComponent()
   n6 = BreakRotator(InRot=BreakTransform(InTransform=GetDeltaTransformFromRefPose(self=@n5, BoneName=thigh_l, BaseName=pelvis)).Rotation)
   > Set thigh_rot_up_l(thigh_rot_up_l:double=MapRangeClamped(Value=@n6.Yaw, InRangeA=-10, InRangeB=-100)) -> (Output_Get:double)
   > Set thigh_rot_inr_l(thigh_rot_inr_l:double=MapRangeClamped(Value=@n6.Pitch, InRangeA=0, InRangeB=-10, OutRangeA=1, OutRangeB=0)) -> (Output_Get:double)
   > Set thigh_rot_otr_l(thigh_rot_otr_l:double=MapRangeClamped(Value=@n6.Pitch, InRangeA=20, InRangeB=50)) -> (Output_Get:double)
   > Set jumpingAlpha(jumpingAlpha:double="To Float (Boolean)"(InBool=Get bIsJumping())) -> (Output_Get:double)
   n7 = Get bIsMoving2D()
   > Set isNotCrouchRunning(isNotCrouchRunning:double="To Float (Boolean)"(InBool=NOT Boolean(A=@n7))) -> (Output_Get:double)
   > Set ParachuteAlpha(ParachuteAlpha:double="To Float (Boolean)"(InBool=Get bIsParachuteOpened())) -> (Output_Get:double)
   > Set skydiveAlpha(skydiveAlpha:double="To Float (Boolean)"(InBool=Get bIsSkyDiving())) -> (Output_Get:double)
   > Set crouchingAlpha(crouchingAlpha:double="To Float (Boolean)"(InBool=Get bIsCrouching())) -> (Output_Get:double)
   > Set targeting(targeting:double="To Float (Boolean)"(InBool=Get bIsTargeting())) -> (Output_Get:double)
   > Set emote(emote:double="To Float (Boolean)"(InBool=Get bIsPlayingEmote())) -> (Output_Get:double)
   n8 = PlayerMovement(execute=@n1.then_1, AnimInstance=Self-Reference(), XDotProductCoefficient=1, YDotProductCoefficient=1, ZDotProductCoefficient=1)
   n9 = float + float(A=float * float(A="Absolute (Float)"(A=@n8), B=0.5), B=MapRangeClamped(Value=@n8.YMovement, InRangeA=0, InRangeB=-1)) {is=title K2Node_CommutativeAssociativeBinaryOperator}
   > Set BwdMovement(BwdMovement:double=MapRangeClamped(Value=@n9, InRangeA=0, InRangeB=1.1, OutRangeA=0.4, OutRangeB=0)) -> (Output_Get:double)
   n10 = MapRangeClamped(Value=@n8, InRangeA=-0.2, InRangeB=1)
   n11 = float * float(A="Absolute (Float)"(A=@n8.YMovement), B=0.2)
   > Set RightMovement(RightMovement:double=MapRangeClamped(Value=float + float(A=@n11, B=@n10), InRangeA=0, InRangeB=1.2, OutRangeB=0.5)) -> (Output_Get:double)
   n12 = MapRangeClamped(Value=@n8, InRangeA=0.2, InRangeB=-1, OutRangeB=0.3)
   n13 = float * float(A="Absolute (Float)"(A=@n8.YMovement), B=0.2)
   > Set LeftMovement(LeftMovement:double=MapRangeClamped(Value=float + float(A=@n13, B=@n12), InRangeA=0, InRangeB=1.2, OutRangeB=0.5)) -> (Output_Get:double)
   > n14 = Set BwdMovement_sword(BwdMovement_sword:double=MapRangeClamped(Value=@n9, InRangeA=0, InRangeB=1.1)) -> (Output_Get:double)
   > Set BwdMovement_sword_bool(BwdMovement_sword_bool:bool="float >= float"(A=@n14, B=0.5) {is=title K2Node_CallFunction}) -> (Output_Get:bool)
   n15 = Branch(execute=@n1.then_2, Condition=Get bIsSkyDiving())
   > n16 = Branch(Condition="float > float"(A=Get PawnVelocityZ(), B=50) {is=title K2Node_CallFunction})
   > Set BwdMovement(BwdMovement=float * float(A=Get BwdMovement() -> (BwdMovement:double), B=0))
   Set BwdMovement(execute=[@n16.else, @n15.else], BwdMovement=float * float(A=Get BwdMovement(), B=1))
   n17 = ComplexSimSpaceFunction DEPRECATED!(
    execute=@n1.then_3
    CharacterPartAnimInstance:Custom Character Part Anim Instance Object Reference=Self-Reference()
    InterpSpeedMax:double=9
    PawnSpeedMax:double=20
    RInterpErrorTolerance:double=1
    DeltaTime:double=0
    DebugDraw:bool=false
    sim_space_rotation_in:Rotator=Get sim_space_rotation() -> (sim_space_rotation:Rotator)
    inComponentSpaceIn:bool=Get InComponentSpace() -> (InComponentSpace:bool)
    inComponentSpaceLastFrameIn:bool=Get InComponentSpaceLastFrame() -> (InComponentSpaceLastFrame:bool)
    lockedOnSimSpaceIn:bool=Get LockedOnSimSpace() -> (LockedOnSimSpace:bool)
    InterpSpeedIn:double=Get InterpSpeed() -> (InterpSpeed:double)
   ) -> (
    then
    sim_space_rotation:Rotator
    inComponentSpaceOut:bool
    inComponentSpaceLastFrameOut:bool
    lockedOnSimSpaceOut:bool
    InterpSpeedOut:double
   )
   > Set sim_space_rotation(sim_space_rotation:Rotator=@n17) -> (Output_Get:Rotator)
   > Set InComponentSpace(InComponentSpace:bool=@n17.inComponentSpaceOut) -> (Output_Get:bool)
   > Set InComponentSpaceLastFrame(InComponentSpaceLastFrame:bool=@n17.inComponentSpaceLastFrameOut) -> (Output_Get:bool)
   > Set LockedOnSimSpace(LockedOnSimSpace:bool=@n17.lockedOnSimSpaceOut) -> (Output_Get:bool)
   > Set InterpSpeed(InterpSpeed:double=@n17.InterpSpeedOut) -> (Output_Get:double)
   n18 = Gravity Override State Function DEPRECATED(
    execute=@n1.then_4
    CharacterPartAnimInstance=Self-Reference()
    GameTimeInSeconds=GetGameTimeInSeconds()
    GravityOverrideParams=Get Gravity Override Params()
   )
   > Set GravityOverride(GravityOverride=@n18)
   Set sword rot alpha(
    execute=@n1.then_5
    sword rot alpha:double=GetAlphaFromReferenceTransformDelta(
     SkeletalMeshComponent=Get Skel_Mesh()
     BoneName=spine_05
     BaseName=clavicle_r
     DeltaComponentType=EulerX
     InputRange=(X=-1,Y=6)
     TwistAxis="1,0,0"
    )
   ) -> (Output_Get:double)
   n19 = Branch(execute=@n1.then_6, Condition=OR Boolean(A=Get bIsSkydiveFloating(), B=Get bIsSkydiveDiveMode()))
   > Set sword skydive(sword skydive:bool=true) -> (Output_Get:bool)
   Set sword skydive(execute=@n19.else, sword skydive=false)
   Set Shoulder_rot_back_r(
    execute=@n1.then_7
    Shoulder_rot_back_r:double=GetAlphaFromReferenceTransformDelta(
     SkeletalMeshComponent=Get Skel_Mesh()
     BoneName=upperarm_r
     BaseName=clavicle_r
     DeltaComponentType=EulerX
     InputRange=(X=-60,Y=60)
     TwistAxis="0,0,1"
    )
   ) -> (Output_Get:double)
   Event BlueprintInitializeAnimation()
   > Set Skel_Mesh(Skel_Mesh=GetOwnerSkeletalMeshComponent())
   n20 = Event BlueprintProcessAnimInputEvent()
   > n21 = Branch(Condition=IsNamed(self=@n20.AnimInputEvent, InEventName=aie.Player.Zipline.Entered))
   > Set zipline(zipline:bool=true) -> (Output_Get:bool)
   Branch(execute=[@n21.else, @n29.then], Condition=IsNamed(self=@n20.AnimInputEvent, InEventName=aie.Player.Zipline.Exited))
   > Set zipline(zipline=false)
   n22 = Event OnBeginSkydiving()
   n23 = Get Dynamic Material Instance(execute=[@n27.then, @n28.then], SkeletalMeshComponent=GetPartSkeletalMeshComponent(), ElementIndex=1)
   SetScalarParameterValue(execute=[@n22.then, @n32.then, @n26.else, @n31.else], self=@n23, ParameterName=Hide Element 01, Value=0)
   n24 = Event OnEndSkydiving()
   n25 = PickaxeItemDefinitionHasAnyMetaTag(Tags=Make Array("[0]"=(TagName="Cosmetics.MetaTags.HeadbandS")), bRequiresEquipped=true)
   n26 = Branch(execute=[@n24.then, @n30.then_1], Condition=@n25)
   SetScalarParameterValue(execute=[@n32.else, @n26.then], self=@n23, ParameterName=Hide Element 01, Value=1)
   n27 = Event OnWeaponChanged()
   n28 = Event OnCharacterCustomizationLoaded()
   n29 = Set zipline(zipline=false)
   TryGetPawnOwner()
   Get bIsCrouchSprinting()
   n30 = Sequence!(execute=@n23.then) -> (then_0, then_1)
   > n31 = Branch(Condition=@n25)
   > n32 = Branch(Condition=Get bIsSkyDiving())
  ]}
]
variables [
 2 * Alpha {type=number, default=1}
 BwdMovement {type=number}
 BwdMovement_sword {type=number}
 BwdMovement_sword_bool {type=boolean}
 crouchingAlpha {type=number}
 emote {type=number}
 Gravity Override Params {type=GravityOverrideParamsStruct}
 GravityOverride {type=Vector}
 InComponentSpace {type=boolean}
 InComponentSpaceLastFrame {type=boolean}
 InterpSpeed {type=number}
 isNotCrouchRunning {type=number}
 jumpingAlpha {type=number}
 LeftMovement {type=number}
 LockedOnSimSpace {type=boolean}
 ParachuteAlpha {type=number}
 RBAN disabled {type=boolean}
 RightMovement {type=number}
 Shoulder_rot_back_l {type=number}
 Shoulder_rot_back_l_0 {type=number}
 Shoulder_rot_back_r {type=number}
 Shoulder_rot_back_r_0 {type=number}
 sim_space_rotation {type=Rotator}
 Skel_Mesh {type=SkeletalMeshComponent}
 skydiveAlpha {type=number}
 sword rot alpha {type=number}
 sword skydive {type=boolean}
 targeting {type=number}
 thigh_rot_extra_r {type=number}
 thigh_rot_inr_l {type=number}
 thigh_rot_inr_r {type=number}
 thigh_rot_l {type=number}
 thigh_rot_otr_l {type=number}
 thigh_rot_otr_r {type=number}
 thigh_rot_r {type=number}
 thigh_rot_up_l {type=number}
 thigh_rot_up_r {type=number}
 zipline {type=boolean}
]
