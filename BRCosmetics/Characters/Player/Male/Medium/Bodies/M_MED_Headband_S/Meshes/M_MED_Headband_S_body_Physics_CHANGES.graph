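# Physics asset for the male medium Headband_S player body, covering a knot with dangling rope and loop chains, plus lower-body shapes.
# Two three-segment ropes and two two-segment loops branch from tiny master capsules on a knot constrained to the pelvis.
# Chain links have limited swing and tight twist, slightly looser on loops, while master, knot and body joints rotate freely.
# A root sphere and pelvis, thigh and calf capsules, sized asymmetrically left to right, form the body.
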PhysicsAsset /BRCosmetics/Characters/Player/Male/Medium/Bodies/M_MED_Headband_S/Meshes/M_MED_Headband_S_body_Physics_CHANGES
bodies {
 calf_l [{shapeType=Capsule, length=35.1309, radius=14.869, shapeName=calf_l_capsule, center=<-26.247 -3.0905 2.0001>, rotation=<rot 90 0 0>}]
 calf_r [{shapeType=Capsule, length=35.0459, radius=8.7107, shapeName=calf_r_capsule, center=<21.0609 0.8163 -0.7294>, rotation=<rot 90 0 0>}]
 dyn_knot_MSTR [{shapeType=Capsule, length=0.505, radius=0.505, shapeName=dyn_knot_MSTR_capsule, rotation=<rot 0 -90 0>}]
 dyn_loop_a_01 [{shapeType=Capsule, length=4.5146, radius=9, shapeName=dyn_loop_a_01_capsule, center=<6.0594 3 0>}]
 dyn_loop_a_02 [{shapeType=Capsule, length=1.8173, radius=9, shapeName=dyn_loop_a_02_capsule, center=<6.0132 3 0>}]
 dyn_loop_a_MSTR [{shapeType=Capsule, length=0.505, radius=0.505, shapeName=dyn_loop_a_MSTR_capsule, rotation=<rot 0 -90 0>}]
 dyn_loop_b_01 [{shapeType=Capsule, length=5.5, radius=7, shapeName=dyn_loop_b_01_capsule, center=<5.2687 3 0>, rotation=<rot -8.1843 0.4267 -2.9947>}]
 dyn_loop_b_02 [{shapeType=Capsule, length=5.8828, radius=7, shapeName=dyn_loop_b_02_capsule, center=<6.2555 3 0>, rotation=<rot 0 0 -4.7394>}]
 dyn_loop_b_MSTR [{shapeType=Capsule, length=0.505, radius=0.505, shapeName=dyn_loop_b_MSTR_capsule, rotation=<rot 0 -90 0>}]
 dyn_rope_a_01 [{shapeType=Capsule, length=9.0223, radius=3.0406, shapeName=dyn_rope_a_01_capsule, center=<5.2851 0 0>, rotation=<rot 90 0 0>}]
 dyn_rope_a_02 [{shapeType=Capsule, length=5.4196, radius=3.2985, shapeName=dyn_rope_a_02_capsule, center=<8.5 0 0>, rotation=<rot 90 0 0>}]
 dyn_rope_a_03 [{shapeType=Capsule, length=7.4414, radius=3.6209, shapeName=dyn_rope_a_03_capsule, center=<9.2025 0 0>, rotation=<rot 90 0 0>}]
 dyn_rope_a_MSTR [{shapeType=Capsule, length=0.505, radius=0.505, shapeName=dyn_rope_a_MSTR_capsule, rotation=<rot 0 -90 0>}]
 dyn_rope_b_01 [{shapeType=Capsule, length=5.9629, radius=1.9802, shapeName=dyn_rope_b_01_capsule, center=<3.7365 0 0>, rotation=<rot 90 0 0>}]
 dyn_rope_b_02 [{shapeType=Capsule, length=5.9007, radius=1.7576, shapeName=dyn_rope_b_02_capsule, center=<5.0848 0 0>, rotation=<rot 90 0 0>}]
 dyn_rope_b_03 [{shapeType=Capsule, length=7.4422, radius=3.4833, shapeName=dyn_rope_b_03_capsule, center=<7.0319 0 0>, rotation=<rot 90 0 0>}]
 dyn_rope_b_MSTR [{shapeType=Capsule, length=0.505, radius=0.505, shapeName=dyn_rope_b_MSTR_capsule, rotation=<rot 0 -90 0>}]
 pelvis [{shapeType=Capsule, length=12.5725, radius=9.3807, center=<3.5211 -0.0001 0>}]
 root [{shapeType=Sphere, radius=30.4872, shapeName=root_sphere, center=<0 2.5 8.5>}]
 thigh_l [{shapeType=Capsule, length=25, radius=14, shapeName=thigh_l_capsule, center=<-15 0 0>, rotation=<rot 90 -90 -89.1297>}]
 thigh_r [{shapeType=Capsule, length=24.5, radius=11.5, shapeName=thigh_r_capsule, center=<15.0485 0.7102 0>, rotation=<rot 90 -3.9099 0>}]
}
constraints [
 {bone1Name=dyn_rope_b_03, bone2Name=dyn_rope_b_02, swing1LimitDegrees=35, swing2LimitDegrees=35, twistLimitDegrees=3}
 {
  bone1Name=dyn_rope_a_MSTR
  bone2Name=dyn_knot_MSTR
  swing1LimitDegrees=45
  swing1Motion=Free
  swing2LimitDegrees=45
  swing2Motion=Free
  twistLimitDegrees=45
  twistMotion=Free
 }
 {bone1Name=dyn_rope_a_01, bone2Name=dyn_rope_a_MSTR, swing1LimitDegrees=35, swing2LimitDegrees=35, twistLimitDegrees=3}
 {bone1Name=dyn_rope_a_02, bone2Name=dyn_rope_a_01, swing1LimitDegrees=35, swing2LimitDegrees=35, twistLimitDegrees=3}
 {bone1Name=dyn_rope_a_03, bone2Name=dyn_rope_a_02, swing1LimitDegrees=35, swing2LimitDegrees=35, twistLimitDegrees=3}
 {
  bone1Name=dyn_rope_b_MSTR
  bone2Name=dyn_knot_MSTR
  swing1LimitDegrees=45
  swing1Motion=Free
  swing2LimitDegrees=45
  swing2Motion=Free
  twistLimitDegrees=45
  twistMotion=Free
 }
 {bone1Name=dyn_rope_b_01, bone2Name=dyn_rope_b_MSTR, swing1LimitDegrees=35, swing2LimitDegrees=35, twistLimitDegrees=3}
 {bone1Name=dyn_rope_b_02, bone2Name=dyn_rope_b_01, swing1LimitDegrees=35, swing2LimitDegrees=35, twistLimitDegrees=3}
 {
  bone1Name=dyn_knot_MSTR
  bone2Name=pelvis
  swing1LimitDegrees=45
  swing1Motion=Free
  swing2LimitDegrees=45
  swing2Motion=Free
  twistLimitDegrees=45
  twistMotion=Free
 }
 {
  bone1Name=thigh_r
  bone2Name=pelvis
  swing1LimitDegrees=45
  swing1Motion=Free
  swing2LimitDegrees=45
  swing2Motion=Free
  twistLimitDegrees=45
  twistMotion=Free
 }
 {
  bone1Name=thigh_l
  bone2Name=pelvis
  swing1LimitDegrees=45
  swing1Motion=Free
  swing2LimitDegrees=45
  swing2Motion=Free
  twistLimitDegrees=45
  twistMotion=Free
 }
 {
  bone1Name=dyn_loop_a_MSTR
  bone2Name=dyn_knot_MSTR
  swing1LimitDegrees=45
  swing1Motion=Free
  swing2LimitDegrees=45
  swing2Motion=Free
  twistLimitDegrees=45
  twistMotion=Free
 }
 {bone1Name=dyn_loop_a_01, bone2Name=dyn_loop_a_MSTR, swing1LimitDegrees=35, swing2LimitDegrees=35, twistLimitDegrees=10}
 {bone1Name=dyn_loop_a_02, bone2Name=dyn_loop_a_01, swing1LimitDegrees=35, swing2LimitDegrees=35, twistLimitDegrees=10}
 {
  bone1Name=dyn_loop_b_MSTR
  bone2Name=dyn_knot_MSTR
  swing1LimitDegrees=45
  swing1Motion=Free
  swing2LimitDegrees=45
  swing2Motion=Free
  twistLimitDegrees=45
  twistMotion=Free
 }
 {bone1Name=dyn_loop_b_01, bone2Name=dyn_loop_b_MSTR, swing1LimitDegrees=35, swing2LimitDegrees=35, twistLimitDegrees=10}
 {bone1Name=dyn_loop_b_02, bone2Name=dyn_loop_b_01, swing1LimitDegrees=35, swing2LimitDegrees=35, twistLimitDegrees=10}
 {
  bone1Name=pelvis
  bone2Name=root
  swing1LimitDegrees=45
  swing1Motion=Free
  swing2LimitDegrees=45
  swing2Motion=Free
  twistLimitDegrees=45
  twistMotion=Free
 }
 {
  bone1Name=calf_l
  bone2Name=thigh_l
  swing1LimitDegrees=45
  swing1Motion=Free
  swing2LimitDegrees=45
  swing2Motion=Free
  twistLimitDegrees=45
  twistMotion=Free
 }
 {
  bone1Name=calf_r
  bone2Name=thigh_r
  swing1LimitDegrees=45
  swing1Motion=Free
  swing2LimitDegrees=45
  swing2Motion=Free
  twistLimitDegrees=45
  twistMotion=Free
 }
]
