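# Physics asset for the face accessory part of the male medium Headband character, giving the hair bones physical secondary motion around the head.
# Each front, side and general hair bone has a single capsule body, alongside a larger capsule for the head.
# Every hair body is constrained directly to the head rather than chained, with tight swing and twist limits.
# Front locks swing mostly along one axis, while side locks get slightly wider limits.
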PhysicsAsset /BRCosmetics/Characters/Player/Male/Medium/Bodies/M_MED_Headband_S/Meshes/Parts/M_MED_Headband_S_FaceAcc_Physics
bodies {
 dyn_front_hair_01 [{shapeType=Capsule, length=5.7817, radius=0.9117, shapeName=dyn_front_hair_01_capsule, center=<4 0 0>, rotation=<rot 90 0 0>}]
 dyn_front_hair_02 [{shapeType=Capsule, length=7.1603, radius=0.9443, shapeName=dyn_front_hair_02_capsule, center=<5 0 0>, rotation=<rot 90 0 0>}]
 dyn_front_hair_03 [{shapeType=Capsule, length=6.9326, radius=1.0204, shapeName=dyn_front_hair_03_capsule, center=<5 0 0>, rotation=<rot 90 0 0>}]
 dyn_front_hair_04 [{shapeType=Capsule, length=5.5773, radius=1.264, shapeName=dyn_front_hair_04_capsule, center=<4 0 0>, rotation=<rot 90 0 0>}]
 dyn_hair_01 [{shapeType=Capsule, length=3.2096, radius=1.6184, shapeName=dyn_hair_01_capsule, center=<6 0.0619 0.2166>, rotation=<rot 90 0 0>}]
 dyn_hair_010 [{shapeType=Capsule, length=2.844, radius=2.5504, shapeName=dyn_hair_010_capsule, center=<3 -0.167 0.888>, rotation=<rot 90 0 0>}]
 dyn_hair_011 [{shapeType=Capsule, length=3.3197, radius=2.2655, shapeName=dyn_hair_011_capsule, center=<5 0.1455 -0.678>, rotation=<rot 90 0 0>}]
 dyn_hair_012 [{shapeType=Capsule, length=1.1829, radius=2.114, shapeName=dyn_hair_012_capsule, center=<4 -0.0369 0.1409>, rotation=<rot 90 0 0>}]
 dyn_hair_013 [{shapeType=Capsule, length=1.7758, radius=1.7088, shapeName=dyn_hair_013_capsule, center=<3 0.0537 -0.0352>, rotation=<rot 90 0 0>}]
 dyn_hair_014 [{shapeType=Capsule, length=3.2595, radius=1.5613, shapeName=dyn_hair_014_capsule, center=<6.9591 0.4088 0.5311>, rotation=<rot 90 0 0>}]
 dyn_hair_02 [{shapeType=Capsule, length=2.7465, radius=1.812, shapeName=dyn_hair_02_capsule, center=<5 -0.0619 0.0258>, rotation=<rot 90 0 0>}]
 dyn_hair_03 [{shapeType=Capsule, length=3.0461, radius=2.0818, shapeName=dyn_hair_03_capsule, center=<8.2441 -0.2687 0.4767>, rotation=<rot 90 0 0>}]
 dyn_hair_04 [{shapeType=Capsule, length=3.5691, radius=1.5293, shapeName=dyn_hair_04_capsule, center=<6.6152 -0.0553 -0.6311>, rotation=<rot 90 0 0>}]
 dyn_hair_05 [{shapeType=Capsule, length=4.0337, radius=2.8051, shapeName=dyn_hair_05_capsule, center=<5 -0.0903 0.8198>, rotation=<rot 90 0 0>}]
 dyn_hair_06 [{shapeType=Capsule, length=2.954, radius=1.3193, shapeName=dyn_hair_06_capsule, center=<5 -0.4498 -2.323>, rotation=<rot 90 0 0>}]
 dyn_hair_07 [{shapeType=Capsule, length=2.1566, radius=1.3209, shapeName=dyn_hair_07_capsule, center=<3 -0.0742 -0.0036>, rotation=<rot 90 0 0>}]
 dyn_hair_08 [{shapeType=Capsule, length=4.0963, radius=2.841, shapeName=dyn_hair_08_capsule, center=<4 -0.5054 -0.2544>, rotation=<rot 90 0 0>}]
 dyn_hair_09 [{shapeType=Capsule, length=3.9931, radius=3.1601, shapeName=dyn_hair_09_capsule, center=<6 -0.0957 0.6518>, rotation=<rot 90 0 0>}]
 dyn_l_side_hair_01 [{shapeType=Capsule, length=6.6683, radius=1.8044, shapeName=dyn_l_side_hair_01_capsule, center=<9 0 0>, rotation=<rot 90 0 0>}]
 dyn_l_side_hair_02 [{shapeType=Capsule, length=3.4484, radius=1.1862, shapeName=dyn_l_side_hair_02_capsule, center=<3 0 0>, rotation=<rot 90 0 0>}]
 dyn_l_side_hair_03 [{shapeType=Capsule, length=5.7743, radius=0.8763, shapeName=dyn_l_side_hair_03_capsule, center=<4 0 0>, rotation=<rot 90 0 0>}]
 dyn_r_side_hair_01 [{shapeType=Capsule, length=8.885, radius=1.4486, shapeName=dyn_r_side_hair_01_capsule, center=<6 0 0>, rotation=<rot 90 0 0>}]
 dyn_r_side_hair_02 [{shapeType=Capsule, length=6.0998, radius=1.3548, shapeName=dyn_r_side_hair_02_capsule, center=<3 0 0>, rotation=<rot 90 0 0>}]
 head [{shapeType=Capsule, length=7.193, radius=6.5868, center=<6.6196 3.3592 0>, rotation=<rot 90 0 0>}]
}
constraints [
 {bone1Name=dyn_hair_01, bone2Name=head, swing1LimitDegrees=10, swing2LimitDegrees=10, twistLimitDegrees=3}
 {bone1Name=dyn_hair_02, bone2Name=head, swing1LimitDegrees=10, swing2LimitDegrees=10, twistLimitDegrees=3}
 {bone1Name=dyn_hair_03, bone2Name=head, swing1LimitDegrees=10, swing2LimitDegrees=10, twistLimitDegrees=3}
 {bone1Name=dyn_hair_04, bone2Name=head, swing1LimitDegrees=10, swing2LimitDegrees=10, twistLimitDegrees=3}
 {bone1Name=dyn_hair_05, bone2Name=head, swing1LimitDegrees=10, swing2LimitDegrees=10, twistLimitDegrees=3}
 {bone1Name=dyn_hair_06, bone2Name=head, swing1LimitDegrees=10, swing2LimitDegrees=10, twistLimitDegrees=3}
 {bone1Name=dyn_hair_09, bone2Name=head, swing1LimitDegrees=10, swing2LimitDegrees=10, twistLimitDegrees=3}
 {bone1Name=dyn_hair_07, bone2Name=head, swing1LimitDegrees=10, swing2LimitDegrees=10, twistLimitDegrees=3}
 {bone1Name=dyn_hair_08, bone2Name=head, swing1LimitDegrees=10, swing2LimitDegrees=10, twistLimitDegrees=3}
 {bone1Name=dyn_hair_010, bone2Name=head, swing1LimitDegrees=10, swing2LimitDegrees=10, twistLimitDegrees=3}
 {bone1Name=dyn_hair_011, bone2Name=head, swing1LimitDegrees=10, swing2LimitDegrees=10, twistLimitDegrees=3}
 {bone1Name=dyn_hair_012, bone2Name=head, swing1LimitDegrees=10, swing2LimitDegrees=10, twistLimitDegrees=3}
 {bone1Name=dyn_hair_013, bone2Name=head, swing1LimitDegrees=10, swing2LimitDegrees=10, twistLimitDegrees=3}
 {bone1Name=dyn_hair_014, bone2Name=head, swing1LimitDegrees=10, swing2LimitDegrees=10, twistLimitDegrees=3}
 {bone1Name=dyn_l_side_hair_02, bone2Name=head, swing1LimitDegrees=15, swing2LimitDegrees=10, twistLimitDegrees=3}
 {bone1Name=dyn_r_side_hair_02, bone2Name=head, swing1LimitDegrees=15, swing2LimitDegrees=10, twistLimitDegrees=3}
 {bone1Name=dyn_front_hair_01, bone2Name=head, swing1LimitDegrees=15, swing2LimitDegrees=3, twistLimitDegrees=3}
 {bone1Name=dyn_front_hair_02, bone2Name=head, swing1LimitDegrees=15, swing2LimitDegrees=3, twistLimitDegrees=3}
 {bone1Name=dyn_front_hair_03, bone2Name=head, swing1LimitDegrees=15, swing2LimitDegrees=3, twistLimitDegrees=3}
 {bone1Name=dyn_front_hair_04, bone2Name=head, swing1LimitDegrees=15, swing2LimitDegrees=3, twistLimitDegrees=3}
 {bone1Name=dyn_l_side_hair_03, bone2Name=head, swing1LimitDegrees=15, swing2LimitDegrees=10, twistLimitDegrees=3}
 {bone1Name=dyn_r_side_hair_01, bone2Name=head, swing1LimitDegrees=15, swing2LimitDegrees=10, twistLimitDegrees=3}
 {bone1Name=dyn_l_side_hair_01, bone2Name=head, swing1LimitDegrees=15, swing2LimitDegrees=10, twistLimitDegrees=3}
]
